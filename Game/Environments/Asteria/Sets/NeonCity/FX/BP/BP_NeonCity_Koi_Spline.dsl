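; BP_NeonCity_Koi_Spline is a Blueprint wrapper for the NS_NeonCity_Koi_Spline Niagara effect in the NeonCity environment set.
; Its Initialize function pushes the GlowColor and Radius variables into the effect's BaseColor and Radius parameters.
; Initialize runs from both the construction script and BeginPlay, so the effect reflects those values in editor and at runtime.
; Tick and actor overlap events are empty.
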(event EventBeginPlay
  (CallFunction|Initialize))

(event Collision|EventActorBeginOverlap
  (params (OtherActor :type Actor%20Object%20Reference)))

(event EventTick
  (params (DeltaSeconds :type Float%20%28single-precision%29)))

(fn Initialize
  (Niagara|SetNiagaraVariable(LinearColor) (Variables|Default|GetNS_NeonCity_Koi_Spline) "BaseColor" (Variables|Default|GetGlowColor))
  (Niagara|SetNiagaraVariable(Float) (Variables|Default|GetNS_NeonCity_Koi_Spline) "Radius" (Variables|Default|GetRadius)))

(fn ConstructionScript
  (CallFunction|Initialize))
