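# Two-sided, default-lit material for Niagara mesh particles in the NeonCity effects set, named for cherry petals.
# A linear gradient along UV V drives a subtle shading ramp that is tinted by particle color.
# That tint feeds a saturated branch and a branch scaled by a dynamic parameter for per-particle emissive strength, alongside a scalar Roughness parameter.
# It samples no textures.
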Material /Game/Environments/Asteria/Sets/NeonCity/FX/Materials/M_Cherry_Petal_MeshParticle_01
graph {
 properties {bUsedWithNiagaraMeshParticles=true, pixelDepthOffsetMode=PDOM_Legacy, twoSided=true, usedShadingModels=Default Lit}
 nodes [
  n1 = TextureCoordinate()
  n2 = Multiply(
   A=LinearInterpolate(Alpha=MaterialFunctionCall(UV Channel=@n1) {materialFunction=&LinearGradient}.VGradient) {constA=0.7}
   B=ParticleColor().RGB
  )
  Multiply(A=@n2, B=DynamicParameter() {paramNames [Emmisve,Param2,Param3,Param4]}.Emmisve)
  Saturate(None=@n2)
  ScalarParameter() {defaultValue=0.6, parameterName=Roughness}
 ]
}
