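# Unlit translucent flame material for Niagara mesh particles, used by the NeonCity CyberDragon effect.
# A panning swirl texture, UV-warped by a second swirl sample, is eroded through SmoothStep by inverse particle alpha.
# Axis-aligned fresnel and a UV gradient shape particle-colored emissive and opacity, with vertex-blue and optional near-camera fades.
# Dynamic parameters drive glow, gradient, fade, softness and warp; Low quality skips the warp and extra mask.
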Material /Game/Environments/Asteria/Sets/NeonCity/FX/Materials/M_NeonCity_CyberDragon_Flame
graph {
 properties {
  bUsedWithNiagaraMeshParticles=true
  blendMode=BLEND_Translucent
  pixelDepthOffsetMode=PDOM_Legacy
  shadingModel=MSM_Unlit
  translucencyLightingMode=TLM_VolumetricDirectional
  usedShadingModels=Unlit
 }
 nodes [
  n1 = TextureCoordinate() {vTiling=0.2}
  Add(A=@n1, B=Frac(None=Constant2Vector() {g=8, r=3}))
  n2 = MaterialFunctionCall(Axis=Constant3Vector() {constant=<rgba 0 1 0 0>}, Exponent=ScalarParameter() {defaultValue=5, parameterName="Exponent (S)"}) {materialFunction=&AxisAlignedFresnel}
  n3 = DynamicParameter() {defaultValue=<rgba 1 1 0 1>, paramNames [Emmisive,PanY,Fade,Param4]}
  n4 = LinearInterpolate(A=VertexColor().B, B=Constant(), Alpha=@n3.Fade) {constAlpha=0}
  n5 = OneMinus(None=Power(Base=TextureCoordinate(), Exp=@n3.PanY) {constExponent=0})
  n6 = ComponentMask(None=ComponentMask(None=Clamp(None=@n5)) {g=true}) {g=true}
  n7 = TextureCoordinate() {vTiling=2}
  n8 = Multiply(A=ScalarParameter() {defaultValue=5, parameterName=WarpTiling}, B=@n1)
  n9 = ComponentMask(None=TextureSample(UVs=Panner(Coordinate=@n8) {speedY=-0.2}) {samplerType=SAMPLERTYPE_LinearColor, texture=/Game/Global/Textures/Default/Flowmaps/T_SwirlFlow_02}.RGB) {g=true, r=true}
  n10 = DynamicParameter() {parameterIndex=1, defaultValue=<rgba 0.1 1 0.5 1>, paramNames [SmoothstepSoftness, Random, WarpStrength, Panning Speed]}
  n11 = QualitySwitch(Default=Add(A=@n7, B=Multiply(A=@n9, B=@n10.WarpStrength)), Low=@n7)
  n12 = Power(Base=TextureSample(UVs=Panner(Coordinate=@n11) {speedX=0.1, speedY=-0.3}) {samplerType=SAMPLERTYPE_LinearColor, texture=&T_SwirlFlow_02}.G) {constExponent=8}
  n13 = OneMinus(None=ParticleColor().A)
  n14 = DynamicParameter() {parameterIndex=1, defaultValue=<rgba 0.1 1 0.5 1>, paramNames [SmoothstepSoftness, Random, WarpStrength, Panning Speed]}
  n15 = Add(A=@n13, B=@n14.SmoothstepSoftness)
  n16 = QualitySwitch(
   Default=Power(Base=TextureSample(UVs=@n1) {samplerType=SAMPLERTYPE_LinearColor, texture=&T_SwirlFlow_02}.G) {constExponent=0.5}
   Low=Constant() {r=1}
  )
  n17 = Multiply(
   A=MaterialFunctionCall(Distance=ScalarParameter() {defaultValue=256, parameterName=Near Camera Fade Distance}) {materialFunction=&Near_Camera_Fade}.Result
   B=@n16
  )
  n18 = Multiply(
   A=Multiply(
    A=MaterialFunctionCall(Alpha=@n12, Min=@n13, Max=@n15) {materialFunction=&SmoothStep}.Result
    B=StaticSwitchParameter(True=@n17, False=@n16) {parameterName=Use Near Camera Fade}
   )
   B=ScalarParameter() {defaultValue=1, parameterName=OverallOpacity}
  )
  Multiply(A=Multiply(A=@n2.Fresnel, B=Multiply(A=@n4, B=@n6)), B=@n18)
  Multiply(A=VertexNormalWS(), B=Multiply(A=ScalarParameter() {defaultValue=1}, B=@n18))
  Multiply()
  n19 = ParticleColor()
  Multiply(A=@n2.Fresnel, B=Multiply(A=Multiply(A=@n19.RGB, B=@n3.Emmisive), B=@n6))
 ]
}
