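# Masked, lit Niagara mesh-particle material for the NeonCity cyberdragon's holographic flame.
# Panning world-space noise, using a different noise function on forward and mobile, drives a Fresnel exponent feeding emissive and erosion opacity, tinted by particle color.
# Atlas scanlines add hologram banding, and inverted noise wobbles vertices along normals, weighted by inverted V.
# Dynamic parameters control the effect, with HDR intensity clamped on forward and mobile.
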Material /Game/Environments/Asteria/Sets/NeonCity/FX/Materials/M_NeonCity_Cyberdragon_Hologram_Flame_Mesh
graph {
 properties {bUsedWithNiagaraMeshParticles=true, blendMode=BLEND_Masked, pixelDepthOffsetMode=PDOM_Legacy, usedShadingModels=Default Lit}
 nodes [
  n1 = DynamicParameter() {paramNames [Emmisive, Fresnel, HDR Mult, Erode]}
  n2 = Add(
   A=Multiply(
    A=Time() {bIgnorePause=true, bOverride_Period=true, period=100}
    B=VectorParameter() {parameterName=PanSpeed, defaultValue=<rgba 250 1500 -500 1>}.RGB
   ) {constB=250}
   B=WorldPosition().XYZ
  )
  n3 = Noise!(World Position=@n2, FilterWidth) {bTiling=true, bTurbulence=false, levelScale=1, levels=2, noiseFunction=NOISEFUNCTION_GradientTex3D, repeatSize=500, scale=0.0015}
  n4 = Noise(World Position=@n2) {bTiling=true, bTurbulence=false, levelScale=1, levels=2, noiseFunction=NOISEFUNCTION_ValueALU, repeatSize=500, scale=0.002}
  n5 = ShadingPathSwitch(Default=@n3, Forward=@n4, Mobile=@n4)
  n6 = Fresnel(ExponentIn=@n5, BaseReflectFractionIn=@n1.Fresnel)
  n7 = OneMinus(None=Multiply(A=@n1.Erode, B=@n6))
  n8 = ParticleColor()
  Clamp(None=Multiply(A=Multiply(A=@n7, B=@n8.A), B=NamedRerouteUsage()))
  n9 = Multiply(A=NamedRerouteUsage(), B=@n8.RGB)
  n10 = Multiply(A=@n1.Emmisive, B=@n6)
  n11 = DynamicParameter() {paramNames [Emmisive, Fresnel, HDR Mult, Erode]}
  n12 = Clamp(None=@n11.HDR Mult) {maxDefault=3}
  MaterialFunctionCall(Emissive=MaterialFunctionCall(LDR Input=Multiply(A=@n9, B=@n10), HDR Intensity Multiplier=ShadingPathSwitch(Default=@n11.HDR Mult, Forward=@n12, Mobile=@n12)) {materialFunction=&DeriveHDRfromLDR}.RemappedRange) {materialFunction=&MF_FX_EmissiveAdjustments}
  n13 = DynamicParameter() {parameterIndex=1, defaultValue=<rgba 250 1 1 1>, paramNames [Wobble,Param2,Param3,Param4]}
  n14 = OneMinus(None=ComponentMask(None=TextureCoordinate()) {g=true})
  Multiply(A=VertexNormalWS(), B=Multiply(A=OneMinus(None=@n5), B=Multiply(A=@n13.Wobble, B=@n14)) {constB=50})
  n15 = DynamicParameter() {
   parameterIndex=2
   defaultValue=<rgba 150 0.008 0.2 0.4>
   paramNames [Scanline_Amount, Scanline_Speed, Scanline_Brightness, Scanline_BlockOpacity]
  }
  n16 = DynamicParameter() {
   parameterIndex=2
   defaultValue=<rgba 150 0.008 0.2 0.4>
   paramNames [Scanline_Amount, Scanline_Speed, Scanline_Brightness, Scanline_BlockOpacity]
  }
  n17 = DynamicParameter() {
   parameterIndex=2
   defaultValue=<rgba 150 0.008 0.2 0.4>
   paramNames [Scanline_Amount, Scanline_Speed, Scanline_Brightness, Scanline_BlockOpacity]
  }
  n18 = Time()
  n19 = MaterialFunctionCall!(
   Number of Scanlines=@n15.Scanline_Amount
   Scanline Speed=@n16.Scanline_Speed
   "Scanlines Thickness (Bias)"
   Scanlines Dark Value=ScalarParameter() {defaultValue=0.5, parameterName=Scanline Darkness}
   Scanlines Light Value
   Block Opacity=@n17.Scanline_BlockOpacity
   UVs
   Time=Multiply(A=Multiply(A=@n18, B=@n16.Scanline_Speed)) {constB=20}
   Block Repeat=Constant() {r=3}
  ) {materialFunction=/Game/Effects/Fort_Effects/Effects/Atlas/Materials/MF_Atlas_Scanlines_01}
  n20 = DynamicParameter() {
   parameterIndex=2
   defaultValue=<rgba 150 0.008 0.2 0.4>
   paramNames [Scanline_Amount, Scanline_Speed, Scanline_Brightness, Scanline_BlockOpacity]
  }
  NamedRerouteDeclaration(None=Add(A=@n19.Result, B=Multiply(A=@n19.Secondary Lines, B=@n20.Scanline_Brightness) {constB=3})) {name=Scanlines, nodeColor=<rgba 1 0 0.1641 1>}
  ScalarParameter() {defaultValue=1, parameterName=Metallic}
  ScalarParameter() {defaultValue=1, parameterName=Specular}
  ScalarParameter() {defaultValue=5, parameterName=Scanline amount}
  ScalarParameter() {parameterName=Roughness}
  VectorParameter() {parameterName=BaseColor, defaultValue=<rgba 0 0 0 1>}
 ]
}
