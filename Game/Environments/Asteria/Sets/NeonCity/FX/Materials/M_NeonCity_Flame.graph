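# Translucent unlit Niagara sprite material for a NeonCity flame effect with animated scanlines.
# A packed flipbook is sampled by SubUV with dynamic-parameter frame and image count; the channel-select If takes blue in every branch.
# That mask multiplies particle alpha, and particle color times a Scanlines reroute.
# The scanlines come from the Atlas scanline function, driven by Time and a second dynamic parameter set for amount, speed, brightness and block opacity.
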Material /Game/Environments/Asteria/Sets/NeonCity/FX/Materials/M_NeonCity_Flame
graph {
 properties {bUsedWithNiagaraSprites=true, blendMode=BLEND_Translucent, pixelDepthOffsetMode=PDOM_Legacy, shadingModel=MSM_Unlit, usedShadingModels=Unlit}
 nodes [
  n1 = DynamicParameter() {paramNames [Param1, Packed Channel, SubUV_Images, SubUV_Frame]}
  n2 = TextureObjectParameter() {group=Flame Channels, parameterName=Texture, sortPriority=1, texture=/Game/Effects/Fort_Effects/Textures/Fire/2DGasTorch_Flipbk_packed}
  n3 = MaterialFunctionCall(Float2=TextureCoordinate()) {materialFunction=&BreakOutFloat2Components}
  n4 = MaterialFunctionCall(Float3=MaterialFunctionCall!(
    Texture=@n2
    UVs=MaterialFunctionCall(X=@n3.R, Y=@n3.R) {materialFunction=&MakeFloat2}.Result
    SubImages=AppendVector(A=@n1.SubUV_Images, B=@n1.SubUV_Images)
    Frame=@n1.SubUV_Frame
   ) {materialFunction=/Engine/Functions/Engine_MaterialFunctions03/Texturing/SubUV_Function_MipDerivative}.RGB) {materialFunction=&BreakOutFloat3Components}
  n5 = If(A=@n1.Packed Channel, "A > B"=@n4.B, "A == B"=@n4.B, "A < B"=@n4.B) {constB=1, desc=pick flipbook channels, equalsThreshold=0.1}
  n6 = ParticleColor()
  Multiply(A=@n5, B=@n6.A)
  Multiply(A=NamedRerouteUsage(), B=Multiply(A=@n5, B=@n6.RGB))
  n7 = DynamicParameter() {
   parameterIndex=3
   defaultValue=<rgba 150 0.008 0.2 0.4>
   paramNames [Scanline_Amount, Scanline_Speed, Scanline_Brightness, Scanline_BlockOpacity]
  }
  n8 = DynamicParameter() {
   parameterIndex=3
   defaultValue=<rgba 150 0.008 0.2 0.4>
   paramNames [Scanline_Amount, Scanline_Speed, Scanline_Brightness, Scanline_BlockOpacity]
  }
  n9 = DynamicParameter() {
   parameterIndex=3
   defaultValue=<rgba 150 0.008 0.2 0.4>
   paramNames [Scanline_Amount, Scanline_Speed, Scanline_Brightness, Scanline_BlockOpacity]
  }
  n10 = Time()
  n11 = MaterialFunctionCall!(
   Number of Scanlines=@n7.Scanline_Amount
   Scanline Speed=@n8.Scanline_Speed
   "Scanlines Thickness (Bias)"
   Scanlines Dark Value=ScalarParameter() {defaultValue=0.5, parameterName=Scanline Darkness}
   Scanlines Light Value
   Block Opacity=@n9.Scanline_BlockOpacity
   UVs
   Time=Multiply(A=Multiply(A=@n10, B=@n8.Scanline_Speed)) {constB=20}
   Block Repeat=Constant() {r=1}
  ) {materialFunction=/Game/Effects/Fort_Effects/Effects/Atlas/Materials/MF_Atlas_Scanlines_01}
  n12 = DynamicParameter() {
   parameterIndex=3
   defaultValue=<rgba 150 0.008 0.2 0.4>
   paramNames [Scanline_Amount, Scanline_Speed, Scanline_Brightness, Scanline_BlockOpacity]
  }
  NamedRerouteDeclaration(None=Add(A=@n11.Result, B=Multiply(A=@n11.Secondary Lines, B=@n12.Scanline_Brightness) {constB=3})) {name=Scanlines, nodeColor=<rgba 1 0 0.1641 1>}
 ]
}
