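# Masked, lit material for koi Niagara mesh particles in the Asteria NeonCity set, with animated scanlines and fish-like motion.
# A panned gradient, offset by vertex alpha and masked by vertex red, drives a Wiggle offset.
# A Bend rotates vertices around the particle's Z axis along UV1; a static switch enables summed world position offset.
# Niagara dynamic parameters control wiggle, bend, scanlines and emissive strength, alongside particle color and a fish-scale normal map.
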Material /Game/Environments/Asteria/Sets/NeonCity/FX/Materials/M_NeonCity_Koi
graph {
 properties {bUsedWithNiagaraMeshParticles=true, blendMode=BLEND_Masked, pixelDepthOffsetMode=PDOM_Legacy, usedShadingModels=Default Lit}
 nodes [
  ActorPositionWS()
  n1 = NamedRerouteUsage()
  n2 = ParticleColor()
  MaterialFunctionCall(Emissive=Multiply(A=Multiply(A=@n1, B=@n2.RGB), B=DynamicParameter() {parameterIndex=1, paramNames [Emmisive,Param2,Param3,Param4]}.Emmisive)) {materialFunction=&MF_FX_EmissiveAdjustments}
  n3 = TextureCoordinate() {coordinateIndex=1}
  Multiply(A=@n3, B=Constant2Vector() {g=3, r=3})
  n4 = VertexColor()
  n5 = Multiply(
   A=Add(A=@n4.A, B=TextureCoordinate() {coordinateIndex=1})
   B=ComponentMask(None=VectorParameter() {group=WPO, parameterName=TextureWPO_UVtiling, defaultValue=<rgba 1 0 0 0>}.RGB) {g=true, r=true}
  )
  n6 = DynamicParameter() {paramNames [TextureWPO_Speed_Mult, TextureWPO_UVTiling_Mult, AmountWiggle_Random, BendAmount]}
  n7 = LinearInterpolate(
   A=@n5
   B=Multiply(A=@n5, B=@n6.TextureWPO_UVTiling_Mult)
   Alpha=ScalarParameter() {group=WPO, parameterName=TextureWPO_UVTiling_Per_Particle}
  )
  n8 = Add(A=@n7, B=@n6.AmountWiggle_Random)
  n9 = ScalarParameter() {group=WPO, parameterName=TextureWPO_UVSpeed_IntegrateFromNiagara}
  n10 = VectorParameter() {group=WPO, parameterName=TextureWPO_Speed, defaultValue=<rgba 1 0 0 0>}
  n11 = AppendVector(
   A=Add(A=ComponentMask(None=@n10.RGB) {r=true}, B=LinearInterpolate(Alpha=@n4.A) {constA=0.15, constB=0.35})
   B=ComponentMask(None=@n10.RGB) {g=true}
  )
  n12 = Multiply(A=@n11, B=@n6.TextureWPO_Speed_Mult)
  n13 = ScalarParameter() {group=WPO, parameterName=TextureWPO_Per_Particle}
  n14 = TextureSample(UVs=Panner(Coordinate=LinearInterpolate(A=@n7, B=@n8, Alpha=@n9), Speed=LinearInterpolate(A=@n11, B=@n12, Alpha=@n13)) {speedX=0.15}) {texture=/Game/Global/Textures/Default/Masks_and_Gradients/WeaponGradients/T_Gradient_DesertSky_Color}
  n15 = Multiply(A=@n14.R, B=ScalarParameter() {defaultValue=1, group=WPO, parameterName=WPO_HeightX}) {constB=0.5}
  n16 = Multiply(
   A=@n14.R
   B=Multiply(
    A=ScalarParameter() {defaultValue=1, group=WPO, parameterName=WPO_HeightY}
    B=DynamicParameter() {paramNames [TextureWPO_Speed_Mult, TextureWPO_UVTiling_Mult, AmountWiggle_Random, BendAmount]}.AmountWiggle_Random
   )
  ) {constB=0.5}
  n17 = ScalarParameter() {defaultValue=1, group=WPO, parameterName=WPO_HeightZ}
  n18 = LinearInterpolate(
   A=Constant3Vector()
   B=AppendVector(A=AppendVector(A=@n15, B=@n16), B=Multiply(A=@n14.R, B=@n17) {constB=0.6})
   Alpha=LinearInterpolate(B=ScalarParameter() {group=WPO, parameterName=WPO_EyeMovent}, Alpha=VertexColor().R) {constA=1, constB=0}
  )
  n19 = ScalarParameter() {defaultValue=1, group=WPO, parameterName=WPO_Wave_Intenstiy}
  NamedRerouteDeclaration(None=Transform(None=Multiply(A=@n18, B=@n19) {constB=80}) {transformSourceType=TRANSFORMSOURCE_Local}) {name=Wiggle, nodeColor=<rgba 1 0 0.7555 1>}
  n20 = DynamicParameter() {
   parameterIndex=2
   defaultValue=<rgba 100 0.004 1 0.2>
   paramNames [Scanline_Amount, Scanline_Speed, Scanline_Brightness, Scanline_BlockOpacity]
  }
  n21 = DynamicParameter() {
   parameterIndex=2
   defaultValue=<rgba 100 0.004 1 0.2>
   paramNames [Scanline_Amount, Scanline_Speed, Scanline_Brightness, Scanline_BlockOpacity]
  }
  n22 = DynamicParameter() {
   parameterIndex=2
   defaultValue=<rgba 100 0.004 1 0.2>
   paramNames [Scanline_Amount, Scanline_Speed, Scanline_Brightness, Scanline_BlockOpacity]
  }
  n23 = Time()
  n24 = MaterialFunctionCall!(
   Number of Scanlines=@n20.Scanline_Amount
   Scanline Speed=@n21.Scanline_Speed
   "Scanlines Thickness (Bias)"
   Scanlines Dark Value=ScalarParameter() {defaultValue=0.5, parameterName=Scanline Darkness}
   Scanlines Light Value
   Block Opacity=@n22.Scanline_BlockOpacity
   UVs
   Time=Multiply(A=Multiply(A=@n23, B=@n21.Scanline_Speed)) {constB=20}
   Block Repeat=Constant() {r=3}
  ) {materialFunction=/Game/Effects/Fort_Effects/Effects/Atlas/Materials/MF_Atlas_Scanlines_01}
  n25 = DynamicParameter() {
   parameterIndex=2
   defaultValue=<rgba 100 0.004 1 0.2>
   paramNames [Scanline_Amount, Scanline_Speed, Scanline_Brightness, Scanline_BlockOpacity]
  }
  NamedRerouteDeclaration(None=Add(A=@n24.Result, B=Multiply(A=@n24.Secondary Lines, B=@n25.Scanline_Brightness) {constB=3})) {name=Scanlines, nodeColor=<rgba 1 0 0.1641 1>}
  n26 = DynamicParameter() {paramNames [TextureWPO_Speed_Mult, TextureWPO_UVTiling_Mult, AmountWiggle_Random, BendAmount]}
  n27 = ComponentMask(None=TextureCoordinate() {coordinateIndex=1}) {r=true}
  NamedRerouteDeclaration(None=RotateAboutAxis(
    NormalizedRotationAxis=Constant3Vector() {constant=<rgba 0 0 1 0>}
    RotationAngle=Multiply(A=@n26.BendAmount, B=@n27)
    PivotPoint=ParticlePositionWS()
    Position=WorldPosition().XYZ
   )) {name=Bend, nodeColor=<rgba 0.3125 1 0 1>}
  ScalarParameter() {defaultValue=1, parameterName=Metallic}
  ScalarParameter() {defaultValue=1, parameterName=Specular}
  ScalarParameter() {parameterName=Roughness}
  ScalarParameter() {defaultValue=100, parameterName=Scanline amount}
  StaticSwitchParameter(True=Add(A=NamedRerouteUsage(), B=NamedRerouteUsage()), False=Constant()) {defaultValue=true, parameterName=UseWPO?}
  TextureSample(UVs=@n3) {samplerType=SAMPLERTYPE_Normal, texture=/Game/Characters/Player/Female/Medium/Bodies/F_MED_Lady_Atlantis/Textures/T_FishScales_N}
  Transform(None=Constant3Vector()) {transformSourceType=TRANSFORMSOURCE_World, transformType=TRANSFORM_Local}
 ]
}
