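# Additive Niagara sprite material for NeonCity screen rain drops, drawn without depth testing.
# Three offset bokeh texture samples form an RG vector that distorts a fourth sample, which is tinted and scaled by particle color.
# Dynamic parameters drive distortion, rain normal flattening and an emissive multiplier after MF_FX_EmissiveAdjustments.
# The camera-Z lerp reads CameraDown_Mult on both inputs, so view direction has no effect.
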Material /Game/Environments/Asteria/Sets/NeonCity/FX/Materials/M_NeonCity_Rain_ScreenDrop
graph {
 properties {
  bDisableDepthTest=true
  bUsedWithNiagaraSprites=true
  blendMode=BLEND_Additive
  pixelDepthOffsetMode=PDOM_Legacy
  translucencyLightingMode=TLM_VolumetricPerVertexDirectional
  usedShadingModels=Default Lit
 }
 nodes [
  MaterialFunctionCall(
   Normal=TextureSample() {samplerType=SAMPLERTYPE_Normal, texture=/Game/Environments/Asteria/Sets/NeonCity/Techart/Rain/T_RainRound_N}.RGB
   Flatness=DynamicParameter() {defaultValue=<rgba 0.2 1 1 1>, paramNames [Distortion, FlattenNormal, CameraUp_Mult, CameraDown_Mult]}.FlattenNormal
  ) {materialFunction=&FlattenNormal}
  n1 = ParticleColor()
  n2 = VectorParameter() {parameterName=Color Tint, defaultValue=<rgba 0.0469 0.0469 0.0469 0>}
  n3 = TextureCoordinate()
  n4 = Multiply(
   A=Constant3Vector() {constant=<rgba 1 0 0 0>}
   B=TextureSample(UVs=@n3) {samplerType=SAMPLERTYPE_LinearColor, texture=/Game/TimeOfDay/TODM/BR/s11/ExposureAssets/FN_BokehS11}.R
  )
  n5 = Constant2Vector() {g=-0.03, r=-0.03}
  n6 = Multiply(
   A=Constant3Vector() {constant=<rgba 0 1 0 0>}
   B=TextureSample(UVs=Add(A=@n3, B=@n5)) {samplerType=SAMPLERTYPE_LinearColor, texture=&FN_BokehS11}.R
  )
  n7 = Constant3Vector() {constant=<rgba 0 0 1 0>}
  n8 = TextureSample(UVs=Add(A=@n3, B=Constant2Vector() {g=0.03, r=0.03})) {samplerType=SAMPLERTYPE_LinearColor, texture=&FN_BokehS11}
  n9 = ComponentMask(None=Add(A=Add(A=@n4, B=@n6), B=Multiply(A=@n7, B=@n8.R))) {g=true, r=true}
  n10 = DynamicParameter() {defaultValue=<rgba 0.2 1 1 1>, paramNames [Distortion, FlattenNormal, CameraUp_Mult, CameraDown_Mult]}
  n11 = TextureSample(UVs=Add(A=Multiply(A=@n9, B=@n10.Distortion), B=TextureCoordinate())) {samplerType=SAMPLERTYPE_LinearColor, texture=&FN_BokehS11}
  n12 = Constant() {r=1}
  n13 = Multiply(A=Add(A=Multiply(A=@n2.RGB, B=@n11.RGB), B=Multiply(A=@n11.RGB, B=@n12)), B=ScalarParameter() {defaultValue=0.25, parameterName=overall}) {constB=0.5}
  Multiply(A=@n1.A, B=@n13)
  n14 = DynamicParameter() {defaultValue=<rgba 0.2 1 1 1>, paramNames [Distortion, FlattenNormal, CameraUp_Mult, CameraDown_Mult]}
  n15 = ComponentMask(None=MaterialFunctionCall() {materialFunction=&CameraDirectionVector}.Result) {b=true}
  Multiply(
   A=MaterialFunctionCall(Emissive=Multiply(A=@n1.RGB, B=@n13)) {materialFunction=&MF_FX_EmissiveAdjustments}.Result
   B=LinearInterpolate(A=@n14.CameraDown_Mult, B=@n14.CameraDown_Mult, Alpha=Saturate(None=@n15))
  )
 ]
}
