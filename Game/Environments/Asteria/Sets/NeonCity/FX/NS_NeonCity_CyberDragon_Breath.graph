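# Niagara system for the cyber dragon's breath effect in the Asteria NeonCity set.
# GPU flame sprites and eroding cubes spawn from a static flame mesh, alongside mesh-particle flame blobs and a dragon erode layer; ember and glow emitters are disabled.
# System loop age drives spawn windows, shared color gradients and a user timing curve fed to materials.
# Region masks and kill volumes moving along Y progressively clear sprites and cubes.
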NiagaraSystem /Game/Environments/Asteria/Sets/NeonCity/FX/NS_NeonCity_CyberDragon_Breath
dependencies {
 usedDataInterfaces [
  &NiagaraDataInterfaceStaticMesh
  &NiagaraDataInterfaceVectorCurve
  &NiagaraDataInterfaceGBuffer
  &NiagaraDataInterfaceCollisionQuery
  &NiagaraDataInterfaceLandscape
  &NiagaraDataInterfaceMeshRendererInfo
 ]
 usedDynamicInputs [&Multiply_Int, &MakeIntFromFloat]
 usedModules [
  &SystemState
  &self:SystemSpawnScript.SystemScriptSource.SystemScriptGraph.NiagaraNodeAssignment_0.SetVariables_#1
 ]
}
emitters {
 After_Flame {
  simTarget=GPUComputeSim
  emitterUpdateScript [
   &EmitterState()
   &SpawnRate(
    SpawnRate=&FloatFromCurve(
     FloatCurve=DataInterface() {propertyValues {Curve {keys [<at 0.3 0>,<at 0.35 1>,<at 0.5 0>]}}}
     CurveIndex=$System.NormalizedLoopAge
     Scale Curve=100
    )
    Use Spawn Probability=false
   )
  ]
  particleSpawnScript [
   &InitializeParticle(
    Lifetime=&RandomRangeFloat(Minimum=4, Maximum=5)
    Sprite Rotation Angle=270
    Sprite Rotation Mode=::NewEnumerator0
    Sprite Size Mode=::NewEnumerator3
    Uniform Sprite Size=&Multiply_Float(
     A=&RandomRangeFloat(Minimum=800, Maximum=1500)
     B=&FloatFromCurve(FloatCurve=DataInterface() {propertyValues {Curve {keys [<at 0 0.5>,<at 1 1>]}}}, CurveIndex=$System.NormalizedLoopAge)
    )
    UsePositionOffset=false
    Write Material Random=true
   )
   /Niagara/Modules/Spawn/Location/StaticMeshLocation(
    "Inherit Velocity (Scale)"=25
    Inherited Velocity Speed Limit=1
    MaskSpawnGroup=false
    OffsetAlongNormal=true
    OverrideRandomSeed=false
    Position Sampling:ENiagaraSkelMeshPositionSamplingMode=::NewEnumerator1
    SamplePosOffset=true
    Sampled Position Offset:Vector3f=<0 1 0>
    Sampled Velocity Coordinate Space:ENiagaraCoordinateSpace=::Local
    Static Mesh=DataInterface() {propertyValues {
      DefaultMesh=/Game/Environments/Asteria/Sets/NeonCity/FX/Meshes/SM_CyberDragon_Flame_01
      InstanceIndex=-1
      LODIndex=0
      PreviewMesh=&SM_CyberDragon_Flame_01
      SoftSourceActor=null
      SourceMode=DefaultMeshOnly
      bAllowSamplingFromStreamingLODs=false
      bCaptureTransformsPerFrame=true
      bUsePhysicsBodyVelocity=false
      FilteredSockets []
      LODIndexUserParameter {parameter {name=None, typeDefHandle {registeredTypeIndex=108}, varData []}}
      MeshParameterBinding {parameter {name=None, typeDefHandle {registeredTypeIndex=1}, varData []}}
      SectionFilter {allowedMaterialSlots []}
     }}
    Velocity Sampling:ENiagara_AttributeSamplingApplyOutput=::NewEnumerator0
    VelocitySpeedLimit=true
   )
   &Color(Color=&MultiplyLinearColor(Linear Color=$Particles.StaticMeshLocation.SampledVertexColor, Scale Factor=<0 0 10 1>)) {enabled=false}
   /Niagara/Modules/Spawn/MeshInterface/SampleStaticMesh(
    OverrideRandomSeed=false
    StaticMesh=DataInterface() {propertyValues {
      DefaultMesh=&SM_CyberDragon_Flame_01
      InstanceIndex=-1
      LODIndex=0
      PreviewMesh=&SM_CyberDragon_Flame_01
      SoftSourceActor=null
      SourceMode=DefaultMeshOnly
      bAllowSamplingFromStreamingLODs=false
      bCaptureTransformsPerFrame=true
      bUsePhysicsBodyVelocity=false
      FilteredSockets []
      LODIndexUserParameter {parameter {name=None, typeDefHandle {registeredTypeIndex=108}, varData []}}
      MeshParameterBinding {parameter {name=None, typeDefHandle {registeredTypeIndex=1}, varData []}}
      SectionFilter {allowedMaterialSlots [0]}
     }}
   )
   /Niagara/Modules/Spawn/Velocity/StaticMeshVelocity(Add Velocity Along Mesh Normals=500, Module.Enable Inherit Velocity=true, Module.Enable Mesh Normal Sampling=true)
   &ShapeLocation(MaskSpawnGroup=false, Override Seed=false, Sphere Radius=500, UseOwnerScale=true)
   &AddVelocity(Velocity=<0 -1500 0>)
   &DynamicMaterialParameters(
    Index 0 Param 1=&FloatFromCurve(FloatCurve=DataInterface() {propertyValues {Curve {keys [
         {arriveTangent=8.397, interpMode=RCIM_Cubic, leaveTangent=4.6996, tangentMode=RCTM_User, time=-0.0012, value=-0.0261}
         {arriveTangent=1.3472, interpMode=RCIM_Cubic, leaveTangent=1.3472, tangentMode=RCTM_User, time=0.1801, value=0.7926}
         <at 0.5 1 RCIM_Cubic>
        ]}}})
    Index 0 Param 2=&UniformRangedFloat(Minimum=0.333, Maximum=1)
    Index 0 Param 3=&FloatFromCurve(FloatCurve=DataInterface() {propertyValues {Curve {keys [{arriveTangent=1.4224, interpMode=RCIM_Cubic, time=0.5, value=1}, <at 1 2 RCIM_Cubic>]}}})
    Index 0 Param 4=&UniformRangedFloat(Minimum=1.333, Maximum=2)
    Module.Param0WriteEnabled=true
    Module.Param1WriteEnabled=true
    Module.Param2WriteEnabled=true
    Module.Param3WriteEnabled=true
   ) {module=DynamicMaterialParameters001}
  ]
  particleUpdateScript [
   &ParticleState()
   &RegionMask(
    Offset:Vector3f=&MakeVector(
     X=0
     Y=&FloatFromCurve(FloatCurve=DataInterface() {propertyValues {Curve {keys [<at 0 0>,<at 1 1>]}}}, CurveIndex=$System.NormalizedLoopAge, Scale Curve=15000)
     Z=0
    )
    Region Falloff=0.45
    Region Radius=6500
    UseOffset=true
    UseRotation=false
   )
   &CurlNoiseForce(
    Noise Frequency=1
    Noise Strength=&FloatFromCurve(
     FloatCurve=DataInterface() {propertyValues {Curve {keys [{interpMode=RCIM_Cubic, tangentMode=RCTM_User, time=0, value=1}, {interpMode=RCIM_Cubic, tangentMode=RCTM_User, time=0.1, value=0}]}}}
     Scale Curve=500
    )
    Pan Noise=true
    Pan Noise Field=<0 -2 0>
   )
   &GravityForce(Gravity=&VectorFromCurve(
     VectorCurve=DataInterface() {propertyValues {XCurve {keys [<at 0 1>,<at 1 0>]}, YCurve {keys [<at 0 1>,<at 1 0>]}, ZCurve {keys [<at 0 0>,<at 0.25 1>]}}}
     Scale Curve=<1 1 -50>
    ))
   &Drag(
    Drag=&FloatFromCurve(FloatCurve=DataInterface() {propertyValues {Curve {keys [
         {interpMode=RCIM_Cubic, tangentMode=RCTM_User, time=0, value=0}
         {interpMode=RCIM_Cubic, tangentMode=RCTM_User, time=0.5, value=1}
         {tangentMode=RCTM_User, time=1, value=0}
        ]}}})
    Use Linear Drag=true
    Use Rotational Drag=true
   )
   &SubUVAnimation(
    Start Frame:NiagaraInt32=&RandomRangeInt(Minimum=0, Maximum=16)
    SubUV Animation Mode:ENiagara_SubUVLookupMode=::NewEnumerator1
    SubUV Curve=DataInterface() {propertyValues {Curve {keys [{arriveTangent=1.6038, time=0, value=0}, {leaveTangent=0.6912, time=1, value=1}]}}}
    SubUV Loop Count:NiagaraFloat=&RandomRangeFloat(Minimum=1, Maximum=1)
   )
   &SolveForcesAndVelocity(Clamp Velocity=false, Limit Acceleration=false)
   &DynamicMaterialParameters(
    Index 0 Param 1=&FloatFromCurve(FloatCurve=DataInterface() {propertyValues {Curve {keys [<at 0 1>,<at 1 0>]}}})
    Index 0 Param 2=&MakeFloatFromInt(Int=/Niagara/DynamicInputs/Divide/Divide_Int(A:NiagaraInt32=&RandomRangeInt(Minimum=0, Maximum=2), B:NiagaraInt32=1))
    Index 0 Param 3=8
    Index 0 Param 4=$Particles.SubImageIndex
    Index 1 Param 1=&RandomRangeFloat(Minimum=0, Maximum=1)
    Index 1 Param 2=&RandomRangeFloat(Minimum=-1, Maximum=1)
    Index 1 Param 3=&FloatFromCurve(FloatCurve=DataInterface() {propertyValues {Curve {keys [
         {arriveTangent=4.1785, interpMode=RCIM_Cubic, leaveTangent=4.1785, tangentMode=RCTM_User, time=0, value=0}
         {arriveTangent=0.9995, interpMode=RCIM_Cubic, leaveTangent=0.9995, tangentMode=RCTM_User, time=0.2223, value=0.6825}
         {arriveTangent=-0.0405, interpMode=RCIM_Cubic, leaveTangent=-0.0405, tangentMode=RCTM_User, time=0.7, value=0.9}
         <at 1 0.7>
        ]}}})
    Index 2 Param 1=100
    Index 2 Param 2=0.004
    Index 2 Param 3=150
    Index 2 Param 4=2
    Index 3 Param 1=100
    Index 3 Param 2=0.02
    Index 3 Param 3=1
    Index 3 Param 4=&FloatFromCurve(FloatCurve=$User.MainTimingCurve, CurveIndex=$System.NormalizedLoopAge)
    Module.Param0WriteEnabled=true
    Module.Param0WriteEnabled1=true
    Module.Param0WriteEnabled2=true
    Module.Param0WriteEnabled3=true
    Module.Param1WriteEnabled=true
    Module.Param1WriteEnabled1=true
    Module.Param1WriteEnabled2=true
    Module.Param1WriteEnabled3=true
    Module.Param2WriteEnabled=true
    Module.Param2WriteEnabled1=true
    Module.Param2WriteEnabled2=true
    Module.Param2WriteEnabled3=true
    Module.Param3WriteEnabled=true
    Module.Param3WriteEnabled1=false
    Module.Param3WriteEnabled2=true
    Module.Param3WriteEnabled3=true
    Write Parameter Index 1=true
    Write Parameter Index 2=true
    Write Parameter Index 3=true
   ) {module=DynamicMaterialParameters002}
   &ScaleColor(
    Scale Alpha=&FloatFromCurve(
     FloatCurve=DataInterface() {propertyValues {Curve {keys [
         {arriveTangent=0.0516, interpMode=RCIM_Cubic, leaveTangent=0.0516, tangentMode=RCTM_User, time=0, value=0}
         {arriveTangent=-1.4181, interpMode=RCIM_Cubic, leaveTangent=-1.4181, tangentMode=RCTM_User, time=0.3, value=1.125}
         <at 1 0>
        ]}}}
     Scale Curve=&Lerp_Float(A=2, B=0, Alpha=$Particles.RegionMask.Float)
    )
    Scale RGB=&Multiply_Vector(
     A=&MakeVectorFromLinearColorRGB(Color=&ColorFromCurve(
       ColorCurve=DataInterface() {propertyValues {
         AlphaCurve {keys [<at 0 1 RCIM_Cubic>,<at 1 1 RCIM_Cubic>]}
         BlueCurve {keys [<at 0.0631 0.3625 RCIM_Cubic>, {arriveTangent=-0.7387, interpMode=RCIM_Cubic, leaveTangent=-0.7387, time=0.3006, value=1}, <at 0.5538 0>]}
         GreenCurve {keys [<at 0.0631 0 RCIM_Cubic>, {arriveTangent=0.0274, interpMode=RCIM_Cubic, leaveTangent=0.0274, time=0.3006, value=0}, <at 0.5538 0.0134>]}
         RedCurve {keys [<at 0.0631 0.5156 RCIM_Cubic>, {arriveTangent=0.987, interpMode=RCIM_Cubic, leaveTangent=0.987, time=0.3006, value=0.1928}, <at 0.5538 1>]}
        }}
       CurveIndex=$System.NormalizedLoopAge
      ))
     B=&VectorFromFloat(Value=&FloatFromCurve(
       FloatCurve=DataInterface() {propertyValues {Curve {keys [
           {arriveTangent=9.0725, interpMode=RCIM_Cubic, leaveTangent=9.0725, tangentMode=RCTM_User, time=0, value=0}
           {arriveTangent=-2.5249, interpMode=RCIM_Cubic, leaveTangent=-2.5249, tangentMode=RCTM_User, time=0.1408, value=10}
           {arriveTangent=-1.5305, interpMode=RCIM_Cubic, leaveTangent=-1.5305, tangentMode=RCTM_User, time=0.5, value=0.35}
           {arriveTangent=-0.0405, interpMode=RCIM_Cubic, leaveTangent=-0.0405, tangentMode=RCTM_User, time=1, value=0}
          ]}}}
       Scale Curve=5
      ))
    )
    ScaleA=true
    ScaleRGB=true
   )
   /SetVariables(Particles.SpriteSize2=&Multiply_Vector2D(A=$Particles.SpriteSize, B=<1 0.35>)) {isSetParameters=true, module=SetVariables_#2}
   /SetVariables(Particles.SpriteRotation2:NiagaraFloat=&RandomRangeFloat(Minimum=0, Maximum=180)) {isSetParameters=true, module=SetVariables_#3}
   &ScaleSpriteSizeBySpeed(
    Max Scale Factor=<2 3>
    Min Scale Factor=<2 2>
    Sample Scale Factor By Curve=true
    Scale Factor Curve=DataInterface() {propertyValues {Curve {keys [{interpMode=RCIM_Cubic, tangentMode=RCTM_User, time=0.1, value=0}, {interpMode=RCIM_Cubic, tangentMode=RCTM_User, time=1, value=1}]}}}
   )
   &ScaleSpriteSize(Uniform Curve Sprite Scale=DataInterface() {propertyValues {Curve {keys [
        {arriveTangent=0.3375, interpMode=RCIM_Cubic, leaveTangent=0.3375, tangentMode=RCTM_User, time=0, value=1}
        {interpMode=RCIM_Cubic, tangentMode=RCTM_User, time=1, value=2}
       ]}}})
   &KillParticlesInVolume(
    Origin Offset:Vector3f=&MakeVector(
     X=0
     Y=&FloatFromCurve(FloatCurve=DataInterface() {propertyValues {Curve {keys [<at 0 0>,<at 0.5 1>]}}}, CurveIndex=$System.NormalizedLoopAge, Scale Curve=15000)
     Z=0
    )
    Sphere Radius=6500
   )
   &CameraOffset(Camera Offset Amount=&FloatFromCurve(FloatCurve=DataInterface() {propertyValues {Curve {keys [<at 0 -1>,<at 0.25 1>]}}}, Scale Curve=2500))
  ]
  rendererClasses [&NiagaraSpriteRendererProperties]
  renderers [{rendererClass=&NiagaraSpriteRendererProperties}]
 }
 BlobFlameMesh {
  emitterUpdateScript [
   &EmitterState(Life Cycle Mode=::NewEnumerator1, Loop Behavior=::NewEnumerator1, Loop Duration=1, UseLoopDelay=false)
   &SpawnBurst_Instantaneous(Spawn Count=2, Use Spawn Probability=false, UseLoopCountLimit=false)
   &SpawnRate(
    SpawnRate=&FloatFromCurve(FloatCurve=DataInterface() {propertyValues {Curve {keys [<at 0 1>,<at 0.5 0>]}}}, CurveIndex=$Emitter.NormalizedLoopAge)
    Use Spawn Probability=false
   ) {enabled=false}
  ]
  particleSpawnScript [
   &InitializeParticle(
    Lifetime Max=4.5
    Lifetime Min=4.5
    Lifetime Mode=::NewEnumerator1
    Mesh Renderer Array Visibility Mode=::NewEnumerator3
    Mesh Renderer Info=DataInterface() {propertyValues {MeshRenderer=&self:BlobFlameMesh_5.NiagaraMeshRendererProperties_0}}
    Mesh Scale Mode=::NewEnumerator1
    Mesh Uniform Scale Max=1
    Mesh Uniform Scale Min=0.6
    UsePositionOffset=false
    Write Material Random=true
   )
   &AddVelocity(
    Cone Angle=30
    Cone Axis=<0 0 1>
    Module.UseConeFalloff=true
    Module.UseVelDistribution=true
    Override Seed=false
    Speed Falloff From Cone Axis=0.875
    Velocity Mode=::NewEnumerator2
    Velocity Speed=0
   )
   &InitialMeshOrientation(
    Mesh Orientation Mode=::NewEnumerator1
    Rotation=&RandomRangeVector(Minimum=<0 0.1 -0.25>, Maximum=<0 0.1 -0.25>)
    Rotation Coordinate Space=::NewEnumerator2
    Use Rotation Vector=true
   )
   &SystemLocation(MaskSpawnGroup=false, Offset=<0 4700 150>)
  ]
  particleUpdateScript [
   &ParticleState()
   &ScaleColor(
    Scale Alpha=&FloatFromCurve(FloatCurve=DataInterface() {propertyValues {Curve {keys [<at 0.5 1 RCIM_Cubic>,<at 1 1 RCIM_Cubic>]}}})
    Scale RGB=&MakeVectorFromLinearColorRGB(Color=&ColorFromCurve(
      ColorCurve=DataInterface() {propertyValues {
        AlphaCurve {keys [<at 0 1 RCIM_Cubic>,<at 1 1 RCIM_Cubic>]}
        BlueCurve {keys [<at 0.0631 0.3625 RCIM_Cubic>, {interpMode=RCIM_Cubic, leaveTangent=-0.7387, time=0.3006, value=1}, <at 0.4111 0>]}
        GreenCurve {keys [<at 0.0631 0 RCIM_Cubic>, {interpMode=RCIM_Cubic, leaveTangent=0.0274, time=0.3006, value=0}, <at 0.4111 0.0134>]}
        RedCurve {keys [<at 0.0631 0.5156 RCIM_Cubic>, {interpMode=RCIM_Cubic, leaveTangent=0.987, time=0.3006, value=0.1928}, <at 0.4111 1>]}
       }}
      CurveIndex=$System.NormalizedLoopAge
     ))
    ScaleA=true
    ScaleRGB=true
   )
   &ScaleMeshSize(Scale Factor=&VectorFromFloat(Value=&FloatFromCurve(
      FloatCurve=DataInterface() {propertyValues {Curve {keys [
          {arriveTangent=2.6728, interpMode=RCIM_Cubic, leaveTangent=2.6728, tangentMode=RCTM_User, time=0.06, value=0}
          {interpMode=RCIM_Cubic, tangentMode=RCTM_User, time=1, value=1}
         ]}}}
      CurveIndex=&Multiply_Float(A=$Particles.NormalizedAge, B=1)
      Scale Curve=3
     )))
   &GravityForce(Gravity=<0 0 0>)
   &SolveForcesAndVelocity(Clamp Velocity=false, Limit Acceleration=false)
   &DynamicMaterialParameters(
    Index 0 Param 1=&FloatFromCurve(FloatCurve=DataInterface() {propertyValues {Curve {keys [
         {arriveTangent=9.0725, interpMode=RCIM_Cubic, leaveTangent=9.0725, tangentMode=RCTM_User, time=0, value=0}
         {arriveTangent=-6.1533, interpMode=RCIM_Cubic, leaveTangent=-6.1533, tangentMode=RCTM_User, time=0.1408, value=0.9847}
        ]}}})
    Index 0 Param 2=&FloatFromCurve(FloatCurve=DataInterface() {propertyValues {Curve {keys [<at 0 0.5>,<at 1 1>]}}}, Scale Curve=0.25)
    Index 0 Param 3=&FloatFromCurve(FloatCurve=DataInterface() {propertyValues {Curve {keys [<at 0 2>,<at 0.5 0.5>]}}})
    Index 0 Param 4=&FloatFromCurve(
     FloatCurve=DataInterface() {propertyValues {Curve {keys [
         {interpMode=RCIM_Cubic, tangentMode=RCTM_User, time=0.5, value=0}
         <at 0.8 0.65>
         {arriveTangent=1.858, interpMode=RCIM_Cubic, leaveTangent=1.858, tangentMode=RCTM_User, time=1, value=2}
        ]}}}
     Scale Curve=2
    )
    Index 1 Param 1=&FloatFromCurve(FloatCurve=DataInterface() {propertyValues {Curve {keys [<at 0 0.1>,<at 0.3 1.1>,<at 1 1>]}}}, Scale Curve=350)
    Index 2 Param 1=&FloatFromCurve(FloatCurve=DataInterface() {propertyValues {Curve {keys [<at 0 1>,<at 1 0>]}}}, Scale Curve=250)
    Index 2 Param 2=0.01
    Index 2 Param 3=1
    Index 2 Param 4=&FloatFromCurve(FloatCurve=$User.MainTimingCurve, CurveIndex=$System.NormalizedLoopAge)
    Module.Param0WriteEnabled=true
    Module.Param0WriteEnabled1=true
    Module.Param0WriteEnabled2=true
    Module.Param1WriteEnabled=true
    Module.Param1WriteEnabled1=true
    Module.Param1WriteEnabled2=true
    Module.Param2WriteEnabled=true
    Module.Param2WriteEnabled1=true
    Module.Param2WriteEnabled2=true
    Module.Param3WriteEnabled=true
    Module.Param3WriteEnabled1=true
    Module.Param3WriteEnabled2=true
    Write Parameter Index 1=true
    Write Parameter Index 2=true
   )
  ]
  rendererClasses [&NiagaraMeshRendererProperties]
  renderers [{rendererClass=&NiagaraMeshRendererProperties, rendererIndex=0}]
 }
 Dragon_Embers {
  bEnabled=false
  simTarget=GPUComputeSim
  emitterUpdateScript [
   &EmitterState()
   &SpawnRate(
    SpawnRate=&Multiply_Float(
     A=1500
     B=&FloatFromCurve(FloatCurve=DataInterface() {propertyValues {Curve {keys [<at 0.1 0>,<at 0.15 1>,<at 0.5 0>]}}}, CurveIndex=$System.NormalizedLoopAge)
    )
    Use Spawn Probability=false
   ) {module=SpawnRate001}
  ]
  particleSpawnScript [
   /Niagara/Modules/Spawn/Initialization/InitializeParticle(
    Color=&UniformRangedLinearColor(Minimum=<rgba 1 0 0.5038 1>, Maximum=<rgba 1 0.0788 0 1>)
    Lifetime=&UniformRangedFloat(Minimum=1, Maximum=5)
    Module.Write Color=true
    Module.Write Lifetime=true
    Module.Write Mass=true
    Module.Write Material Random=true
    Module.Write Position=false
    Module.Write Scale=false
    Module.Write SpriteRotation=false
    Module.Write SpriteSize=true
    Position=&SimulationPosition()
    Sprite Size=&UniformRangedVector2D(Minimum=<25 25>, Maximum=<50 50>)
   ) {module=InitializeParticle001}
   &StaticMeshLocation(
    MaskSpawnGroup=false
    Offset Position Along Sampled Normal=0
    OffsetAlongNormal=true
    OverrideRandomSeed=false
    RGBA Color Application Mask:Vector4f=<0 0 0 0>
    SamplePosOffset=true
    Sampled Position Offset:Vector3f=<0 0 350>
    Sampled Position Offset Coordinate Space:ENiagaraSkelMeshTransforms=::NewEnumerator2
    Static Mesh=DataInterface() {propertyValues {
      DefaultMesh=/Game/Environments/Asteria/Sets/NeonCity/FX/Meshes/SM_CyberDragon_WholeMesh_01
      InstanceIndex=-1
      LODIndex=0
      PreviewMesh=&SM_CyberDragon_WholeMesh_01
      SoftSourceActor=null
      SourceMode=DefaultMeshOnly
      bAllowSamplingFromStreamingLODs=false
      bCaptureTransformsPerFrame=true
      bUsePhysicsBodyVelocity=false
      FilteredSockets []
      LODIndexUserParameter {parameter {name=None, typeDefHandle {registeredTypeIndex=108}, varData []}}
      MeshParameterBinding {parameter {name=None, typeDefHandle {registeredTypeIndex=1}, varData []}}
      SectionFilter {allowedMaterialSlots []}
     }}
    Vertex Color Sampling:ENiagara_AttributeSamplingApplyOutput=::NewEnumerator0
   )
   &AddVelocity(
    Cone Angle=25
    Cone Axis=<0 -1 0>
    Inner Cone Angle=5
    Module.UseConeFalloff=false
    Module.UseVelDistribution=false
    Override Seed=false
    Velocity Mode=::NewEnumerator2
    Velocity Speed=1
   ) {module=AddVelocity001}
   &SampleStaticMesh(
    OverrideRandomSeed=false
    StaticMesh=DataInterface() {propertyValues {
      DefaultMesh=&SM_CyberDragon_WholeMesh_01
      InstanceIndex=-1
      LODIndex=0
      PreviewMesh=&SM_CyberDragon_WholeMesh_01
      SoftSourceActor=null
      SourceMode=DefaultMeshOnly
      bAllowSamplingFromStreamingLODs=false
      bCaptureTransformsPerFrame=true
      bUsePhysicsBodyVelocity=false
      FilteredSockets []
      LODIndexUserParameter {parameter {name=None, typeDefHandle {registeredTypeIndex=108}, varData []}}
      MeshParameterBinding {parameter {name=None, typeDefHandle {registeredTypeIndex=1}, varData []}}
      SectionFilter {allowedMaterialSlots [0]}
     }}
   )
   &AddVelocity(
    Cone Axis=<0 1 1>
    Module.UseConeFalloff=false
    Module.UseVelDistribution=false
    Override Seed=false
    Velocity Mode=::NewEnumerator2
    Velocity Speed=2500
   ) {module=AddVelocity002}
   &StaticMeshVelocity(Add Velocity Along Mesh Normals=250, Module.Enable Inherit Velocity=true, Module.Enable Mesh Normal Sampling=true)
   &AddVelocity(
    Velocity=&Multiply_VectorByFloat(
     Vector=<1 -500 1>
     Float=&DistanceBetweenPositions(A=&AddVectorToPosition(Position=$Engine.Owner.Position, Vector=<0 0 0>), B=$Particles.Position)
    )
    Velocity Speed Scale=0.001
   ) {enabled=false}
   &SystemLocation(MaskSpawnGroup=false, Offset=<0 6040 500>) {enabled=false}
   &CameraOffset(Camera Offset Amount=&UniformRangedFloat(Minimum=25, Maximum=25))
  ]
  particleUpdateScript [
   &ParticleState()
   &CurlNoiseForce(
    Noise Frequency=2
    Noise Strength=&FloatFromCurve(
     FloatCurve=DataInterface() {propertyValues {Curve {keys [{interpMode=RCIM_Cubic, tangentMode=RCTM_User, time=0, value=1}, {interpMode=RCIM_Cubic, tangentMode=RCTM_User, time=0.1, value=0}]}}}
     Scale Curve=250
    )
    Pan Noise=true
    Pan Noise Field=<0 -1 0>
   )
   &GravityForce(Gravity=<0 0 -500>)
   &Drag(
    Drag=&FloatFromCurve(FloatCurve=DataInterface() {propertyValues {Curve {keys [<at 0 1>,<at 1 0>]}}})
    Use Linear Drag=true
    Use Rotational Drag=true
   )
   &SolveForcesAndVelocity(Clamp Velocity=false, Limit Acceleration=false)
   &ScaleSpriteSize(Scale Factor=&Vector2DFromFloat(Value=&FloatFromCurve(FloatCurve=DataInterface() {propertyValues {Curve {keys [
          {arriveTangent=56.1861, interpMode=RCIM_Cubic, leaveTangent=56.1859, tangentMode=RCTM_User, time=0, value=0}
          <at 0.03 1 RCIM_Cubic>
          <at 1 0 RCIM_Cubic>
         ]}}})))
   &ScaleSpriteSizeBySpeed(
    Max Scale Factor=<3 3>
    Sample Scale Factor By Curve=true
    Scale Factor Curve=DataInterface() {propertyValues {Curve {keys [{interpMode=RCIM_Cubic, tangentMode=RCTM_User, time=0.5, value=0}, {interpMode=RCIM_Cubic, tangentMode=RCTM_User, time=1, value=1}]}}}
    Velocity Threshold=1800
   )
   &ScaleColor(
    Scale Alpha=&FloatFromCurve(FloatCurve=DataInterface() {propertyValues {Curve {keys [
         {arriveTangent=-2.5249, interpMode=RCIM_Cubic, leaveTangent=-2.5249, tangentMode=RCTM_User, time=0, value=0.975}
         {arriveTangent=-0.0405, interpMode=RCIM_Cubic, leaveTangent=-0.0405, tangentMode=RCTM_User, time=0.25, value=0.1}
         <at 1 0>
        ]}}})
    Scale RGB=&Multiply_VectorByFloat(
     Vector=&MakeVectorFromLinearColorRGB(Color=&ColorFromCurve(ColorCurve=DataInterface() {propertyValues {
         AlphaCurve {keys [<at 0 1 RCIM_Cubic>,<at 1 1 RCIM_Cubic>]}
         BlueCurve {keys [<at 0 1 RCIM_Cubic>,<at 1 0.8118 RCIM_Cubic>]}
         GreenCurve {keys [<at 0 0.0527 RCIM_Cubic>,<at 1 0 RCIM_Cubic>]}
         RedCurve {keys [<at 0 0 RCIM_Cubic>,<at 1 1 RCIM_Cubic>]}
        }}))
     Float=&FloatFromCurve(
      FloatCurve=DataInterface() {propertyValues {Curve {keys [
          {interpMode=RCIM_Cubic, tangentMode=RCTM_User, time=0, value=0}
          {interpMode=RCIM_Cubic, tangentMode=RCTM_User, time=0.5, value=1}
          {tangentMode=RCTM_User, time=1, value=0}
         ]}}}
      Scale Curve=250
     )
    )
    ScaleA=true
    ScaleRGB=true
   )
  ]
  rendererClasses [&NiagaraSpriteRendererProperties]
  renderers [{rendererClass=&NiagaraSpriteRendererProperties}]
 }
 Dragon_Erode {
  emitterUpdateScript [
   &EmitterState()
   &SpawnBurst_Instantaneous(Spawn Time=0.15, Use Spawn Probability=false, UseLoopCountLimit=false)
  ]
  particleSpawnScript [
   /Niagara/Modules/Spawn/Initialization/V2/InitializeParticle(
    Lifetime=3
    Mesh Scale Mode=::NewEnumerator3
    Sprite Size Mode=::NewEnumerator3
    Uniform Sprite Size=50
    UsePositionOffset=false
    Write Material Random=true
   )
   &InitialMeshOrientation(Mesh Orientation Mode=::NewEnumerator1, Rotation=<0 0 0.5>, Use Rotation Vector=true)
   &SystemLocation(MaskSpawnGroup=false, Offset=<0 6000 -500>)
   &AddVelocity(Velocity=<0 50 25>)
  ]
  particleUpdateScript [
   &ParticleState()
   &ScaleMeshSize(Scale Factor=&VectorFromCurve(
     VectorCurve=DataInterface() {propertyValues {
       XCurve {keys [{interpMode=RCIM_Cubic, tangentMode=RCTM_User, time=0, value=0.7}, <at 1 1>]}
       YCurve {keys [<at 0 0.7>,<at 1 1>]}
       ZCurve {keys [<at 0 0.33>,<at 1 1>]}
      }}
     Scale Curve=<1 1 2>
    ))
   &ScaleColor(
    Scale Alpha=&FloatFromCurve(
     FloatCurve=DataInterface() {propertyValues {Curve {keys [{tangentMode=RCTM_User, time=0, value=0.5}, {tangentMode=RCTM_User, time=1, value=0.45}]}}}
     Scale Curve=2.02
    )
    Scale RGB=&Multiply_Vector(
     A=&MakeVectorFromLinearColorRGB(Color=&ColorFromCurve(
       ColorCurve=DataInterface() {propertyValues {
         AlphaCurve {keys [<at 0 1 RCIM_Cubic>,<at 1 1 RCIM_Cubic>]}
         BlueCurve {keys [<at 0.0631 0.3625 RCIM_Cubic>, {arriveTangent=-0.7387, interpMode=RCIM_Cubic, leaveTangent=-0.7387, time=0.3006, value=1}, <at 0.5538 0>]}
         GreenCurve {keys [<at 0.0631 0 RCIM_Cubic>, {arriveTangent=0.0274, interpMode=RCIM_Cubic, leaveTangent=0.0274, time=0.3006, value=0}, <at 0.5538 0.0134>]}
         RedCurve {keys [<at 0.0631 0.5156 RCIM_Cubic>, {arriveTangent=0.987, interpMode=RCIM_Cubic, leaveTangent=0.987, time=0.3006, value=0.1928}, <at 0.5538 1>]}
        }}
       CurveIndex=$System.NormalizedLoopAge
      ))
     B=&VectorFromFloat(Value=&FloatFromCurve(FloatCurve=DataInterface() {propertyValues {Curve {keys [
           {arriveTangent=9.0725, interpMode=RCIM_Cubic, leaveTangent=9.0725, tangentMode=RCTM_User, time=0, value=0}
           {arriveTangent=-2.5249, interpMode=RCIM_Cubic, leaveTangent=-2.5249, tangentMode=RCTM_User, time=0.1408, value=10}
           {arriveTangent=-0.0405, interpMode=RCIM_Cubic, leaveTangent=-0.0405, tangentMode=RCTM_User, time=1, value=0}
          ]}}}))
    )
    ScaleA=true
    ScaleRGB=true
   )
   &SolveForcesAndVelocity(Clamp Velocity=false, Limit Acceleration=false)
   &DynamicMaterialParameters(
    Index 0 Param 1=&FloatFromCurve(
     FloatCurve=DataInterface() {propertyValues {Curve {keys [
         {arriveTangent=20.7302, interpMode=RCIM_Cubic, leaveTangent=20.7302, tangentMode=RCTM_User, time=0, value=0}
         {arriveTangent=-5.16, interpMode=RCIM_Cubic, leaveTangent=-5.16, tangentMode=RCTM_User, time=0.02, value=2}
         {arriveTangent=-1.8298, interpMode=RCIM_Cubic, leaveTangent=-1.8298, tangentMode=RCTM_User, time=0.25, value=1}
         {arriveTangent=-0.5381, interpMode=RCIM_Cubic, leaveTangent=-0.5381, tangentMode=RCTM_User, time=1, value=0}
        ]}}}
     Scale Curve=50
    )
    Index 0 Param 2=&FloatFromCurve(FloatCurve=DataInterface() {propertyValues {Curve {keys [<at 0.1 0>,<at 1 2>]}}})
    Index 0 Param 3=&FloatFromCurve(FloatCurve=DataInterface() {propertyValues {Curve {keys [<at 0.1 1>,<at 1 0>]}}}, Scale Curve=0)
    Index 0 Param 4=1
    Index 1 Param 3=&FloatFromCurve(
     FloatCurve=DataInterface() {propertyValues {Curve {keys [{tangentMode=RCTM_User, time=0, value=0}, {tangentMode=RCTM_User, time=1, value=0.5}]}}}
     Scale Curve=5
    )
    Module.Param0WriteEnabled=true
    Module.Param0WriteEnabled1=true
    Module.Param1WriteEnabled=true
    Module.Param1WriteEnabled1=true
    Module.Param2WriteEnabled=true
    Module.Param2WriteEnabled1=true
    Module.Param3WriteEnabled=true
    Module.Param3WriteEnabled1=true
    Write Parameter Index 1=true
   )
   &CameraOffset(Camera Offset Amount=&FloatFromCurve(
     FloatCurve=DataInterface() {propertyValues {Curve {keys [
         {interpMode=RCIM_Cubic, tangentMode=RCTM_User, time=0, value=0.6}
         {interpMode=RCIM_Cubic, tangentMode=RCTM_User, time=0.5, value=1.1}
         {tangentMode=RCTM_User, time=1, value=1}
        ]}}}
     Scale Curve=500
    )) {module=CameraOffset001}
  ]
  rendererClasses [&NiagaraMeshRendererProperties]
  renderers [{rendererClass=&NiagaraMeshRendererProperties, rendererIndex=0}]
 }
 Dragon_Glow {
  bEnabled=false
  emitterUpdateScript [
   &EmitterState()
   &SpawnBurst_Instantaneous(Use Spawn Probability=false, UseLoopCountLimit=false)
  ]
  particleSpawnScript [
   &InitializeParticle(
    Lifetime=3
    Mesh Scale Mode=::NewEnumerator3
    Sprite Size Mode=::NewEnumerator3
    Uniform Sprite Size=50
    UsePositionOffset=false
    Write Material Random=true
   )
   &InitialMeshOrientation(Mesh Orientation Mode=::NewEnumerator1, Rotation=<0 0 0.5>, Use Rotation Vector=true)
   &SystemLocation(MaskSpawnGroup=false, Offset=<0 5750 -500>)
  ]
  particleUpdateScript [
   &ParticleState()
   &ScaleMeshSize(Scale Factor=&VectorFromCurve(VectorCurve=DataInterface() {propertyValues {
       XCurve {keys [
         {interpMode=RCIM_Cubic, tangentMode=RCTM_User, time=0, value=0.25}
         {arriveTangent=3.0837, interpMode=RCIM_Cubic, leaveTangent=3.0837, tangentMode=RCTM_User, time=0.25, value=1}
        ]}
       YCurve {keys [<at 0 0.55>,<at 0.25 0.95>]}
       ZCurve {keys [<at 0 0.55>,<at 1 1>]}
      }})) {enabled=false}
   &ScaleColor(
    Scale Alpha=&FloatFromCurve(
     FloatCurve=DataInterface() {propertyValues {Curve {keys [{arriveTangent=-3.6361, interpMode=RCIM_Cubic, leaveTangent=-3.6361, tangentMode=RCTM_User, time=0, value=1}, <at 1 0 RCIM_Cubic>]}}}
     Scale Curve=5
    )
    Scale RGB=&MakeVectorFromLinearColorRGB(Color=&MultiplyLinearColor(
      Linear Color=&LerpLinearColors(
       StartColor=<rgba 1 0 0.2073 1>
       EndColor=<rgba 0.1423 0 0.4678 1>
       LerpFactor=&FloatFromCurve(FloatCurve=DataInterface() {propertyValues {Curve {keys [<at 0 1>,<at 1 0>]}}})
      )
      Scale Factor=&MakeVector4_FromFloat(Float=&FloatFromCurve(
        FloatCurve=DataInterface() {propertyValues {Curve {keys [
            {interpMode=RCIM_Cubic, tangentMode=RCTM_User, time=0, value=0}
            {arriveTangent=-3.1378, interpMode=RCIM_Cubic, leaveTangent=-3.1378, tangentMode=RCTM_User, time=0.1, value=0.95}
            {tangentMode=RCTM_User, time=1, value=0}
           ]}}}
        Scale Curve=100
       ))
     ))
    ScaleA=true
    ScaleRGB=true
   )
   &SolveForcesAndVelocity(Clamp Velocity=false, Limit Acceleration=false)
   &DynamicMaterialParameters(
    Index 0 Param 1=&FloatFromCurve(
     FloatCurve=DataInterface() {propertyValues {Curve {keys [
         {arriveTangent=20.7302, interpMode=RCIM_Cubic, leaveTangent=20.7302, tangentMode=RCTM_User, time=0, value=0}
         {arriveTangent=-5.16, interpMode=RCIM_Cubic, leaveTangent=-5.16, tangentMode=RCTM_User, time=0.02, value=2}
         {arriveTangent=-1.8298, interpMode=RCIM_Cubic, leaveTangent=-1.8298, tangentMode=RCTM_User, time=0.25, value=1}
         {arriveTangent=-0.5381, interpMode=RCIM_Cubic, leaveTangent=-0.5381, tangentMode=RCTM_User, time=1, value=0}
        ]}}}
     Scale Curve=25
    )
    Index 0 Param 2=&FloatFromCurve(FloatCurve=DataInterface() {propertyValues {Curve {keys [<at 0.1 0>,<at 1 2>]}}})
    Index 0 Param 3=&FloatFromCurve(FloatCurve=DataInterface() {propertyValues {Curve {keys [<at 0.1 1>,<at 1 0>]}}}, Scale Curve=0)
    Module.Param0WriteEnabled=true
    Module.Param1WriteEnabled=true
    Module.Param2WriteEnabled=true
    Module.Param3WriteEnabled=true
   )
   &CameraOffset(Camera Offset Amount=&FloatFromCurve(
     FloatCurve=DataInterface() {propertyValues {Curve {keys [
         {interpMode=RCIM_Cubic, tangentMode=RCTM_User, time=0, value=0}
         {interpMode=RCIM_Cubic, tangentMode=RCTM_User, time=0.1, value=1.1}
         {tangentMode=RCTM_User, time=1, value=0}
        ]}}}
     Scale Curve=250
    )) {module=CameraOffset001}
  ]
  rendererClasses [&NiagaraMeshRendererProperties]
  renderers [{rendererClass=&NiagaraMeshRendererProperties, rendererIndex=0}]
 }
 ErodeCubes {
  simTarget=GPUComputeSim
  emitterUpdateScript [
   &EmitterState()
   &SpawnRate(
    SpawnRate=&FloatFromCurve(
     FloatCurve=DataInterface() {propertyValues {Curve {keys [<at 0.15 0>,<at 0.2 1>,<at 0.6 0>]}}}
     CurveIndex=$System.NormalizedLoopAge
     Scale Curve=10000
    )
    Use Spawn Probability=false
   ) {module=SpawnRate003}
  ]
  particleSpawnScript [
   /Niagara/Modules/Spawn/Initialization/InitializeParticle(
    Color=<rgba 0.6038 0.6038 0.6038 1>
    Lifetime=&RandomRangeFloat(Minimum=2, Maximum=3)
    Mass=&UniformRangedFloat(Minimum=1, Maximum=2)
    Mesh Scale=&RandomRangeVector(
     Minimum=&VectorFromFloat(Value=25)
     Maximum=&VectorFromFloat(Value=250)
    )
    Module.Write Color=true
    Module.Write Lifetime=true
    Module.Write Mass=true
    Module.Write Material Random=true
    Module.Write Position=false
    Module.Write Scale=true
    Module.Write SpriteRotation=false
    Module.Write SpriteSize=true
    Position=&SimulationPosition()
    Sprite Size=<50 50>
   )
   &AddVelocity(Velocity=<0 -5000 50>) {module=AddVelocity001}
   &StaticMeshLocation(
    MaskSpawnGroup=false
    Offset Position Along Sampled Normal=25
    OffsetAlongNormal=true
    Orientation Sampling:ENiagara_AttributeSamplingApplyOutput=::NewEnumerator1
    OverrideRandomSeed=false
    Position Sampling:ENiagaraSkelMeshPositionSamplingMode=::NewEnumerator1
    RGBA Color Application Mask:Vector4f=<0 0 0 0>
    SamplePosOffset=true
    Sampled Position Offset:Vector3f=<0 0 500>
    Sampled Position Offset Coordinate Space:ENiagaraSkelMeshTransforms=::NewEnumerator1
    Static Mesh=DataInterface() {propertyValues {
      DefaultMesh=&SM_CyberDragon_Flame_01
      InstanceIndex=-1
      LODIndex=0
      PreviewMesh=&SM_CyberDragon_Flame_01
      SoftSourceActor=null
      SourceMode=DefaultMeshOnly
      bAllowSamplingFromStreamingLODs=false
      bCaptureTransformsPerFrame=true
      bUsePhysicsBodyVelocity=false
      FilteredSockets []
      LODIndexUserParameter {parameter {name=None, typeDefHandle {registeredTypeIndex=108}, varData []}}
      MeshParameterBinding {parameter {name=None, typeDefHandle {registeredTypeIndex=1}, varData []}}
      SectionFilter {allowedMaterialSlots []}
     }}
    Vertex Color Sampling:ENiagara_AttributeSamplingApplyOutput=::NewEnumerator0
   )
   &SampleStaticMesh(
    OverrideRandomSeed=false
    StaticMesh=DataInterface() {propertyValues {
      DefaultMesh=&SM_CyberDragon_Flame_01
      InstanceIndex=-1
      LODIndex=0
      PreviewMesh=&SM_CyberDragon_Flame_01
      SoftSourceActor=null
      SourceMode=DefaultMeshOnly
      bAllowSamplingFromStreamingLODs=false
      bCaptureTransformsPerFrame=true
      bUsePhysicsBodyVelocity=false
      FilteredSockets []
      LODIndexUserParameter {parameter {name=None, typeDefHandle {registeredTypeIndex=108}, varData []}}
      MeshParameterBinding {parameter {name=None, typeDefHandle {registeredTypeIndex=1}, varData []}}
      SectionFilter {allowedMaterialSlots [0]}
     }}
   )
   &StaticMeshVelocity(Add Velocity Along Mesh Normals=500, Module.Enable Inherit Velocity=true, Module.Enable Mesh Normal Sampling=true)
   &CameraOffset(Camera Offset Amount=-250)
   &InitialMeshOrientation(
    Rotation=&UniformRangedVector(Minimum=<1 1 1>, Maximum=<1 1 1>)
    Use Orientation Vector=false
    Use Rotation Vector=true
   ) {module=InitialMeshOrientation001}
   /Game/Effects/Niagara/Modules/Spawn/FN_Spawn_GBuffer_Color()
  ]
  particleUpdateScript [
   &ParticleState()
   &RegionMask(
    Inner Float=0
    Offset:Vector3f=&MakeVector(
     X=0
     Y=&FloatFromCurve(
      FloatCurve=DataInterface() {propertyValues {Curve {keys [
          {arriveTangent=4.1785, interpMode=RCIM_Cubic, leaveTangent=4.1785, tangentMode=RCTM_User, time=0, value=0}
          {arriveTangent=1.1179, interpMode=RCIM_Cubic, leaveTangent=1.1179, tangentMode=RCTM_User, time=0.1, value=0.9}
          {arriveTangent=0.0224, interpMode=RCIM_Cubic, leaveTangent=0.0224, tangentMode=RCTM_User, time=1, value=1}
         ]}}}
      CurveIndex=$System.NormalizedLoopAge
      Scale Curve=-15000
     )
     Z=0
    )
    Region Radius=5000
    UseOffset=true
    UseRotation=false
   )
   &ScaleMeshSize(Scale Factor=&Multiply_VectorByFloat(
     Vector=&VectorFromFloat(Value=&FloatFromCurve(FloatCurve=DataInterface() {propertyValues {Curve {keys [
           {interpMode=RCIM_Cubic, tangentMode=RCTM_User, time=0, value=0}
           {interpMode=RCIM_Cubic, tangentMode=RCTM_User, time=0.5, value=1}
           {tangentMode=RCTM_User, time=1, value=0}
          ]}}}))
     Float=&Lerp_Float(A=1, B=0, Alpha=$Particles.RegionMask.Float)
    )) {module=ScaleMeshSize001}
   &WindForce(Scale=2, Turbulence Flow Loop Duration=1, Use Landscape=false, Wind Speed=<0 -500 100>)
   &CurlNoiseForce(Noise Frequency=1, Noise Strength=5, Pan Noise=false)
   &AerodynamicDrag(
    Aerodynamic Drag=0.2
    Lift Contribution=0
    Pivot Offset Maximum:Vector3f=<0.1 0.1 1>
    Pivot Offset Minimum:Vector3f=<0 0 -1>
    Pivot Offset Minimum Length:Vector3f=<25 5 2500>
    Use Linear Drag=true
    Use Rotational Drag=true
   )
   &GravityForce(Gravity=&VectorFromCurve(
     VectorCurve=DataInterface() {propertyValues {XCurve {keys [<at 0 1>,<at 1 0>]}, YCurve {keys [<at 0 1>,<at 1 0>]}, ZCurve {keys [<at 0 0>,<at 1 1>]}}}
     Scale Curve=<0 0 -5000>
    ))
   &SolveForcesAndVelocity(Clamp Velocity=true, Limit Acceleration=false, Speed Limit=15000)
   &ScaleMeshSizeBySpeed(Max Scale Factor=<0 0 0>, Speed Threshold=2500)
   &ScaleColor(
    Scale Alpha=&FloatFromCurve(FloatCurve=DataInterface() {propertyValues {Curve {keys [
         <at 0 0 RCIM_Cubic>
         <at 0.6 0.85>
         {arriveTangent=-4.6302, interpMode=RCIM_Cubic, leaveTangent=-4.6302, tangentMode=RCTM_User, time=1, value=0}
        ]}}})
    Scale RGB=&Multiply_Vector(
     A=&MakeVectorFromLinearColorRGB(Color=&ColorFromCurve(
       ColorCurve=DataInterface() {propertyValues {
         AlphaCurve {keys [<at 0 1 RCIM_Cubic>,<at 1 1 RCIM_Cubic>]}
         BlueCurve {keys [<at 0.0631 0.3625 RCIM_Cubic>, {arriveTangent=-0.7387, interpMode=RCIM_Cubic, leaveTangent=-0.7387, time=0.3006, value=1}, <at 0.5538 0>]}
         GreenCurve {keys [<at 0.0631 0 RCIM_Cubic>, {arriveTangent=0.0274, interpMode=RCIM_Cubic, leaveTangent=0.0274, time=0.3006, value=0}, <at 0.5538 0.0134>]}
         RedCurve {keys [<at 0.0631 0.5156 RCIM_Cubic>, {arriveTangent=0.987, interpMode=RCIM_Cubic, leaveTangent=0.987, time=0.3006, value=0.1928}, <at 0.5538 1>]}
        }}
       CurveIndex=$System.NormalizedLoopAge
      ))
     B=&VectorFromFloat(Value=&FloatFromCurve(
       FloatCurve=DataInterface() {propertyValues {Curve {keys [
           {interpMode=RCIM_Cubic, tangentMode=RCTM_User, time=0, value=0}
           {interpMode=RCIM_Cubic, tangentMode=RCTM_User, time=0.1, value=1.05}
           {tangentMode=RCTM_User, time=1, value=1}
          ]}}}
       Scale Curve=&FloatFromCurve(
        FloatCurve=DataInterface() {propertyValues {Curve {keys [
            {arriveTangent=9.0725, interpMode=RCIM_Cubic, leaveTangent=9.0725, tangentMode=RCTM_User, time=0, value=0}
            <at 0.5 1>
            {arriveTangent=-0.0405, interpMode=RCIM_Cubic, leaveTangent=-0.0405, tangentMode=RCTM_User, time=1, value=1}
           ]}}}
        Scale Curve=15
       )
      ))
    )
    ScaleA=true
    ScaleRGB=true
   )
   &KillParticlesInVolume(
    Origin Offset:Vector3f=&MakeVector(
     X=0
     Y=&FloatFromCurve(
      FloatCurve=DataInterface() {propertyValues {Curve {keys [
          {arriveTangent=1.5884, interpMode=RCIM_Cubic, leaveTangent=1.5884, tangentMode=RCTM_User, time=0, value=0}
          {arriveTangent=1.1179, interpMode=RCIM_Cubic, leaveTangent=1.1179, tangentMode=RCTM_User, time=0.5, value=1}
          {arriveTangent=0.0224, interpMode=RCIM_Cubic, leaveTangent=0.0224, tangentMode=RCTM_User, time=1, value=1}
         ]}}}
      CurveIndex=$System.NormalizedLoopAge
      Scale Curve=-15000
     )
     Z=0
    )
    Sphere Radius=6500
   )
   &AlignSpriteToMeshOrientation(
    Mesh Orientation Relative Sprite Alignment Vector=<0 0 0>
    Mesh Orientation Relative Sprite Facing Vector:Vector3f=<0 0 0>
    SetAlignment=true
    SetFacing=true
   )
   &ScaleMeshSize(Scale Factor=&VectorFromFloat(Value=&FloatFromCurve(FloatCurve=DataInterface() {propertyValues {Curve {keys [<at 0.75 1 RCIM_Cubic>, {arriveTangent=-4.0622, interpMode=RCIM_Cubic, leaveTangent=-4.0622, tangentMode=RCTM_User, time=1, value=0.25}]}}}))) {enabled=false}
   &DynamicMaterialParameters(
    Index 1 Param 1=0.25
    Index 2 Param 1=100
    Index 2 Param 2=0.02
    Index 2 Param 3=1
    Index 2 Param 4=&FloatFromCurve(FloatCurve=$User.MainTimingCurve, CurveIndex=$System.NormalizedLoopAge)
    Module.Param0WriteEnabled=true
    Module.Param0WriteEnabled1=true
    Module.Param0WriteEnabled2=true
    Module.Param1WriteEnabled=true
    Module.Param1WriteEnabled1=true
    Module.Param1WriteEnabled2=true
    Module.Param2WriteEnabled=true
    Module.Param2WriteEnabled1=true
    Module.Param2WriteEnabled2=true
    Module.Param3WriteEnabled=true
    Module.Param3WriteEnabled1=true
    Module.Param3WriteEnabled2=true
    Write Parameter Index 1=true
    Write Parameter Index 2=true
   )
  ]
  rendererClasses [&NiagaraMeshRendererProperties]
  renderers [
   {rendererClass=&NiagaraMeshRendererProperties, rendererIndex=0}
   {rendererClass=&NiagaraMeshRendererProperties, rendererIndex=1}
  ]
 }
 ErodeCubes001 {
  bEnabled=false
  simTarget=GPUComputeSim
  emitterUpdateScript [
   &EmitterState()
   &SpawnRate(
    SpawnRate=&FloatFromCurve(
     FloatCurve=DataInterface() {propertyValues {Curve {keys [<at 0.7 0>,<at 0.8 1>,<at 1 0>]}}}
     CurveIndex=$System.NormalizedLoopAge
     Scale Curve=500
    )
    Use Spawn Probability=false
   ) {module=SpawnRate003}
  ]
  particleSpawnScript [
   /Niagara/Modules/Spawn/Initialization/V2/InitializeParticle(
    Color=<rgba 0.6038 0.6038 0.6038 1>
    Lifetime Max=4.5
    Lifetime Min=2
    Lifetime Mode=::NewEnumerator1
    Mesh Renderer Array Visibility Mode=::NewEnumerator3
    Mesh Renderer Info=DataInterface() {propertyValues {MeshRenderer=&self:BlobFlameMesh_5.NiagaraMeshRendererProperties_0}}
    Mesh Scale Mode=::NewEnumerator1
    Mesh Uniform Scale Max=100
    Mesh Uniform Scale Min=25
    UsePositionOffset=false
    Write Material Random=true
   ) {module=InitializeParticle001}
   &AddVelocity(Velocity=<0 0 5>) {module=AddVelocity001}
   &CameraOffset(Camera Offset Amount=-250)
   &InitialMeshOrientation(
    Rotation=&UniformRangedVector(Minimum=<1 1 1>, Maximum=<1 1 1>)
    Use Orientation Vector=false
    Use Rotation Vector=true
   ) {module=InitialMeshOrientation001}
   &FN_Spawn_GBuffer_Color()
   /Niagara/Modules/Math/PlaceParticlesOnDepthBuffer_GPU(Max Z Depth=5000, Module.Use Max Z Depth=true)
   /Game/Effects/Niagara/Modules/Spawn/FN_Spawn_GBuffer_Color_WithinRangeV2(HueMax=350)
   &ShapeLocation(MaskSpawnGroup=false, Override Seed=false, UseOwnerScale=true)
   &InitialMeshOrientation(Use Rotation Vector=false)
  ]
  particleUpdateScript [
   &ParticleState()
   &ScaleMeshSize(Scale Factor=&Multiply_VectorByFloat(
     Vector=&VectorFromFloat(Value=&FloatFromCurve(FloatCurve=DataInterface() {propertyValues {Curve {keys [
           {interpMode=RCIM_Cubic, tangentMode=RCTM_User, time=0, value=0}
           {interpMode=RCIM_Cubic, tangentMode=RCTM_User, time=0.5, value=1}
           {tangentMode=RCTM_User, time=1, value=0}
          ]}}}))
     Float=&Lerp_Float(A=1, B=0, Alpha=$Particles.RegionMask.Float)
    )) {module=ScaleMeshSize001}
   &CurlNoiseForce(Noise Frequency=5, Noise Strength=25, Pan Noise=false)
   &GravityForce(Gravity=&VectorFromCurve(
     VectorCurve=DataInterface() {propertyValues {XCurve {keys [<at 0 1>,<at 1 0>]}, YCurve {keys [<at 0 1>,<at 1 0>]}, ZCurve {keys [<at 0 0>,<at 1 1>]}}}
     Scale Curve=<0 0 -5000>
    ))
   &SolveForcesAndVelocity(Clamp Velocity=true, Limit Acceleration=false, Speed Limit=15000)
   &ScaleMeshSizeBySpeed(Max Scale Factor=<0 0 0>, Speed Threshold=2500)
   &ScaleColor(
    Scale Alpha=&FloatFromCurve(FloatCurve=DataInterface() {propertyValues {Curve {keys [
         <at 0 0 RCIM_Cubic>
         <at 0.6 0.85>
         {arriveTangent=-4.6302, interpMode=RCIM_Cubic, leaveTangent=-4.6302, tangentMode=RCTM_User, time=1, value=0}
        ]}}})
    Scale RGB=&Multiply_Vector(
     A=&MakeVectorFromLinearColorRGB(Color=&ColorFromCurve(
       ColorCurve=DataInterface() {propertyValues {
         AlphaCurve {keys [<at 0 1 RCIM_Cubic>,<at 1 1 RCIM_Cubic>]}
         BlueCurve {keys [<at 0.0631 0.3625 RCIM_Cubic>, {arriveTangent=-0.7387, interpMode=RCIM_Cubic, leaveTangent=-0.7387, time=0.3006, value=1}, <at 0.5538 0>]}
         GreenCurve {keys [<at 0.0631 0 RCIM_Cubic>, {arriveTangent=0.0274, interpMode=RCIM_Cubic, leaveTangent=0.0274, time=0.3006, value=0}, <at 0.5538 0.0134>]}
         RedCurve {keys [<at 0.0631 0.5156 RCIM_Cubic>, {arriveTangent=0.987, interpMode=RCIM_Cubic, leaveTangent=0.987, time=0.3006, value=0.1928}, <at 0.5538 1>]}
        }}
       CurveIndex=$System.NormalizedLoopAge
      ))
     B=&VectorFromFloat(Value=&FloatFromCurve(
       FloatCurve=DataInterface() {propertyValues {Curve {keys [
           {interpMode=RCIM_Cubic, tangentMode=RCTM_User, time=0, value=0}
           {interpMode=RCIM_Cubic, tangentMode=RCTM_User, time=0.1, value=1.05}
           {tangentMode=RCTM_User, time=1, value=1}
          ]}}}
       Scale Curve=&FloatFromCurve(
        FloatCurve=DataInterface() {propertyValues {Curve {keys [
            {arriveTangent=9.0725, interpMode=RCIM_Cubic, leaveTangent=9.0725, tangentMode=RCTM_User, time=0, value=0}
            <at 0.5 1>
            {arriveTangent=-0.0405, interpMode=RCIM_Cubic, leaveTangent=-0.0405, tangentMode=RCTM_User, time=1, value=1}
           ]}}}
        Scale Curve=15
       )
      ))
    )
    ScaleA=true
    ScaleRGB=true
   )
   &ScaleMeshSize(Scale Factor=&VectorFromFloat(Value=&FloatFromCurve(FloatCurve=DataInterface() {propertyValues {Curve {keys [<at 0.75 1 RCIM_Cubic>, {arriveTangent=-4.0622, interpMode=RCIM_Cubic, leaveTangent=-4.0622, tangentMode=RCTM_User, time=1, value=0.25}]}}}))) {enabled=false}
   &DynamicMaterialParameters(
    Index 1 Param 1=0.25
    Index 2 Param 1=100
    Index 2 Param 2=0.02
    Index 2 Param 3=1
    Index 2 Param 4=&FloatFromCurve(FloatCurve=$User.MainTimingCurve, CurveIndex=$System.NormalizedLoopAge)
    Module.Param0WriteEnabled=true
    Module.Param0WriteEnabled1=true
    Module.Param0WriteEnabled2=true
    Module.Param1WriteEnabled=true
    Module.Param1WriteEnabled1=true
    Module.Param1WriteEnabled2=true
    Module.Param2WriteEnabled=true
    Module.Param2WriteEnabled1=true
    Module.Param2WriteEnabled2=true
    Module.Param3WriteEnabled=true
    Module.Param3WriteEnabled1=true
    Module.Param3WriteEnabled2=true
    Write Parameter Index 1=true
    Write Parameter Index 2=true
   )
  ]
  rendererClasses [&NiagaraMeshRendererProperties]
  renderers [
   {rendererClass=&NiagaraMeshRendererProperties, rendererIndex=0}
   {rendererClass=&NiagaraMeshRendererProperties, rendererIndex=1}
  ]
 }
 Flame_LIght {
  emitterUpdateScript [
   &EmitterState()
   &SpawnBurst_Instantaneous(Spawn Time=0.2, Use Spawn Probability=false, UseLoopCountLimit=false)
   &SpawnRate(
    SpawnRate=&Multiply_Float(
     A=20
     B=&FloatFromCurve(FloatCurve=DataInterface() {propertyValues {Curve {keys [<at 0.3 1>,<at 0.5 0>]}}}, CurveIndex=$Emitter.NormalizedLoopAge)
    )
    Use Spawn Probability=false
   ) {enabled=false, module=SpawnRate001}
  ]
  particleSpawnScript [
   /Niagara/Modules/Spawn/Initialization/InitializeParticle(
    Lifetime=&UniformRangedFloat(Minimum=3, Maximum=3)
    Module.Write Color=true
    Module.Write Lifetime=true
    Module.Write Mass=true
    Module.Write Material Random=true
    Module.Write Position=false
    Module.Write Scale=false
    Module.Write SpriteRotation=false
    Module.Write SpriteSize=true
    Position=&SimulationPosition()
    Sprite Size=&UniformRangedVector2D(Minimum=<25 25>, Maximum=<25 25>)
   ) {module=InitializeParticle001}
   &SystemLocation(MaskSpawnGroup=false, Offset=<0 2500 0>)
  ]
  particleUpdateScript [
   &ParticleState()
   &CurlNoiseForce(
    Noise Frequency=1
    Noise Strength=&FloatFromCurve(
     FloatCurve=DataInterface() {propertyValues {Curve {keys [{interpMode=RCIM_Cubic, tangentMode=RCTM_User, time=0, value=1}, {interpMode=RCIM_Cubic, tangentMode=RCTM_User, time=0.1, value=0}]}}}
     Scale Curve=5000
    )
    Pan Noise=true
    Pan Noise Field=<0 -1 0>
   )
   &GravityForce(Gravity=<0 0 -250>)
   &Drag(
    Drag=&FloatFromCurve(FloatCurve=DataInterface() {propertyValues {Curve {keys [<at 0 1>,<at 1 0>]}}}, Scale Curve=3)
    Use Linear Drag=true
    Use Rotational Drag=true
   )
   &SolveForcesAndVelocity(Clamp Velocity=false, Limit Acceleration=false)
   &ScaleColor(
    Scale Alpha=&FloatFromCurve(FloatCurve=DataInterface() {propertyValues {Curve {keys [<at 0 1>,<at 1 0>]}}})
    Scale RGB=&Multiply_Vector(
     A=&MakeVectorFromLinearColorRGB(Color=&ColorFromCurve(ColorCurve=DataInterface() {propertyValues {
         AlphaCurve {keys [<at 0 1 RCIM_Cubic>,<at 1 1 RCIM_Cubic>]}
         BlueCurve {keys [<at 0 1 RCIM_Cubic>,<at 0.811 0.7384 RCIM_Cubic>]}
         GreenCurve {keys [<at 0 0 RCIM_Cubic>,<at 0.811 0 RCIM_Cubic>]}
         RedCurve {keys [<at 0 0.0969 RCIM_Cubic>,<at 0.811 1 RCIM_Cubic>]}
        }}))
     B=&VectorFromFloat(Value=250)
    )
    ScaleA=true
    ScaleRGB=true
   )
   /SetVariables(Particles.LightRadius=&FloatFromCurve(
     FloatCurve=DataInterface() {propertyValues {Curve {keys [{interpMode=RCIM_Cubic, tangentMode=RCTM_User, time=0, value=0}, <at 0.25 1>, <at 0.98 0.95>, {tangentMode=RCTM_User, time=1, value=0}]}}}
     Scale Curve=80
    )) {isSetParameters=true}
  ]
  rendererClasses [&NiagaraLightRendererProperties]
  renderers [{rendererClass=&NiagaraLightRendererProperties, rendererIndex=0}]
 }
 PreFlame_Cone {
  emitterUpdateScript [
   &EmitterState()
   &SpawnBurst_Instantaneous(Use Spawn Probability=false, UseLoopCountLimit=false)
  ]
  particleSpawnScript [
   /Niagara/Modules/Spawn/Initialization/V2/InitializeParticle(
    Color=<rgba 0 0.5153 10 0.5>
    Lifetime=0.6
    Mesh Scale=<15 15 15>
    Mesh Scale Mode=::NewEnumerator4
    Sprite Size Mode=::NewEnumerator3
    Uniform Sprite Size=2500
    UsePositionOffset=false
    Write Material Random=true
   ) {module=InitializeParticle001}
   &DynamicMaterialParameters(
    Index 0 Param 1=1
    Index 0 Param 2=0.35
    Index 0 Param 3=1
    Module.Param0WriteEnabled=true
    Module.Param1WriteEnabled=true
    Module.Param2WriteEnabled=true
    Module.Param3WriteEnabled=false
   )
   &InitialMeshOrientation(Mesh Orientation Mode=::NewEnumerator2, Orientation Vector=<0 0 0>, Use Rotation Vector=false)
   &CameraOffset(Camera Offset Amount=1500)
   &SystemLocation(MaskSpawnGroup=false, Offset=<0 4500 0>)
  ]
  particleUpdateScript [
   &ParticleState()
   &ScaleSpriteSize(Uniform Curve Sprite Scale=DataInterface() {propertyValues {Curve {keys [
        {arriveTangent=3.5436, interpMode=RCIM_Cubic, leaveTangent=3.5436, tangentMode=RCTM_User, time=0, value=0.25}
        {arriveTangent=-2.006, interpMode=RCIM_Cubic, leaveTangent=-2.006, tangentMode=RCTM_User, time=1, value=0.8}
       ]}}})
   &ScaleMeshSize(Scale Factor=&VectorFromCurve(VectorCurve=DataInterface() {propertyValues {
       XCurve {keys [
         {arriveTangent=1.5385, time=0, value=0}
         {arriveTangent=0.3483, interpMode=RCIM_Cubic, leaveTangent=0.3483, tangentMode=RCTM_User, time=1, value=1}
        ]}
       YCurve {keys [
         {arriveTangent=2.3109, interpMode=RCIM_Cubic, leaveTangent=2.3109, tangentMode=RCTM_User, time=0, value=0}
         {arriveTangent=0.6124, interpMode=RCIM_Cubic, leaveTangent=0.6124, tangentMode=RCTM_User, time=1, value=1}
        ]}
       ZCurve {keys [
         {arriveTangent=2.3109, interpMode=RCIM_Cubic, leaveTangent=2.3109, tangentMode=RCTM_User, time=0, value=0}
         {arriveTangent=0.6124, interpMode=RCIM_Cubic, leaveTangent=0.6124, tangentMode=RCTM_User, time=1, value=1}
        ]}
      }}))
   &ScaleColor(
    Scale Alpha=&FloatFromCurve(FloatCurve=DataInterface() {propertyValues {Curve {keys [
         {arriveTangent=10809.126, interpMode=RCIM_Cubic, time=0, value=0}
         {arriveTangent=-2.7716, interpMode=RCIM_Cubic, leaveTangent=-2.7716, tangentMode=RCTM_User, time=0.1, value=1}
         {interpMode=RCIM_Cubic, leaveTangent=-3.1778, time=1, value=0}
        ]}}})
    Scale RGB=&MakeVectorFromLinearColorRGB(Color=&ColorFromCurve(ColorCurve=DataInterface() {propertyValues {
        AlphaCurve {keys []}
        BlueCurve {keys [<at 0.1026 1 RCIM_Cubic>,<at 0.959 0.5 RCIM_Cubic>]}
        GreenCurve {keys [<at 0.1026 1 RCIM_Cubic>,<at 0.959 0.2461 RCIM_Cubic>]}
        RedCurve {keys [<at 0.1026 1 RCIM_Cubic>,<at 0.959 0.2651 RCIM_Cubic>]}
       }}))
    ScaleA=true
    ScaleRGB=true
   )
   /SetVariables(Particles.FlareColor=&MultiplyLinearColor(
     Linear Color=<rgba 0.121 0 25 1>
     Scale Factor=&MakeVector4_FromFloat(Float=&FloatFromCurve(FloatCurve=DataInterface() {propertyValues {Curve {keys [
           {interpMode=RCIM_Cubic, tangentMode=RCTM_User, time=0, value=0}
           {arriveTangent=-1.9867, interpMode=RCIM_Cubic, leaveTangent=-1.9867, tangentMode=RCTM_User, time=0.1, value=1}
           {tangentMode=RCTM_User, time=1, value=0}
          ]}}}))
    )) {isSetParameters=true}
   &DynamicMaterialParameters(
    Index 1 Param 1=&FloatFromCurve(FloatCurve=DataInterface() {propertyValues {Curve {keys [<at 0 -1 RCIM_Cubic>,<at 1 0>]}}})
    Index 1 Param 2=1
    Module.Param0WriteEnabled1=true
    Module.Param1WriteEnabled1=true
    Module.Param2WriteEnabled1=false
    Module.Param3WriteEnabled1=false
    Write Parameter Index 0=false
    Write Parameter Index 1=true
   ) {module=DynamicMaterialParameters001}
  ]
  rendererClasses [&NiagaraMeshRendererProperties, &NiagaraSpriteRendererProperties]
  renderers [
   {rendererClass=&NiagaraMeshRendererProperties, rendererIndex=0}
   {rendererClass=&NiagaraSpriteRendererProperties, rendererIndex=1}
  ]
 }
 StartCubes {
  simTarget=GPUComputeSim
  emitterUpdateScript [
   &EmitterState()
   &SpawnRate(
    SpawnRate=&FloatFromCurve(
     FloatCurve=DataInterface() {propertyValues {Curve {keys [<at 0.1 0>,<at 0.15 1>,<at 0.4 0>]}}}
     CurveIndex=$System.NormalizedLoopAge
     Scale Curve=5000
    )
    Use Spawn Probability=false
   ) {module=SpawnRate003}
  ]
  particleSpawnScript [
   /Niagara/Modules/Spawn/Initialization/InitializeParticle(
    Color=<rgba 0.6038 0.6038 0.6038 1>
    Lifetime=&RandomRangeFloat(Minimum=0.25, Maximum=1)
    Mass=&UniformRangedFloat(Minimum=1, Maximum=2)
    Mesh Scale=&RandomRangeVector(
     Minimum=&VectorFromFloat(Value=50)
     Maximum=&VectorFromFloat(Value=150)
    )
    Module.Write Color=true
    Module.Write Lifetime=true
    Module.Write Mass=true
    Module.Write Material Random=true
    Module.Write Position=false
    Module.Write Scale=true
    Module.Write SpriteRotation=false
    Module.Write SpriteSize=true
    Position=&SimulationPosition()
    Sprite Size=<50 50>
   )
   &AddVelocity(
    Cone Angle=1
    Cone Axis=<0 -1 0>
    Module.UseConeFalloff=false
    Module.UseVelDistribution=false
    Override Seed=false
    Velocity Mode=::NewEnumerator2
    Velocity Speed=&FloatFromCurve(
     FloatCurve=DataInterface() {propertyValues {Curve {keys [
         {arriveTangent=-2.5249, interpMode=RCIM_Cubic, leaveTangent=-2.5249, tangentMode=RCTM_User, time=0, value=0.9847}
         {arriveTangent=-0.0405, interpMode=RCIM_Cubic, leaveTangent=-0.0405, tangentMode=RCTM_User, time=1, value=0.5}
        ]}}}
     CurveIndex=$Emitter.NormalizedLoopAge
     Scale Curve=1500
    )
   ) {module=AddVelocity001}
   &StaticMeshLocation(
    MaskSpawnGroup=false
    Offset Position Along Sampled Normal=25
    OffsetAlongNormal=true
    Orientation Sampling:ENiagara_AttributeSamplingApplyOutput=::NewEnumerator1
    OverrideRandomSeed=false
    Position Sampling:ENiagaraSkelMeshPositionSamplingMode=::NewEnumerator1
    RGBA Color Application Mask:Vector4f=<0 0 0 0>
    SamplePosOffset=true
    Sampled Position Offset:Vector3f=<0 0 250>
    Sampled Position Offset Coordinate Space:ENiagaraSkelMeshTransforms=::NewEnumerator1
    Static Mesh=DataInterface() {propertyValues {
      DefaultMesh=&SM_CyberDragon_Flame_01
      InstanceIndex=-1
      LODIndex=0
      PreviewMesh=&SM_CyberDragon_Flame_01
      SoftSourceActor=null
      SourceMode=DefaultMeshOnly
      bAllowSamplingFromStreamingLODs=false
      bCaptureTransformsPerFrame=true
      bUsePhysicsBodyVelocity=false
      FilteredSockets []
      LODIndexUserParameter {parameter {name=None, typeDefHandle {registeredTypeIndex=108}, varData []}}
      MeshParameterBinding {parameter {name=None, typeDefHandle {registeredTypeIndex=1}, varData []}}
      SectionFilter {allowedMaterialSlots []}
     }}
    Velocity Sampling:ENiagara_AttributeSamplingApplyOutput=::NewEnumerator0
    VelocitySpeedLimit=false
    Vertex Color Sampling:ENiagara_AttributeSamplingApplyOutput=::NewEnumerator0
   ) {module=StaticMeshLocation001}
   &SampleStaticMesh(
    OverrideRandomSeed=false
    StaticMesh=DataInterface() {propertyValues {
      DefaultMesh=&SM_CyberDragon_Flame_01
      InstanceIndex=-1
      LODIndex=0
      PreviewMesh=&SM_CyberDragon_Flame_01
      SoftSourceActor=null
      SourceMode=DefaultMeshOnly
      bAllowSamplingFromStreamingLODs=false
      bCaptureTransformsPerFrame=true
      bUsePhysicsBodyVelocity=false
      FilteredSockets []
      LODIndexUserParameter {parameter {name=None, typeDefHandle {registeredTypeIndex=108}, varData []}}
      MeshParameterBinding {parameter {name=None, typeDefHandle {registeredTypeIndex=1}, varData []}}
      SectionFilter {allowedMaterialSlots [0]}
     }}
   )
   &StaticMeshVelocity(Add Velocity Along Mesh Normals=25, Module.Enable Inherit Velocity=true, Module.Enable Mesh Normal Sampling=true)
   &CameraOffset()
   &SystemLocation(MaskSpawnGroup=false) {enabled=false}
   &InitialMeshOrientation(
    Rotation=&UniformRangedVector(Minimum=<1 1 1>, Maximum=<1 1 1>)
    Use Orientation Vector=false
    Use Rotation Vector=true
   ) {enabled=false, module=InitialMeshOrientation001}
   &FN_Spawn_GBuffer_Color()
  ]
  particleUpdateScript [
   &ParticleState()
   &RegionMask(
    Inner Float=0
    Invert Region=true
    Offset:Vector3f=&MakeVector(
     X=0
     Y=&FloatFromCurve(
      FloatCurve=DataInterface() {propertyValues {Curve {keys [
          {arriveTangent=4.1785, interpMode=RCIM_Cubic, leaveTangent=4.1785, tangentMode=RCTM_User, time=0, value=0}
          {arriveTangent=1.1179, interpMode=RCIM_Cubic, leaveTangent=1.1179, tangentMode=RCTM_User, time=0.25, value=0.95}
          {arriveTangent=0.0224, interpMode=RCIM_Cubic, leaveTangent=0.0224, tangentMode=RCTM_User, time=1, value=1}
         ]}}}
      CurveIndex=$System.NormalizedLoopAge
      Scale Curve=15000
     )
     Z=0
    )
    Region Falloff=0
    Region Radius=5000
    UseOffset=true
    UseRotation=false
   )
   &ScaleMeshSize(Scale Factor=&Multiply_VectorByFloat(
     Vector=&VectorFromFloat(Value=&FloatFromCurve(FloatCurve=DataInterface() {propertyValues {Curve {keys [
           {interpMode=RCIM_Cubic, tangentMode=RCTM_User, time=0, value=0}
           {interpMode=RCIM_Cubic, tangentMode=RCTM_User, time=0.5, value=1}
           {tangentMode=RCTM_User, time=1, value=0}
          ]}}}))
     Float=&Lerp_Float(A=1, B=0, Alpha=$Particles.RegionMask.Float)
    )) {module=ScaleMeshSize001}
   &WindForce(Scale=2, Turbulence Flow Loop Duration=1, Use Landscape=false, Wind Speed=<0 -150 100>, Wind Speed Scale=5)
   &CurlNoiseForce(Noise Frequency=1, Noise Strength=5, Pan Noise=false)
   &AerodynamicDrag(
    Aerodynamic Drag=0.2
    Lift Contribution=0
    Pivot Offset Maximum:Vector3f=<0.1 0.1 1>
    Pivot Offset Minimum:Vector3f=<0 0 -1>
    Pivot Offset Minimum Length:Vector3f=<2 5 2>
    Use Linear Drag=true
    Use Rotational Drag=true
   )
   &SolveForcesAndVelocity(Clamp Velocity=true, Limit Acceleration=false, Speed Limit=2500)
   &ScaleMeshSizeBySpeed(Max Scale Factor=<0 0 0>, Speed Threshold=2500)
   &ScaleColor(
    Scale Alpha=&FloatFromCurve(FloatCurve=DataInterface() {propertyValues {Curve {keys [
         <at 0 0 RCIM_Cubic>
         <at 0.6 0.85>
         {arriveTangent=-4.6302, interpMode=RCIM_Cubic, leaveTangent=-4.6302, tangentMode=RCTM_User, time=1, value=0}
        ]}}})
    Scale RGB=&Multiply_Vector(
     A=&MakeVectorFromLinearColorRGB(Color=&ColorFromCurve(
       ColorCurve=DataInterface() {propertyValues {
         AlphaCurve {keys [<at 0 1 RCIM_Cubic>,<at 1 1 RCIM_Cubic>]}
         BlueCurve {keys [<at 0.0631 0.3625 RCIM_Cubic>, {arriveTangent=-0.7387, interpMode=RCIM_Cubic, leaveTangent=-0.7387, time=0.3006, value=1}, <at 0.5538 0>]}
         GreenCurve {keys [<at 0.0631 0 RCIM_Cubic>, {arriveTangent=0.0274, interpMode=RCIM_Cubic, leaveTangent=0.0274, time=0.3006, value=0}, <at 0.5538 0.0134>]}
         RedCurve {keys [<at 0.0631 0.5156 RCIM_Cubic>, {arriveTangent=0.987, interpMode=RCIM_Cubic, leaveTangent=0.987, time=0.3006, value=0.1928}, <at 0.5538 1>]}
        }}
       CurveIndex=$System.NormalizedLoopAge
      ))
     B=&VectorFromFloat(Value=&FloatFromCurve(
       FloatCurve=DataInterface() {propertyValues {Curve {keys [
           {interpMode=RCIM_Cubic, tangentMode=RCTM_User, time=0, value=0}
           {interpMode=RCIM_Cubic, tangentMode=RCTM_User, time=0.1, value=1.05}
           {tangentMode=RCTM_User, time=1, value=1}
          ]}}}
       Scale Curve=&FloatFromCurve(
        FloatCurve=DataInterface() {propertyValues {Curve {keys [
            {arriveTangent=9.0725, interpMode=RCIM_Cubic, leaveTangent=9.0725, tangentMode=RCTM_User, time=0, value=0}
            <at 0.5 1>
            {arriveTangent=-0.0405, interpMode=RCIM_Cubic, leaveTangent=-0.0405, tangentMode=RCTM_User, time=1, value=1}
           ]}}}
        Scale Curve=15
       )
      ))
    )
    ScaleA=true
    ScaleRGB=true
   )
   &KillParticlesInVolume(
    Origin Offset:Vector3f=&MakeVector(
     X=0
     Y=&FloatFromCurve(
      FloatCurve=DataInterface() {propertyValues {Curve {keys [{arriveTangent=4.1785, time=0.05, value=0}, {arriveTangent=0.0224, leaveTangent=0.0224, time=0.5, value=1}]}}}
      CurveIndex=$System.NormalizedLoopAge
      Scale Curve=-15000
     )
     Z=0
    )
    Sphere Radius=7500
   )
   &AlignSpriteToMeshOrientation(
    Mesh Orientation Relative Sprite Alignment Vector=<0 0 0>
    Mesh Orientation Relative Sprite Facing Vector:Vector3f=<0 0 0>
    SetAlignment=true
    SetFacing=true
   )
   &ScaleMeshSize(Scale Factor=&VectorFromFloat(Value=&FloatFromCurve(
      FloatCurve=DataInterface() {propertyValues {Curve {keys [<at 0.75 1 RCIM_Cubic>, {arriveTangent=-4.0622, interpMode=RCIM_Cubic, leaveTangent=-4.0622, tangentMode=RCTM_User, time=1, value=0.25}]}}}
      Scale Curve=&Lerp_Float(A=1, B=0, Alpha=$Particles.RegionMask.Float)
     )))
   &DynamicMaterialParameters(
    Index 1 Param 1=0.25
    Index 2 Param 1=100
    Index 2 Param 2=0.02
    Index 2 Param 3=1
    Index 2 Param 4=&FloatFromCurve(FloatCurve=$User.MainTimingCurve, CurveIndex=$System.NormalizedLoopAge)
    Module.Param0WriteEnabled=true
    Module.Param0WriteEnabled1=true
    Module.Param0WriteEnabled2=true
    Module.Param1WriteEnabled=true
    Module.Param1WriteEnabled1=true
    Module.Param1WriteEnabled2=true
    Module.Param2WriteEnabled=true
    Module.Param2WriteEnabled1=true
    Module.Param2WriteEnabled2=true
    Module.Param3WriteEnabled=true
    Module.Param3WriteEnabled1=true
    Module.Param3WriteEnabled2=true
    Write Parameter Index 1=true
    Write Parameter Index 2=true
   )
  ]
  rendererClasses [&NiagaraMeshRendererProperties]
  renderers [{rendererClass=&NiagaraMeshRendererProperties, rendererIndex=0}]
 }
 Start_Flame {
  simTarget=GPUComputeSim
  emitterUpdateScript [
   &EmitterState()
   &SpawnRate(
    SpawnRate=&FloatFromCurve(FloatCurve=DataInterface() {propertyValues {Curve {keys [<at 0.25 1>,<at 0.35 0>]}}}, CurveIndex=$System.NormalizedLoopAge, Scale Curve=25)
    Use Spawn Probability=false
   )
  ]
  particleSpawnScript [
   /Niagara/Modules/Spawn/Initialization/V2/InitializeParticle(
    Color=<rgba 100 0 74.4605 1>
    Lifetime=&RandomRangeFloat(Minimum=2, Maximum=3)
    Sprite Rotation Angle=270
    Sprite Rotation Mode=::NewEnumerator0
    Sprite Size Mode=::NewEnumerator3
    Uniform Sprite Size=&Multiply_Float(
     A=&RandomRangeFloat(Minimum=250, Maximum=800)
     B=&FloatFromCurve(FloatCurve=DataInterface() {propertyValues {Curve {keys [<at 0 0.5>,<at 1 1>]}}}, CurveIndex=$System.NormalizedLoopAge)
    )
    UsePositionOffset=false
    Write Material Random=true
   )
   &Color(Color=&MultiplyLinearColor(Linear Color=$Particles.StaticMeshLocation.SampledVertexColor, Scale Factor=<0 0 10 1>)) {enabled=false}
   &CurlNoiseForce(Noise Frequency=5, Noise Strength=500, Pan Noise=false) {module=CurlNoiseForce001}
   &ShapeLocation(MaskSpawnGroup=false, Offset=<0 4500 -200>, Offset Mode=::NewEnumerator0, Override Seed=false, Sphere Radius=150, UseOwnerScale=true)
   &AddVelocity(
    Velocity=&Multiply_VectorByFloat(
     Vector=<1 -250 1>
     Float=&DistanceBetweenPositions(A=&AddVectorToPosition(Position=$Engine.Owner.Position, Vector=<0 250 0>), B=$Particles.Position)
    )
    Velocity Speed Scale=0.0005
   )
   &ApplyInitialForces()
   &DynamicMaterialParameters(
    Index 0 Param 1=&FloatFromCurve(FloatCurve=DataInterface() {propertyValues {Curve {keys [
         {arriveTangent=8.397, interpMode=RCIM_Cubic, leaveTangent=4.6996, tangentMode=RCTM_User, time=-0.0012, value=-0.0261}
         {arriveTangent=1.3472, interpMode=RCIM_Cubic, leaveTangent=1.3472, tangentMode=RCTM_User, time=0.1801, value=0.7926}
         <at 0.5 1 RCIM_Cubic>
        ]}}})
    Index 0 Param 2=&UniformRangedFloat(Minimum=0.333, Maximum=1)
    Index 0 Param 3=&FloatFromCurve(FloatCurve=DataInterface() {propertyValues {Curve {keys [{arriveTangent=1.4224, interpMode=RCIM_Cubic, time=0.5, value=1}, <at 1 2 RCIM_Cubic>]}}})
    Index 0 Param 4=&UniformRangedFloat(Minimum=1.333, Maximum=2)
    Module.Param0WriteEnabled=true
    Module.Param1WriteEnabled=true
    Module.Param2WriteEnabled=true
    Module.Param3WriteEnabled=true
   ) {module=DynamicMaterialParameters001}
   &CameraOffset(Camera Offset Amount=1500)
  ]
  particleUpdateScript [
   &ParticleState()
   &RegionMask(
    Offset:Vector3f=&MakeVector(
     X=0
     Y=&FloatFromCurve(FloatCurve=DataInterface() {propertyValues {Curve {keys [<at 0 0>,<at 1 1>]}}}, CurveIndex=$System.NormalizedLoopAge, Scale Curve=-15000)
     Z=0
    )
    Region Falloff=0.45
    Region Radius=6500
    UseOffset=true
    UseRotation=false
   )
   &CurlNoiseForce(
    Noise Frequency=1
    Noise Strength=&FloatFromCurve(
     FloatCurve=DataInterface() {propertyValues {Curve {keys [{interpMode=RCIM_Cubic, tangentMode=RCTM_User, time=0, value=1}, {interpMode=RCIM_Cubic, tangentMode=RCTM_User, time=0.1, value=0}]}}}
     Scale Curve=5000
    )
    Pan Noise=true
    Pan Noise Field=<0 -2 0>
   )
   &GravityForce(Gravity=&VectorFromCurve(
     VectorCurve=DataInterface() {propertyValues {XCurve {keys [<at 0 1>,<at 1 0>]}, YCurve {keys [<at 0 1>,<at 1 0>]}, ZCurve {keys [<at 0 0>,<at 0.25 1>]}}}
     Scale Curve=<1 1 -250>
    ))
   &Drag(
    Drag=&FloatFromCurve(FloatCurve=DataInterface() {propertyValues {Curve {keys [
         {interpMode=RCIM_Cubic, tangentMode=RCTM_User, time=0, value=0}
         {interpMode=RCIM_Cubic, tangentMode=RCTM_User, time=0.5, value=1}
         {tangentMode=RCTM_User, time=1, value=0}
        ]}}})
    Use Linear Drag=true
    Use Rotational Drag=true
   )
   &SubUVAnimation(
    Start Frame:NiagaraInt32=&RandomRangeInt(Minimum=0, Maximum=16)
    SubUV Animation Mode:ENiagara_SubUVLookupMode=::NewEnumerator1
    SubUV Curve=DataInterface() {propertyValues {Curve {keys [{arriveTangent=1.6038, time=0, value=0}, {leaveTangent=0.6912, time=1, value=1}]}}}
    SubUV Loop Count:NiagaraFloat=&RandomRangeFloat(Minimum=1, Maximum=1)
   )
   &SolveForcesAndVelocity(Clamp Velocity=false, Limit Acceleration=false)
   &DynamicMaterialParameters(
    Index 0 Param 1=&FloatFromCurve(FloatCurve=DataInterface() {propertyValues {Curve {keys [<at 0 1>,<at 1 0>]}}})
    Index 0 Param 2=&RandomRangeFloat(Minimum=0, Maximum=1)
    Index 0 Param 3=8
    Index 0 Param 4=$Particles.SubImageIndex
    Index 1 Param 1=&RandomRangeFloat(Minimum=0, Maximum=1)
    Index 1 Param 2=&RandomRangeFloat(Minimum=-1, Maximum=1)
    Index 1 Param 3=&FloatFromCurve(
     FloatCurve=DataInterface() {propertyValues {Curve {keys [
         {arriveTangent=4.1785, interpMode=RCIM_Cubic, leaveTangent=4.1785, tangentMode=RCTM_User, time=0, value=0}
         {arriveTangent=0.9995, interpMode=RCIM_Cubic, leaveTangent=0.9995, tangentMode=RCTM_User, time=0.2223, value=0.6825}
         {arriveTangent=-0.0405, interpMode=RCIM_Cubic, leaveTangent=-0.0405, tangentMode=RCTM_User, time=0.7, value=0.9}
         <at 1 0.7>
        ]}}}
     Scale Curve=0.25
    )
    Index 2 Param 4=1
    Module.Param0WriteEnabled=true
    Module.Param0WriteEnabled1=true
    Module.Param0WriteEnabled2=false
    Module.Param1WriteEnabled=true
    Module.Param1WriteEnabled1=true
    Module.Param1WriteEnabled2=false
    Module.Param2WriteEnabled=true
    Module.Param2WriteEnabled1=true
    Module.Param2WriteEnabled2=false
    Module.Param3WriteEnabled=true
    Module.Param3WriteEnabled1=false
    Module.Param3WriteEnabled2=true
    Write Parameter Index 1=true
    Write Parameter Index 2=true
   ) {enabled=false}
   &DynamicMaterialParameters(
    Index 0 Param 1=&FloatFromCurve(FloatCurve=DataInterface() {propertyValues {Curve {keys [<at 0 1>,<at 1 0>]}}})
    Index 0 Param 2=&MakeFloatFromInt(Int=&Divide_Int(A:NiagaraInt32=&RandomRangeInt(Minimum=0, Maximum=2), B:NiagaraInt32=1))
    Index 0 Param 3=8
    Index 0 Param 4=$Particles.SubImageIndex
    Index 1 Param 1=&RandomRangeFloat(Minimum=0, Maximum=1)
    Index 1 Param 2=&RandomRangeFloat(Minimum=-1, Maximum=1)
    Index 1 Param 3=&FloatFromCurve(FloatCurve=DataInterface() {propertyValues {Curve {keys [
         {arriveTangent=4.1785, interpMode=RCIM_Cubic, leaveTangent=4.1785, tangentMode=RCTM_User, time=0, value=0}
         {arriveTangent=0.9995, interpMode=RCIM_Cubic, leaveTangent=0.9995, tangentMode=RCTM_User, time=0.2223, value=0.6825}
         {arriveTangent=-0.0405, interpMode=RCIM_Cubic, leaveTangent=-0.0405, tangentMode=RCTM_User, time=0.7, value=0.9}
         <at 1 0.7>
        ]}}})
    Index 2 Param 4=1
    Module.Param0WriteEnabled=true
    Module.Param0WriteEnabled1=true
    Module.Param0WriteEnabled2=false
    Module.Param1WriteEnabled=true
    Module.Param1WriteEnabled1=true
    Module.Param1WriteEnabled2=false
    Module.Param2WriteEnabled=true
    Module.Param2WriteEnabled1=true
    Module.Param2WriteEnabled2=false
    Module.Param3WriteEnabled=true
    Module.Param3WriteEnabled1=false
    Module.Param3WriteEnabled2=true
    Write Parameter Index 1=true
    Write Parameter Index 2=true
   ) {module=DynamicMaterialParameters002}
   &ScaleColor(
    Scale Alpha=&FloatFromCurve(
     FloatCurve=DataInterface() {propertyValues {Curve {keys [
         {arriveTangent=0.0516, interpMode=RCIM_Cubic, leaveTangent=0.0516, tangentMode=RCTM_User, time=0, value=0}
         {arriveTangent=-1.4181, interpMode=RCIM_Cubic, leaveTangent=-1.4181, tangentMode=RCTM_User, time=0.3, value=1.125}
         <at 1 0>
        ]}}}
     Scale Curve=&Lerp_Float(A=1, B=0, Alpha=$Particles.RegionMask.Float)
    )
    Scale RGB=&MakeVectorFromLinearColorRGB(Color=&MakeLinearColor(
      R:NiagaraFloat=1
      G:NiagaraFloat=1
      B:NiagaraFloat=&FloatFromCurve(
       FloatCurve=DataInterface() {propertyValues {Curve {keys [
           {arriveTangent=-4.2841, interpMode=RCIM_Cubic, leaveTangent=-4.2841, tangentMode=RCTM_User, time=0, value=1}
           {interpMode=RCIM_Cubic, tangentMode=RCTM_User, time=0.35, value=0.01}
          ]}}}
       Scale Curve=5
      )
      A:NiagaraFloat=0
     ))
    ScaleA=true
    ScaleRGB=true
   )
   /SetVariables(Particles.SpriteSize2=&Multiply_Vector2D(A=$Particles.SpriteSize, B=<1 0.35>)) {isSetParameters=true, module=SetVariables_#2}
   /SetVariables(Particles.SpriteRotation2:NiagaraFloat=&RandomRangeFloat(Minimum=0, Maximum=180)) {isSetParameters=true, module=SetVariables_#3}
   &ScaleSpriteSizeBySpeed(
    Max Scale Factor=<2 3>
    Min Scale Factor=<2 2>
    Sample Scale Factor By Curve=true
    Scale Factor Curve=DataInterface() {propertyValues {Curve {keys [{interpMode=RCIM_Cubic, tangentMode=RCTM_User, time=0.1, value=0}, {interpMode=RCIM_Cubic, tangentMode=RCTM_User, time=1, value=1}]}}}
   )
   &ScaleSpriteSize(Uniform Curve Sprite Scale=DataInterface() {propertyValues {Curve {keys [
        {arriveTangent=0.3375, interpMode=RCIM_Cubic, leaveTangent=0.3375, tangentMode=RCTM_User, time=0, value=1}
        {interpMode=RCIM_Cubic, tangentMode=RCTM_User, time=1, value=2}
       ]}}})
   &KillParticlesInVolume(
    Origin Offset:Vector3f=&MakeVector(
     X=0
     Y=&FloatFromCurve(FloatCurve=DataInterface() {propertyValues {Curve {keys [<at 0 0>,<at 0.5 1>]}}}, CurveIndex=$System.NormalizedLoopAge, Scale Curve=-15000)
     Z=0
    )
    Sphere Radius=6500
   )
  ]
  rendererClasses [&NiagaraSpriteRendererProperties]
  renderers [{rendererClass=&NiagaraSpriteRendererProperties}]
 }
}
summary {
 systemName=NS_NeonCity_CyberDragon_Breath
 emitters [
  {emitterName=After_Flame, simTarget=GPUComputeSim, rendererClasses [&NiagaraSpriteRendererProperties]}
  {emitterName=Flame_LIght, rendererClasses [&NiagaraLightRendererProperties]}
  {bEnabled=false, emitterName=Dragon_Glow, rendererClasses [&NiagaraMeshRendererProperties]}
  {emitterName=Dragon_Erode, rendererClasses [&NiagaraMeshRendererProperties]}
  {bEnabled=false, emitterName=Dragon_Embers, simTarget=GPUComputeSim, rendererClasses [&NiagaraSpriteRendererProperties]}
  {emitterName=PreFlame_Cone, rendererClasses [&NiagaraMeshRendererProperties, &NiagaraSpriteRendererProperties]}
  {emitterName=Start_Flame, simTarget=GPUComputeSim, rendererClasses [&NiagaraSpriteRendererProperties]}
  {emitterName=ErodeCubes, simTarget=GPUComputeSim, rendererClasses [&NiagaraMeshRendererProperties]}
  {emitterName=BlobFlameMesh, rendererClasses [&NiagaraMeshRendererProperties]}
  {emitterName=StartCubes, simTarget=GPUComputeSim, rendererClasses [&NiagaraMeshRendererProperties]}
  {bEnabled=false, emitterName=ErodeCubes001, simTarget=GPUComputeSim, rendererClasses [&NiagaraMeshRendererProperties]}
 ]
 userVariables [
  User.MainColor {
   type=NiagaraDataInterfaceColorCurve
   description=""
   defaultValue {
    struct=NiagaraExt_VariableValue_DataInterface
    value {
     dataInterface=&self:NiagaraDataInterfaceColorCurve_0
     dataInterfaceClass=&NiagaraDataInterfaceColorCurve
    }
   }
  }
  User.MainTimingCurve {
   type=NiagaraDataInterfaceCurve
   description=""
   defaultValue {
    struct=NiagaraExt_VariableValue_DataInterface
    value {
     dataInterface=&self:NiagaraDataInterfaceCurve_0
     dataInterfaceClass=&NiagaraDataInterfaceCurve
    }
   }
  }
 ]
}
