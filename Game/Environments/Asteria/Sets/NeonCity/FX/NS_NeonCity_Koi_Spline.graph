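# Niagara system for schools of tinted GPU mesh koi in the NeonCity set, with distance-scaled spawning and culling.
# The active FishTorus emitter uses a looping, randomly offset timeline to move fish around a torus while shifting them vertically.
# The disabled FishSpline emitter instead moves fish along User.Spline with waveform wander and a radius-based spread.
# Both hue-shift User.BaseColor, scale meshes by path progress and feed randomized, oscillating dynamic material parameters.
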NiagaraSystem /Game/Environments/Asteria/Sets/NeonCity/FX/NS_NeonCity_Koi_Spline
dependencies {
 usedDataInterfaces [&NiagaraDataInterfaceCurve]
 usedDynamicInputs [&Subtract_Float]
 usedModules [
  &self:GetSplineLengthSystem
  /Niagara/Modules/System/SystemLifeCycle
  &SystemState
  &self:SystemSpawnScript.SystemScriptSource.SystemScriptGraph.NiagaraNodeAssignment_0.SetVariables_#1
  &self:SystemSpawnScript.SystemScriptSource.SystemScriptGraph.NiagaraNodeAssignment_1.SetVariables_#2
 ]
}
emitters {
 FishSpline {
  bEnabled=false
  simTarget=GPUComputeSim
  emitterUpdateScript [
   &EmitterState(
    Enable Distance Culling=true
    MaxDistance=35000
    Module.bApplySpawnCountScaleByDistance=true
    Module.bUseMaxDistance=true
    Module.bUseMinDistance=false
    Reset Age On Awaken=true
    Scalability Mode=::NewEnumerator1
    Scale Spawn Count=true
    Spawn Count Scale By Distance=DataInterface() {propertyValues {Curve {keys [<at 0 1>,<at 0.85 1>,<at 1 0>]}}}
   )
   &SpawnRate(SpawnRate=50, Use Spawn Probability=true)
  ]
  particleSpawnScript [
   &InitializeParticle(
    Color=&HueShiftLinearColor(
     Hue=&RandomRangeFloat(Minimum=-0.02, Maximum=0.02)
     Saturation=3
     Value=&RandomRangeFloat(Minimum=0.08, Maximum=0.1)
     Linear Color=$User.BaseColor
    )
    Lifetime=15
    Mesh Scale Max=<0.3 0.2 0.2>
    Mesh Scale Min=<0.2 0.15 0.15>
    Mesh Scale Mode=::NewEnumerator2
    Sprite Rotation Mode=::NewEnumerator2
    UsePositionOffset=false
    Write Material Random=true
   ) {module=InitializeParticle001}
   /SetVariables() {isSetParameters=true}
   &ShapeLocation(
    Apply To Particle Position=::NewEnumerator1
    MaskSpawnGroup=false
    Offset Coordinate Space=::Local
    Offset Mode=::NewEnumerator0
    Override Seed=false
    Sphere Radius=$User.Radius
    UseOwnerScale=true
   )
   /SetVariables(
    Particles.SplineEnd:NiagaraFloat=&RandomRangeFloat(Minimum=0.5, Maximum=0.6)
    Particles.SplineStart:NiagaraFloat=&RandomRangeFloat(Minimum=0, Maximum=0.6)
   ) {isSetParameters=true}
   /SetVariables(Particles.SplineEnd:NiagaraFloat=&Add_Float(A=$Particles.SplineEnd, B=$Particles.SplineStart)) {isSetParameters=true}
  ]
  particleUpdateScript [
   &ParticleState()
   &self:Spline_Position_02(
    Position Along Spline:NiagaraFloat=&RemapRange(
     Input Value To Remap=&ModuloFloat(A=&Lerp_Float(A=$Particles.SplineStart, B=$Particles.SplineEnd, Alpha=$Particles.NormalizedAge), B=1)
     Input Min=0
     Input Max=1
     Output Min=1
     Output Max=0
    )
    Position Offset:Vector3f=&MakeVector(
     X=&Waveform(
      Global Amplitude Scale=5
      "Waveform [1]"=::NewEnumerator0
      "[1] Phase Scale"=&RandomRangeFloat(Minimum=-1, Maximum=1)
      "[1] Frequency"=&RandomRangeFloat(Minimum=0, Maximum=1)
      "[1] Amplitude Min/Max"=<-1 1>
      "[1] Scale"=0.2
     )
     Y=&Waveform(
      Global Amplitude Scale=250
      "Waveform [1]"=::NewEnumerator0
      "[1] Phase Scale"=&RandomRangeFloat(Minimum=-1, Maximum=1)
      "[1] Frequency"=0.1
      "[1] Amplitude Min/Max"=<-1 1>
      "[1] Scale"=0.5
     )
     Z=&Waveform(
      Global Amplitude Scale=50
      "Waveform [1]"=::NewEnumerator0
      "[1] Phase Scale"=0.2
      "[1] Frequency"=&RandomRangeFloat(Minimum=0, Maximum=1)
      "[1] Amplitude Min/Max"=<-1 1>
      "[1] Scale"=0.25
     )
    )
    Spline:NiagaraDataInterfaceSpline=$User.Spline
   ) {module=Spline_Position_02}
   /SetVariables(Particles.Position=&AddVectorToPosition(Position=$Output.Spline_Position_02.Position, Vector=$Particles.ShapeLocation.ShapeVector)) {isSetParameters=true}
   /SetVariables(Particles.TimingCOntrol:NiagaraFloat=&RemapRange(
     Input Value To Remap=&ModuloFloat(A=&Lerp_Float(A=$Particles.SplineStart, B=$Particles.SplineEnd, Alpha=$Particles.NormalizedAge), B=1)
     Input Min=0
     Input Max=1
     Output Min=1
     Output Max=0
    )) {isSetParameters=true}
   &ScaleMeshSize(Scale Factor=&VectorFromFloat(Value=&FloatFromCurve(
      FloatCurve=DataInterface() {propertyValues {Curve {keys [
          {interpMode=RCIM_Cubic, tangentMode=RCTM_User, time=0, value=0}
          {arriveTangent=3.445, leaveTangent=3.445, time=0.05, value=1}
          {interpMode=RCIM_Cubic, tangentMode=RCTM_User, time=0.5, value=1}
          {arriveTangent=-2.9691, interpMode=RCIM_Cubic, tangentMode=RCTM_User, time=0.9, value=0.75}
          {tangentMode=RCTM_User, time=1, value=0}
         ]}}}
      Scale Curve=&Multiply_Float(
       A=3
       B=&FloatFromCurve(
        FloatCurve=DataInterface() {propertyValues {Curve {keys [
            {interpMode=RCIM_Cubic, tangentMode=RCTM_User, time=0, value=0}
            {interpMode=RCIM_Cubic, tangentMode=RCTM_User, time=0.1, value=0.875}
            {interpMode=RCIM_Cubic, tangentMode=RCTM_User, time=0.5, value=1}
            {interpMode=RCIM_Cubic, tangentMode=RCTM_User, time=0.9, value=0.75}
            {tangentMode=RCTM_User, time=1, value=0}
           ]}}}
        CurveIndex=$Particles.TimingCOntrol
       )
      )
     )))
   &ScaleColor(
    Scale Alpha=&FloatFromCurve(FloatCurve=DataInterface() {propertyValues {Curve {keys [
         {arriveTangent=4.0343, time=0, value=1}
         {arriveTangent=-1, interpMode=RCIM_Cubic, leaveTangent=-1, time=0.9, value=1}
         {arriveTangent=-0.0405, interpMode=RCIM_Cubic, leaveTangent=-0.0405, tangentMode=RCTM_User, time=1, value=0}
        ]}}})
    Scale RGB=&Multiply_Vector(
     A=&MakeVectorFromLinearColorRGB(Color=$Particles.Initial.Color)
     B=&VectorFromFloat(Value=3)
    )
    ScaleA=true
    ScaleRGB=true
   )
   &DynamicMaterialParameters(
    Index 0 Param 1=&RandomRangeFloat(Minimum=0.7, Maximum=1)
    Index 0 Param 2=&RandomRangeFloat(Minimum=3, Maximum=1)
    Index 0 Param 3=&RandomRangeFloat(Minimum=0.5, Maximum=1)
    Index 0 Param 4=&Waveform(
     "Waveform [1]"=::NewEnumerator0
     "[1] Phase Scale"=&RandomRangeFloat(Minimum=0.8, Maximum=1)
     "[1] Frequency"=&RandomRangeFloat(Minimum=0.7, Maximum=1)
     "[1] Amplitude Min/Max"=<-0.02 0.02>
    )
    Index 1 Param 1=2
    Index 2 Param 1=100
    Index 2 Param 2=0.004
    Index 2 Param 3=2
    Index 2 Param 4=0.4
    Module.Param0WriteEnabled=true
    Module.Param0WriteEnabled1=true
    Module.Param0WriteEnabled2=true
    Module.Param1WriteEnabled=true
    Module.Param1WriteEnabled1=true
    Module.Param1WriteEnabled2=true
    Module.Param2WriteEnabled=true
    Module.Param2WriteEnabled1=true
    Module.Param2WriteEnabled2=true
    Module.Param3WriteEnabled=true
    Module.Param3WriteEnabled1=true
    Module.Param3WriteEnabled2=true
    Write Parameter Index 1=true
    Write Parameter Index 2=true
   ) {module=DynamicMaterialParameters001}
   &VortexForce(
    Origin Pull Amount=25000
    UseOriginOffset=false
    Vortex Axis=<1 0 0>
    Vortex Axis Coordinate Space:ENiagaraCoordinateSpace=::World
    Vortex Force Amount=25000
   ) {enabled=false}
   &SolveForcesAndVelocity(Clamp Velocity=true, Limit Acceleration=false, Speed Limit=100)
   &CalculateAccurateVelocity()
   &SolveForcesAndVelocity(Clamp Velocity=true, Limit Acceleration=false, Speed Limit=100) {module=SolveForcesAndVelocity001}
  ]
  rendererClasses [&NiagaraMeshRendererProperties]
  renderers [{rendererClass=&NiagaraMeshRendererProperties, rendererIndex=0}]
 }
 FishTorus {
  simTarget=GPUComputeSim
  emitterUpdateScript [
   &EmitterState(
    Enable Distance Culling=true
    MaxDistance=35000
    Module.bApplySpawnCountScaleByDistance=true
    Module.bUseMaxDistance=true
    Module.bUseMinDistance=false
    Reset Age On Awaken=true
    Scalability Mode=::NewEnumerator1
    Scale Spawn Count=true
    Spawn Count Scale By Distance=DataInterface() {propertyValues {Curve {keys [<at 0 1>,<at 0.85 1>,<at 1 0>]}}}
   )
   &SpawnRate(SpawnRate=50, Use Spawn Probability=true)
  ]
  particleSpawnScript [&InitializeParticle(
    Color=&HueShiftLinearColor(
     Hue=&RandomRangeFloat(Minimum=-0.02, Maximum=0.02)
     Saturation=3
     Value=&RandomRangeFloat(Minimum=0.08, Maximum=0.1)
     Linear Color=$User.BaseColor
    )
    Lifetime=20
    Mesh Scale Max=<0.3 0.2 0.2>
    Mesh Scale Min=<0.2 0.15 0.15>
    Mesh Scale Mode=::NewEnumerator2
    Sprite Rotation Mode=::NewEnumerator2
    UsePositionOffset=false
    Write Material Random=true
   ) {module=InitializeParticle001}]
  particleUpdateScript [
   &ParticleState()
   &Timeline(Loop While Playing=true, Play Rate=&RandomRangeFloat(Minimum=0.05, Maximum=0.05))
   /SetVariables(Particles.MinPos:NiagaraFloat=&RandomRangeFloat(Minimum=0, Maximum=1)) {isSetParameters=true}
   /SetVariables(Particles.TimingControl:NiagaraFloat=&FloatFromCurve(
     FloatCurve=DataInterface() {propertyValues {Curve {keys [<at 0 0>,<at 1 1.5>]}}}
     CurveIndex=&Add_Float(A=$Particles.Timeline.NormalizedTime, B=$Particles.MinPos)
    )) {isSetParameters=true}
   &ScaleMeshSize(Scale Factor=&VectorFromFloat(Value=&FloatFromCurve(
      FloatCurve=DataInterface() {propertyValues {Curve {keys [
          {interpMode=RCIM_Cubic, tangentMode=RCTM_User, time=0, value=0}
          {arriveTangent=3.445, leaveTangent=3.445, time=0.05, value=1}
          {interpMode=RCIM_Cubic, tangentMode=RCTM_User, time=0.5, value=1}
          {arriveTangent=-2.9691, interpMode=RCIM_Cubic, tangentMode=RCTM_User, time=0.9, value=0.75}
          {tangentMode=RCTM_User, time=1, value=0}
         ]}}}
      Scale Curve=&Multiply_Float(
       A=3
       B=&FloatFromCurve(
        FloatCurve=DataInterface() {propertyValues {Curve {keys [
            {interpMode=RCIM_Cubic, tangentMode=RCTM_User, time=0, value=0}
            {interpMode=RCIM_Cubic, tangentMode=RCTM_User, time=0.1, value=0.875}
            {interpMode=RCIM_Cubic, tangentMode=RCTM_User, time=0.5, value=1}
            {interpMode=RCIM_Cubic, tangentMode=RCTM_User, time=0.9, value=0.75}
            {tangentMode=RCTM_User, time=1, value=0}
           ]}}}
        CurveIndex=&Divide_Float(A=$Particles.TimingControl, B=1.5)
       )
      )
     )))
   &ScaleColor(
    Scale Alpha=&FloatFromCurve(FloatCurve=DataInterface() {propertyValues {Curve {keys [
         {arriveTangent=4.0343, time=0, value=1}
         {arriveTangent=-1, interpMode=RCIM_Cubic, leaveTangent=-1, time=0.9, value=1}
         {arriveTangent=-0.0405, interpMode=RCIM_Cubic, leaveTangent=-0.0405, tangentMode=RCTM_User, time=1, value=0}
        ]}}})
    Scale RGB=&Multiply_Vector(
     A=&MakeVectorFromLinearColorRGB(Color=$Particles.Initial.Color)
     B=&VectorFromFloat(Value=3)
    )
    ScaleA=true
    ScaleRGB=true
   )
   &DynamicMaterialParameters(
    Index 0 Param 1=&RandomRangeFloat(Minimum=0.7, Maximum=1)
    Index 0 Param 2=&RandomRangeFloat(Minimum=3, Maximum=1)
    Index 0 Param 3=&RandomRangeFloat(Minimum=0.5, Maximum=1)
    Index 0 Param 4=&Waveform(
     "Waveform [1]"=::NewEnumerator0
     "[1] Phase Scale"=&RandomRangeFloat(Minimum=0.8, Maximum=1)
     "[1] Frequency"=&RandomRangeFloat(Minimum=0.7, Maximum=1)
     "[1] Amplitude Min/Max"=<-0.02 0.02>
    )
    Index 1 Param 1=1
    Index 2 Param 1=100
    Index 2 Param 2=0.004
    Index 2 Param 3=2
    Index 2 Param 4=0.4
    Module.Param0WriteEnabled=true
    Module.Param0WriteEnabled1=true
    Module.Param0WriteEnabled2=true
    Module.Param1WriteEnabled=true
    Module.Param1WriteEnabled1=true
    Module.Param1WriteEnabled2=true
    Module.Param2WriteEnabled=true
    Module.Param2WriteEnabled1=true
    Module.Param2WriteEnabled2=true
    Module.Param3WriteEnabled=true
    Module.Param3WriteEnabled1=true
    Module.Param3WriteEnabled2=true
    Write Parameter Index 1=true
    Write Parameter Index 2=true
   ) {module=DynamicMaterialParameters001}
   &ShapeLocation(
    Handle Radius:NiagaraFloat=&RandomRangeFloat(Minimum=0, Maximum=500)
    Large Radius=2500
    MaskSpawnGroup=false
    Offset=&Lerp_Vector(A=<0 0 -1000>, B=<0 0 2000>, Alpha=$Particles.TimingControl)
    Offset Mode=::NewEnumerator0
    Shape Primitive=::NewEnumerator3
    Torus Distribution Mode:ENiagaraTorusDistributionMode=::NewEnumerator1
    U Position=$Particles.TimingControl
    UseOwnerScale=true
    V Position=&RandomRangeFloat(Minimum=0, Maximum=1)
   ) {module=ShapeLocation001}
   &SolveForcesAndVelocity(Clamp Velocity=true, Limit Acceleration=false, Speed Limit=100)
   &CalculateAccurateVelocity()
  ]
  rendererClasses [&NiagaraMeshRendererProperties]
  renderers [{rendererClass=&NiagaraMeshRendererProperties, rendererIndex=0}]
 }
}
summary {
 systemName=NS_NeonCity_Koi_Spline
 emitters [
  {bEnabled=false, emitterName=FishSpline, simTarget=GPUComputeSim, rendererClasses [&NiagaraMeshRendererProperties]}
  {emitterName=FishTorus, simTarget=GPUComputeSim, rendererClasses [&NiagaraMeshRendererProperties]}
 ]
 userVariables [
  User.BaseColor {type=LinearColor, description="", defaultValue {struct=LinearColor, value=<rgba 0.1169 0 1 1>}}
  User.End Position {type=NiagaraFloat, defaultValue=1, description=""}
  User.Radius {type=NiagaraFloat, defaultValue=500, description=""}
  User.Spline {
   type=NiagaraDataInterfaceSpline
   description=""
   defaultValue {
    struct=NiagaraExt_VariableValue_DataInterface
    value {
     dataInterface=&self:NiagaraDataInterfaceSpline_0
     dataInterfaceClass=/Script/Niagara.NiagaraDataInterfaceSpline
    }
   }
  }
  User.Start Position {type=NiagaraFloat, defaultValue=0, description=""}
 ]
}
