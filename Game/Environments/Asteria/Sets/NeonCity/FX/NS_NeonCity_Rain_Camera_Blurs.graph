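# Niagara system for NeonCity rain camera blurs: one GPU sprite emitter spawning small, short-lived particulates around the system.
# Spawning applies slight downward velocity and a GBuffer color module.
# Many hard-coded world-space kill boxes, a region mask and a texture-based rain mask over the set cull particles, tying it to this level's layout.
# Updates repeat one kill box, apply drag, shrink sprites at end of life, dim color and flicker alpha through a waveform.
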NiagaraSystem /Game/Environments/Asteria/Sets/NeonCity/FX/NS_NeonCity_Rain_Camera_Blurs
dependencies {
 usedDataInterfaces [
  &NiagaraDataInterfaceCamera
  &NiagaraDataInterfaceGBuffer
  /Script/Niagara.NiagaraDataInterfaceTexture
  &NiagaraDataInterfaceCurve
 ]
 usedDynamicInputs []
 usedModules [&SystemState]
}
emitters {HangingParticulates {
  simTarget=GPUComputeSim
  emitterUpdateScript [
   &EmitterState(Life Cycle Mode=::NewEnumerator1, Loop Duration=2, UseLoopDelay=false)
   &SpawnRate(SpawnRate=150, Use Spawn Probability=false)
   /SetVariables() {enabled=false, isSetParameters=true}
  ]
  particleSpawnScript [
   &InitializeParticle(
    Color=<rgba 0.5 0.5 0.5 1>
    Lifetime Max=3
    Lifetime Mode=::NewEnumerator1
    Sprite Rotation Mode=::NewEnumerator2
    Sprite Size Mode=::NewEnumerator1
    Sprite UV Mode=::NewEnumerator3
    Uniform Sprite Size Max=8
    Uniform Sprite Size Min=3
    UsePositionOffset=false
    Write Material Random=true
   )
   &ShapeLocation(
    Box Midpoint=&RandomRangeVector(Minimum=<0 0 0>, Maximum=<1 1 1>)
    Box Size=<0 100 100>
    MaskSpawnGroup=false
    Override Seed=false
    Shape Primitive=::NewEnumerator2
    Surface Expansion Mode:ENiagaraExpansionMode=::NewEnumerator2
    Surface Only=true
    Surface Only Band Thickness=1
    UseOwnerScale=true
   ) {enabled=false}
   &ShapeLocation(
    Large Radius=50
    MaskSpawnGroup=false
    Non Uniform Scale=<1 2 1>
    Override Seed=false
    Rotation Mode=::NewEnumerator1
    Shape Primitive=::NewEnumerator3
    UseOwnerScale=true
   ) {module=ShapeLocation001}
   &AddVelocity(Velocity=&RandomRangeVector(Minimum=<0 0 0>, Maximum=<0 0 -1>))
   &ApplyInitialForces()
   /Game/Effects/Niagara/Modules/Spawn/FN_Spawn_GBuffer_Color()
   &RegionMask(UseOffset=false, UseRotation=false)
   &KillParticlesInVolume(
    Box Size:Vector3f=<3350 3825 675>
    Kill Shape=::NewEnumerator1
    Volume Origin:Vector3f=&AddVectorToPosition(Position=<-60714 21466 718>, Vector=<0 0 0>)
   )
   &KillParticlesInVolume(
    Box Size:Vector3f=<3350 2725 2625>
    Kill Shape=::NewEnumerator1
    Volume Origin:Vector3f=&AddVectorToPosition(Position=<-52356 19960 958>, Vector=<0 0 0>)
   ) {module=KillParticlesInVolume002}
   &KillParticlesInVolume(
    Box Size:Vector3f=<3350 2725 7600>
    Kill Shape=::NewEnumerator1
    Volume Origin:Vector3f=&AddVectorToPosition(Position=<-46788 16872 2455>, Vector=<0 0 0>)
   ) {enabled=false, module=KillParticlesInVolume003}
   &KillParticlesInVolume(
    Box Size:Vector3f=<3350 3000 2500>
    Kill Shape=::NewEnumerator1
    Volume Origin:Vector3f=&AddVectorToPosition(Position=<-60727 26090 612>, Vector=<0 0 0>)
   ) {module=KillParticlesInVolume004}
   &KillParticlesInVolume(
    Box Size:Vector3f=<3325 3000 6875>
    Kill Shape=::NewEnumerator1
    Volume Origin:Vector3f=&AddVectorToPosition(Position=<-55993 24781 1422>, Vector=<0 0 0>)
   ) {module=KillParticlesInVolume005}
   &KillParticlesInVolume(
    Box Size:Vector3f=<2050 2975 3125>
    Kill Shape=::NewEnumerator1
    Volume Origin:Vector3f=&AddVectorToPosition(Position=<-50898 23839 929>, Vector=<0 0 0>)
   ) {module=KillParticlesInVolume006}
   &KillParticlesInVolume(
    Box Size:Vector3f=<2050 2050 2625>
    Kill Shape=::NewEnumerator1
    Volume Origin:Vector3f=&AddVectorToPosition(Position=<-47631 21734 935>, Vector=<0 0 0>)
   ) {enabled=false, module=KillParticlesInVolume007}
   &KillParticlesInVolume(
    Box Size:Vector3f=<2350 2500 2625>
    Kill Shape=::NewEnumerator1
    Volume Origin:Vector3f=&AddVectorToPosition(Position=<-47604 24485 935>, Vector=<0 0 0>)
   ) {module=KillParticlesInVolume008}
   &KillParticlesInVolume(
    Box Size:Vector3f=<2350 2500 8000>
    Kill Shape=::NewEnumerator1
    Volume Origin:Vector3f=&AddVectorToPosition(Position=<-43503 24008 1359>, Vector=<0 0 0>)
   ) {module=KillParticlesInVolume009}
   &KillParticlesInVolume(
    Box Size:Vector3f=<2350 2475 2500>
    Kill Shape=::NewEnumerator1
    Volume Origin:Vector3f=&AddVectorToPosition(Position=<-55221 30183 1012>, Vector=<0 0 0>)
   ) {module=KillParticlesInVolume010}
   &KillParticlesInVolume(
    Box Size:Vector3f=<4925 4275 15000>
    Kill Shape=::NewEnumerator1
    Volume Origin:Vector3f=&AddVectorToPosition(Position=<-49721 29829 2402>, Vector=<0 0 0>)
   ) {module=KillParticlesInVolume011}
   &KillParticlesInVolume(
    Box Size:Vector3f=<2925 2475 10000>
    Kill Shape=::NewEnumerator1
    Volume Origin:Vector3f=&AddVectorToPosition(Position=<-42923 29166 1729>, Vector=<0 0 0>)
   ) {module=KillParticlesInVolume012}
   &KillParticlesInVolume(
    Box Size:Vector3f=<2925 3950 10000>
    Kill Shape=::NewEnumerator1
    Volume Origin:Vector3f=&AddVectorToPosition(Position=<-56620 38200 1729>, Vector=<0 0 0>)
   ) {module=KillParticlesInVolume013}
   &KillParticlesInVolume(
    Box Size:Vector3f=<2925 3950 10000>
    Kill Shape=::NewEnumerator1
    Volume Origin:Vector3f=&AddVectorToPosition(Position=<-51047 39987 2644>, Vector=<0 0 0>)
   ) {module=KillParticlesInVolume014}
   &KillParticlesInVolume(
    Box Size:Vector3f=<2300 2500 10000>
    Kill Shape=::NewEnumerator1
    Volume Origin:Vector3f=&AddVectorToPosition(Position=<-46708 40471 760>, Vector=<0 0 0>)
   ) {module=KillParticlesInVolume015}
   &KillParticlesInVolume(
    Box Size:Vector3f=<1100 1950 850>
    Kill Shape=::NewEnumerator1
    Volume Origin:Vector3f=&AddVectorToPosition(Position=<-47372 38200 -100>, Vector=<0 0 0>)
   ) {module=KillParticlesInVolume001}
   &self:RainTextureMasking(
    AlphaThreshold=0.5
    BoxCenter:NiagaraPosition=<-50828 29472 9916>
    BoxExtents:Vector3f=<32500 32500 1>
    Position:NiagaraPosition=$Particles.Position
    RejectionTexture=DataInterface() {propertyValues {
      Texture=/Game/Environments/Asteria/Sets/NeonCity/FX/Textures/T_NeonCity_Rain_Mask_01
      TextureUserParameter {parameter {name=None, typeDefHandle {registeredTypeIndex=2}, varData []}}
     }}
   ) {module=RainTextureMasking}
  ]
  particleUpdateScript [
   &KillParticlesInVolume(
    Box Size:Vector3f=<1100 1950 850>
    Kill Shape=::NewEnumerator1
    Volume Origin:Vector3f=&AddVectorToPosition(Position=<-47372 38200 -100>, Vector=<0 0 0>)
   ) {module=KillParticlesInVolume016}
   &ParticleState()
   &ScaleSpriteSize(Uniform Curve Sprite Scale=DataInterface() {propertyValues {Curve {keys [
        {arriveTangent=9.0725, time=0, value=1}
        {interpMode=RCIM_Cubic, tangentMode=RCTM_User, time=0.95, value=1}
        {arriveTangent=-0.0405, interpMode=RCIM_Cubic, leaveTangent=-0.0405, tangentMode=RCTM_User, time=1, value=0}
       ]}}})
   &ScaleColor(
    Scale Alpha=&Multiply_Float(
     A=&FloatFromCurve(FloatCurve=DataInterface() {propertyValues {Curve {keys [
          {arriveTangent=2.4871, time=0, value=0.5}
          <at 0.25 1>
          {interpMode=RCIM_Cubic, tangentMode=RCTM_User, time=0.8, value=0.95}
          {arriveTangent=-0.0405, interpMode=RCIM_Cubic, leaveTangent=-0.0405, tangentMode=RCTM_User, time=1, value=0}
         ]}}})
     B=&Waveform(
      Global Phase=$Particles.Age
      "Waveform [1]"=::NewEnumerator8
      "[1] Frequency"=&RandomRangeFloat(Minimum=50, Maximum=150)
      "[1] Amplitude Min/Max"=<0.8 1>
     )
    )
    Scale RGB=&VectorFromFloat(Value=&FloatFromCurve(
      FloatCurve=DataInterface() {propertyValues {Curve {keys [{tangentMode=RCTM_User, time=0, value=1}, {tangentMode=RCTM_User, time=1, value=0.25}]}}}
      Scale Curve=25
     ))
    ScaleA=true
    ScaleRGB=true
   )
   &Drag(Use Linear Drag=true, Use Rotational Drag=true)
   &SolveForcesAndVelocity(Clamp Velocity=false, Limit Acceleration=false)
   &DynamicMaterialParameters(
    Index 0 Param 1=0.1
    Index 0 Param 2=0.95
    Index 0 Param 3=1
    Module.Param0WriteEnabled=true
    Module.Param1WriteEnabled=true
    Module.Param2WriteEnabled=true
    Module.Param3WriteEnabled=true
   )
  ]
  rendererClasses [&NiagaraSpriteRendererProperties]
  renderers [{rendererClass=&NiagaraSpriteRendererProperties}]
 }}
summary {
 systemName=NS_NeonCity_Rain_Camera_Blurs
 emitters [{emitterName=HangingParticulates, simTarget=GPUComputeSim, rendererClasses [&NiagaraSpriteRendererProperties]}]
 userVariables []
}
