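# Material function that derives tangent-space normals from each RGBA channel of a grayscale mask texture, apparently to bevel sci-fi road mesh detail.
# Per channel, four taps offset by texel size are inverted, clamped and differenced on X and Y, a central-difference gradient rather than a full Sobel kernel.
# Scaled gradients output as RG masks; normal outputs limit their length below one, apply a multiplier and derive Z.
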MaterialFunction /Game/Environments/Asteria/Roads/SciFi/Meshes/Pieces_Update_v2/MF_SobelGrayscale2Normal
graph {nodes [
  n1 = NamedRerouteUsage()
  n2 = NamedRerouteUsage()
  n3 = Add(A=@n1, B=Multiply(A=Constant2Vector() {r=-1}, B=@n2))
  n4 = NamedRerouteUsage()
  n5 = NamedRerouteUsage()
  n6 = Clamp(
   None=OneMinus(None=TextureSample(UVs=@n3, Tex=@n4) {samplerType=SAMPLERTYPE_Masks, texture=/Game/Environments/Asteria/Roads/SciFi/Textures/CityRoads_MaskMap1}.R)
   Max=@n5
  )
  n7 = Add(A=@n1, B=Multiply(A=Constant2Vector() {r=1}, B=@n2))
  n8 = Clamp(
   None=OneMinus(None=TextureSample(UVs=@n7, Tex=@n4) {samplerType=SAMPLERTYPE_Masks, texture=&CityRoads_MaskMap1}.R)
   Max=@n5
  )
  n9 = Add(A=@n1, B=Multiply(A=Constant2Vector() {g=1}, B=@n2))
  n10 = Clamp(
   None=OneMinus(None=TextureSample(UVs=@n9, Tex=@n4) {samplerType=SAMPLERTYPE_Masks, texture=&CityRoads_MaskMap1}.R)
   Max=@n5
  )
  n11 = Add(A=@n1, B=Multiply(A=Constant2Vector() {g=-1}, B=@n2))
  n12 = Clamp(
   None=OneMinus(None=TextureSample(UVs=@n11, Tex=@n4) {samplerType=SAMPLERTYPE_Masks, texture=&CityRoads_MaskMap1}.R)
   Max=@n5
  )
  n13 = Multiply(A=AppendVector(A=Subtract(A=@n6, B=@n8), B=Subtract(A=@n10, B=@n12))) {constB=3}
  n14 = DotProduct(A=@n13, B=@n13)
  n15 = Multiply(A=Divide(A=@n13, B=SquareRoot(None=@n14))) {constB=0.999}
  n16 = If(A=@n14, B=Constant() {r=1}, "A > B"=@n15, "A == B"=@n15, "A < B"=@n13)
  n17 = NamedRerouteUsage()
  FunctionOutput(None=DeriveNormalZ(InXY=Multiply(A=@n16, B=@n17))) {outputName=Result - R, sortPriority=1}
  FunctionOutput(None=@n13) {outputName=R - RG Mask, sortPriority=2}
  n18 = NamedRerouteUsage()
  n19 = NamedRerouteUsage()
  n20 = Add(A=@n18, B=Multiply(A=Constant2Vector() {r=-1}, B=@n19))
  n21 = NamedRerouteUsage()
  n22 = NamedRerouteUsage()
  n23 = Clamp(
   None=OneMinus(None=TextureSample(UVs=@n20, Tex=@n21) {samplerType=SAMPLERTYPE_Masks, texture=&CityRoads_MaskMap1}.G)
   Max=@n22
  )
  n24 = Add(A=@n18, B=Multiply(A=Constant2Vector() {r=1}, B=@n19))
  n25 = Clamp(
   None=OneMinus(None=TextureSample(UVs=@n24, Tex=@n21) {samplerType=SAMPLERTYPE_Masks, texture=&CityRoads_MaskMap1}.G)
   Max=@n22
  )
  n26 = Add(A=@n18, B=Multiply(A=Constant2Vector() {g=1}, B=@n19))
  n27 = Clamp(
   None=OneMinus(None=TextureSample(UVs=@n26, Tex=@n21) {samplerType=SAMPLERTYPE_Masks, texture=&CityRoads_MaskMap1}.G)
   Max=@n22
  )
  n28 = Add(A=@n18, B=Multiply(A=Constant2Vector() {g=-1}, B=@n19))
  n29 = Clamp(
   None=OneMinus(None=TextureSample(UVs=@n28, Tex=@n21) {samplerType=SAMPLERTYPE_Masks, texture=&CityRoads_MaskMap1}.G)
   Max=@n22
  )
  n30 = Multiply(A=AppendVector(A=Subtract(A=@n23, B=@n25), B=Subtract(A=@n27, B=@n29))) {constB=3}
  FunctionOutput(None=@n30) {outputName=G - RG Mask, sortPriority=4}
  n31 = NamedRerouteUsage()
  n32 = NamedRerouteUsage()
  n33 = Add(A=@n31, B=Multiply(A=Constant2Vector() {r=-1}, B=@n32))
  n34 = NamedRerouteUsage()
  n35 = NamedRerouteUsage()
  n36 = Clamp(
   None=OneMinus(None=TextureSample(UVs=@n33, Tex=@n34) {samplerType=SAMPLERTYPE_Masks, texture=&CityRoads_MaskMap1}.B)
   Max=@n35
  )
  n37 = Add(A=@n31, B=Multiply(A=Constant2Vector() {r=1}, B=@n32))
  n38 = Clamp(
   None=OneMinus(None=TextureSample(UVs=@n37, Tex=@n34) {samplerType=SAMPLERTYPE_Masks, texture=&CityRoads_MaskMap1}.B)
   Max=@n35
  )
  n39 = Add(A=@n31, B=Multiply(A=Constant2Vector() {g=1}, B=@n32))
  n40 = Clamp(
   None=OneMinus(None=TextureSample(UVs=@n39, Tex=@n34) {samplerType=SAMPLERTYPE_Masks, texture=&CityRoads_MaskMap1}.B)
   Max=@n35
  )
  n41 = Add(A=@n31, B=Multiply(A=Constant2Vector() {g=-1}, B=@n32))
  n42 = Clamp(
   None=OneMinus(None=TextureSample(UVs=@n41, Tex=@n34) {samplerType=SAMPLERTYPE_Masks, texture=&CityRoads_MaskMap1}.B)
   Max=@n35
  )
  n43 = Multiply(A=AppendVector(A=Subtract(A=@n36, B=@n38), B=Subtract(A=@n40, B=@n42))) {constB=3}
  FunctionOutput(None=@n43) {outputName=B - RG Mask, sortPriority=6}
  n44 = NamedRerouteUsage()
  n45 = NamedRerouteUsage()
  n46 = Add(A=@n44, B=Multiply(A=Constant2Vector() {r=-1}, B=@n45))
  n47 = NamedRerouteUsage()
  n48 = NamedRerouteUsage()
  n49 = Clamp(
   None=OneMinus(None=TextureSample(UVs=@n46, Tex=@n47) {samplerType=SAMPLERTYPE_Masks, texture=&CityRoads_MaskMap1}.A)
   Max=@n48
  )
  n50 = Add(A=@n44, B=Multiply(A=Constant2Vector() {r=1}, B=@n45))
  n51 = Clamp(
   None=OneMinus(None=TextureSample(UVs=@n50, Tex=@n47) {samplerType=SAMPLERTYPE_Masks, texture=&CityRoads_MaskMap1}.A)
   Max=@n48
  )
  n52 = Add(A=@n44, B=Multiply(A=Constant2Vector() {g=1}, B=@n45))
  n53 = Clamp(
   None=OneMinus(None=TextureSample(UVs=@n52, Tex=@n47) {samplerType=SAMPLERTYPE_Masks, texture=&CityRoads_MaskMap1}.A)
   Max=@n48
  )
  n54 = Add(A=@n44, B=Multiply(A=Constant2Vector() {g=-1}, B=@n45))
  n55 = Clamp(
   None=OneMinus(None=TextureSample(UVs=@n54, Tex=@n47) {samplerType=SAMPLERTYPE_Masks, texture=&CityRoads_MaskMap1}.A)
   Max=@n48
  )
  n56 = Multiply(A=AppendVector(A=Subtract(A=@n49, B=@n51), B=Subtract(A=@n53, B=@n55))) {constB=3}
  FunctionOutput(None=@n56) {outputName=A - RG Mask, sortPriority=8}
  n57 = DotProduct(A=@n30, B=@n30)
  n58 = Multiply(A=Divide(A=@n30, B=SquareRoot(None=@n57))) {constB=0.999}
  n59 = If(A=@n57, B=Constant() {r=1}, "A > B"=@n58, "A == B"=@n58, "A < B"=@n30)
  n60 = NamedRerouteUsage()
  FunctionOutput(None=DeriveNormalZ(InXY=Multiply(A=@n59, B=@n60))) {outputName=Result - G, sortPriority=3}
  n61 = DotProduct(A=@n43, B=@n43)
  n62 = Multiply(A=Divide(A=@n43, B=SquareRoot(None=@n61))) {constB=0.999}
  n63 = If(A=@n61, B=Constant() {r=1}, "A > B"=@n62, "A == B"=@n62, "A < B"=@n43)
  n64 = NamedRerouteUsage()
  FunctionOutput(None=DeriveNormalZ(InXY=Multiply(A=@n63, B=@n64))) {outputName=Result - B, sortPriority=5}
  n65 = DotProduct(A=@n56, B=@n56)
  n66 = Multiply(A=Divide(A=@n56, B=SquareRoot(None=@n65))) {constB=0.999}
  n67 = If(A=@n65, B=Constant() {r=1}, "A > B"=@n66, "A == B"=@n66, "A < B"=@n56)
  n68 = NamedRerouteUsage()
  FunctionOutput(None=DeriveNormalZ(InXY=Multiply(A=@n67, B=@n68))) {outputName=Result - A, sortPriority=7}
  NamedRerouteDeclaration(None=FunctionInput() {inputName=UVs, inputType=FunctionInput_Vector2}) {name=UVs, nodeColor=<rgba 1 0.287 0 1>}
  NamedRerouteDeclaration(None=FunctionInput() {inputName=TexelSize, inputType=FunctionInput_Vector2}) {name=TexelSize, nodeColor=<rgba 1 0.0504 0.16 1>}
  NamedRerouteDeclaration(None=FunctionInput(Preview=TextureObject()) {inputName=Texture, inputType=FunctionInput_Texture2D}) {name=Texture, nodeColor=<rgba 0 1 0.1562 1>}
  NamedRerouteDeclaration(None=FunctionInput() {inputName=Bevel Strength, inputType=FunctionInput_Scalar}) {name=Bevel Strength, nodeColor=<rgba 0 1 0.7422 1>}
  NamedRerouteDeclaration(None=FunctionInput() {inputName=Normal Intensity, inputType=FunctionInput_Scalar}) {name=Normal Intensity, nodeColor=<rgba 0.5812 0.1822 1 1>}
 ]}
