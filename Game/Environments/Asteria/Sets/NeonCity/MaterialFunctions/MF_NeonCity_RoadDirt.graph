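# Material function for NeonCity roads that adds a dirt tint to incoming material attributes, toggled by the UseRoadDirt static switch.
# DirtMask comes from the green channel of a road mask texture, sampled with rotated world-XY UVs computed in the vertex interpolator, then shaped by exponent and scale.
# The clamped mask, reduced by Pebble_Mask, blends base color toward DirtColor.
# Roughness and specular math, which also uses Puddle_Mask, is present but unconnected to the output.
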MaterialFunction /Game/Environments/Asteria/Sets/NeonCity/MaterialFunctions/MF_NeonCity_RoadDirt
graph {nodes [
  n1 = FunctionInput() {inputType=FunctionInput_MaterialAttributes}
  n2 = GetMaterialAttributes(None=@n1) {attributeGetTypes [#1, #2, #3]}
  n3 = NamedRerouteUsage()
  n4 = Power(Base=FunctionInput() {inputName=Pebble_Mask, inputType=FunctionInput_Scalar, sortPriority=5}) {constExponent=0.5}
  n5 = Subtract(A=Clamp(None=@n3) {maxDefault=0.25}, B=@n4)
  n6 = ScalarParameter() {defaultValue=0.5, group=Dirt, parameterName=Dirt Color Intensity}
  n7 = LinearInterpolate(
   A=@n2.BaseColor
   B=VectorParameter() {group=Dirt, parameterName=DirtColor, defaultValue=<rgba 0.1276 0.1133 0.0561 1>}.RGB
   Alpha=Saturate(None=Multiply(A=@n5, B=@n6))
  )
  FunctionOutput(None=StaticSwitchParameter(
    True=SetMaterialAttributes!(MaterialAttributes=@n2.MaterialAttributes, BaseColor=@n7, Specular=@n2.MaterialAttributes, Roughness=@n2.MaterialAttributes) {attributeSetTypes [#1, #2, #3]}
    False=@n1
   ) {parameterName=UseRoadDirt})
  n8 = FunctionInput() {inputName=Puddle_Mask, inputType=FunctionInput_Scalar, sortPriority=15}
  LinearInterpolate(
   A=@n2.Roughness
   B=Subtract(A=ScalarParameter() {defaultValue=0.5, group=Dirt, parameterName=Dirt Roughness}, B=@n4)
   Alpha=Subtract(A=@n3, B=@n8)
  )
  n9 = Divide(
   A=ComponentMask(None=WorldPosition().XYZ) {g=true, r=true}
   B=ScalarParameter() {defaultValue=2000, group=Dirt, parameterName=WorldTileDivisor_Dirt}
  )
  n10 = ScalarParameter() {defaultValue=2, group=Dirt, parameterName=Dirt_Rotation}
  n11 = TextureSample(UVs=VertexInterpolator(VS=Rotator(Coordinate=@n9, Time=@n10)).PS) {samplerSource=SSM_Wrap_WorldGroupSettings, texture=/Game/Environments/Asteria/Roads/SciFi/Textures/CityRoads_MaskMap2}
  n12 = ScalarParameter() {defaultValue=1, group=Dirt, parameterName=DirtExponent}
  NamedRerouteDeclaration(None=Multiply(A=Power(Base=@n11.G, Exp=@n12), B=ScalarParameter() {defaultValue=3, group=Dirt, parameterName=DirtScale})) {name=DirtMask, nodeColor=<rgba 0.2219 0.1024 0 1>}
  n13 = Multiply(A=@n3, B=ScalarParameter() {defaultValue=0.1, group=Dirt, parameterName=Dirt Specular Intensity}) {constB=0.1}
  Subtract(A=@n2.Specular, B=Subtract(A=Subtract(A=@n13, B=@n8), B=@n4))
 ]}
