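# Material function for NeonCity roads that overlays painted markings onto incoming material attributes, bypassable through a static switch.
# A custom primitive data scalar selects no marking or one RGBA channel of a packed mask texture, allowing per-primitive variation.
# That channel is the blend alpha for parameterized color, specular and roughness.
# It also feeds NormalFromHeightmap, whose XY is added to the Normal input to give markings relief.
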MaterialFunction /Game/Environments/Asteria/Sets/NeonCity/MaterialFunctions/MF_NeonCity_RoadMarkings
graph {expressions {
  MaterialExpressionAdd_0 {
   class=MaterialExpressionAdd
   inputs [
    A {from=MaterialExpressionFunctionInput_1}
    B {from=MaterialExpressionMultiply_0}
   ]
  }
  MaterialExpressionBlendMaterialAttributes_0 {
   class=MaterialExpressionBlendMaterialAttributes
   inputs [
    A {from=MaterialExpressionFunctionInput_0, output=""}
    B {from=MaterialExpressionSetMaterialAttributes_0, output=""}
    Alpha {from=MaterialExpressionNamedRerouteUsage_1, output=""}
   ]
   properties {vertexAttributeBlendType=UseA}
  }
  MaterialExpressionConstant3Vector_0 {class=MaterialExpressionConstant3Vector, properties {constant=<rgba 0 0 1 0>}}
  MaterialExpressionConstant3Vector_1 {class=MaterialExpressionConstant3Vector, properties {constant=<rgba 1 1 0 0>}}
  MaterialExpressionConstant4Vector_0 {class=MaterialExpressionConstant4Vector, inputs [], properties {constant=<rgba 1 0 0 0>}}
  MaterialExpressionConstant4Vector_1 {class=MaterialExpressionConstant4Vector, inputs [], properties {constant=<rgba 0 1 0 0>}}
  MaterialExpressionConstant4Vector_2 {class=MaterialExpressionConstant4Vector, inputs [], properties {constant=<rgba 0 0 1 0>}}
  MaterialExpressionConstant4Vector_3 {class=MaterialExpressionConstant4Vector, inputs [], properties {constant=<rgba 0 0 0 1>}}
  MaterialExpressionConstant_0 {class=MaterialExpressionConstant}
  MaterialExpressionConstant_22 {class=MaterialExpressionConstant}
  MaterialExpressionFunctionInput_0 {class=MaterialExpressionFunctionInput, inputs [Preview {from="", output=""}], properties {inputType=FunctionInput_MaterialAttributes}}
  MaterialExpressionFunctionInput_1 {
   class=MaterialExpressionFunctionInput
   inputs [Preview {from=MaterialExpressionConstant3Vector_0, output=""}]
   properties {bUsePreviewValueAsDefault=true, inputName=Normal, sortPriority=10}
  }
  MaterialExpressionFunctionOutput_0 {class=MaterialExpressionFunctionOutput, inputs [None {from=MaterialExpressionStaticSwitchParameter_0, output=""}]}
  MaterialExpressionGetMaterialAttributes_2 {
   class=MaterialExpressionGetMaterialAttributes
   inputs [None {from=MaterialExpressionFunctionInput_0, output=""}]
   properties {attributeGetTypes [
     #1
     #2
     #3
     #4
    ]}
  }
  MaterialExpressionMaterialFunctionCall_16 {
   class=MaterialExpressionMaterialFunctionCall
   inputs [
    Height Map {from=MaterialExpressionNamedRerouteUsage_4}
    Normal Map Intensity {from=MaterialExpressionScalarParameter_0}
    Height Map UV Offset {from=MaterialExpressionScalarParameter_1}
    Coordinates {from=MaterialExpressionTextureCoordinate_1}
    Height Map Channel Selector {from=MaterialExpressionSwitch_5}
   ]
   properties {materialFunction=/Engine/Functions/Engine_MaterialFunctions03/Procedurals/NormalFromHeightmap}
  }
  MaterialExpressionMultiply_0 {
   class=MaterialExpressionMultiply
   inputs [
    A {from=MaterialExpressionMaterialFunctionCall_16, output=Result}
    B {from=MaterialExpressionConstant3Vector_1}
   ]
  }
  MaterialExpressionNamedRerouteDeclaration_0 {
   class=MaterialExpressionNamedRerouteDeclaration
   inputs [None {from=MaterialExpressionSwitch_4}]
   properties {name=Mask, nodeColor=<rgba 1 0 0.6094 1>}
  }
  MaterialExpressionNamedRerouteDeclaration_1 {
   class=MaterialExpressionNamedRerouteDeclaration
   inputs [None {from=MaterialExpressionTextureObjectParameter_0}]
   properties {name=Tex, nodeColor=<rgba 0 1 0.2266 1>}
  }
  MaterialExpressionNamedRerouteUsage_1 {class=MaterialExpressionNamedRerouteUsage}
  MaterialExpressionNamedRerouteUsage_4 {class=MaterialExpressionNamedRerouteUsage}
  MaterialExpressionNormalize_0 {class=MaterialExpressionNormalize, inputs [VectorInput {from="", output=""}]}
  MaterialExpressionReroute_0 {class=MaterialExpressionReroute, inputs [None {from=MaterialExpressionFunctionInput_0}]}
  MaterialExpressionReroute_1 {class=MaterialExpressionReroute, inputs [None {from=MaterialExpressionReroute_0}]}
  MaterialExpressionScalarParameter_0 {class=MaterialExpressionScalarParameter, properties {defaultValue=5, group=RoadMarkings, parameterName=RoadMarkings Normal Intensity}}
  MaterialExpressionScalarParameter_1 {class=MaterialExpressionScalarParameter, properties {defaultValue=0.0005, group=RoadMarkings, parameterName=RoadMarkings Normal UVOffset}}
  MaterialExpressionScalarParameter_44 {
   class=MaterialExpressionScalarParameter
   properties {bUseCustomPrimitiveData=true, group=RoadMarkings, parameterName=RoadMarkings Toggle, sliderMax=5}
  }
  MaterialExpressionScalarParameter_45 {class=MaterialExpressionScalarParameter, properties {defaultValue=0.7, group=RoadMarkings, parameterName=RoadMarkingsSpecular}}
  MaterialExpressionScalarParameter_46 {class=MaterialExpressionScalarParameter, properties {defaultValue=0.1, group=RoadMarkings, parameterName=RoadMarkignsRoughness}}
  MaterialExpressionSetMaterialAttributes_0 {
   class=MaterialExpressionSetMaterialAttributes
   inputs [
    MaterialAttributes {from="", output=""}
    BaseColor {from=MaterialExpressionVectorParameter_4, output=RGB}
    Specular {from=MaterialExpressionScalarParameter_45, output=""}
    Roughness {from=MaterialExpressionScalarParameter_46, output=""}
    Normal {from=MaterialExpressionAdd_0, output=""}
   ]
   properties {attributeSetTypes [
     #1
     #2
     #3
     #4
    ]}
  }
  MaterialExpressionShadingPathSwitch_0 {
   class=MaterialExpressionShadingPathSwitch
   inputs [
    Default {from="", output=""}
    Deferred {from="", output=""}
    Forward {from="", output=""}
    Mobile {from="", output=""}
   ]
   properties {desc=Migrated from FeatureLevelSwitch}
  }
  MaterialExpressionStaticSwitchParameter_0 {
   class=MaterialExpressionStaticSwitchParameter
   inputs [
    True {from=MaterialExpressionBlendMaterialAttributes_0, output=""}
    False {from=MaterialExpressionReroute_1, output=""}
   ]
   properties {defaultValue=true, parameterName=USe RoadMarkings}
  }
  MaterialExpressionSwitch_4 {
   class=MaterialExpressionSwitch
   inputs [
    SwitchValue {from=MaterialExpressionScalarParameter_44, output=""}
    Default {from="", output=""}
    None {from=MaterialExpressionConstant_22, output=""}
    None {from=MaterialExpressionTextureSample_15, output=R}
    None {from=MaterialExpressionTextureSample_15, output=R}
    None {from=MaterialExpressionTextureSample_15, output=R}
    None {from=MaterialExpressionTextureSample_15, output=R}
   ]
   properties {inputs [
     {
      inputName=None
      input {
       expression=&self:MaterialExpressionConstant_22
       inputName=None
       mask=0
       maskA=0
       maskB=0
       maskG=0
       maskR=0
       outputIndex=0
      }
     }
     {
      inputName=None
      input {
       expression=&self:MaterialExpressionTextureSample_15
       inputName=None
       mask=1
       maskA=0
       maskB=0
       maskG=0
       maskR=1
       outputIndex=1
      }
     }
     {
      inputName=None
      input {
       expression=&self:MaterialExpressionTextureSample_15
       inputName=None
       mask=1
       maskA=0
       maskB=0
       maskG=1
       maskR=0
       outputIndex=2
      }
     }
     {
      inputName=None
      input {
       expression=&self:MaterialExpressionTextureSample_15
       inputName=None
       mask=1
       maskA=0
       maskB=1
       maskG=0
       maskR=0
       outputIndex=3
      }
     }
     {
      inputName=None
      input {
       expression=&self:MaterialExpressionTextureSample_15
       inputName=None
       mask=1
       maskA=1
       maskB=0
       maskG=0
       maskR=0
       outputIndex=4
      }
     }
    ]}
  }
  MaterialExpressionSwitch_5 {
   class=MaterialExpressionSwitch
   inputs [
    SwitchValue {from=MaterialExpressionScalarParameter_44, output=""}
    Default {from="", output=""}
    None {from=MaterialExpressionConstant_0, output=""}
    None {from=MaterialExpressionConstant4Vector_0, output=RGBA}
    None {from=MaterialExpressionConstant4Vector_1, output=RGBA}
    None {from=MaterialExpressionConstant4Vector_2, output=RGBA}
    None {from=MaterialExpressionConstant4Vector_3, output=RGBA}
   ]
   properties {inputs [
     {
      inputName=None
      input {
       expression=&self:MaterialExpressionConstant_0
       inputName=None
       mask=0
       maskA=0
       maskB=0
       maskG=0
       maskR=0
       outputIndex=0
      }
     }
     {
      inputName=None
      input {
       expression=&self:MaterialExpressionConstant4Vector_0
       inputName=None
       mask=1
       maskA=1
       maskB=1
       maskG=1
       maskR=1
       outputIndex=0
      }
     }
     {
      inputName=None
      input {
       expression=&self:MaterialExpressionConstant4Vector_1
       inputName=None
       mask=1
       maskA=1
       maskB=1
       maskG=1
       maskR=1
       outputIndex=0
      }
     }
     {
      inputName=None
      input {
       expression=&self:MaterialExpressionConstant4Vector_2
       inputName=None
       mask=1
       maskA=1
       maskB=1
       maskG=1
       maskR=1
       outputIndex=0
      }
     }
     {
      inputName=None
      input {
       expression=&self:MaterialExpressionConstant4Vector_3
       inputName=None
       mask=1
       maskA=1
       maskB=1
       maskG=1
       maskR=1
       outputIndex=0
      }
     }
    ]}
  }
  MaterialExpressionTextureCoordinate_0 {class=MaterialExpressionTextureCoordinate}
  MaterialExpressionTextureCoordinate_1 {class=MaterialExpressionTextureCoordinate}
  MaterialExpressionTextureObjectParameter_0 {
   class=MaterialExpressionTextureObjectParameter
   inputs [
    UVs {from="", output=""}
    Apply View MipBias {from="", output=""}
   ]
   properties {
    parameterName=StreetMarking
    samplerType=SAMPLERTYPE_Masks
    texture=/Game/Environments/Asteria/Roads/SciFi/Meshes/Pieces_Update_v2/T_RoadMarkings_mask1
   }
  }
  MaterialExpressionTextureSample_15 {
   class=MaterialExpressionTextureSample
   inputs [
    UVs {from=MaterialExpressionTextureCoordinate_0}
    Tex {from=MaterialExpressionNamedRerouteDeclaration_1}
    Apply View MipBias {from=""}
   ]
   properties {samplerSource=SSM_Wrap_WorldGroupSettings, texture=&T_RoadMarkings_mask1}
  }
  MaterialExpressionVectorParameter_4 {
   class=MaterialExpressionVectorParameter
   properties {group=RoadMarkings, parameterName=RoadMarkingsColor, defaultValue=<rgba 0.3385 0.3385 0.3385 1>}
  }
 }}
