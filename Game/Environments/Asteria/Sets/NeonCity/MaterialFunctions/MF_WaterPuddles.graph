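# Material function that adds rain wetness and puddles to incoming material attributes in the NeonCity environment set.
# A world-space XY texture yields wet and puddle masks, shaped by a Height input and, for puddles, surface slope.
# They adjust color, roughness, specular and normal, with two-scale rain ripples at Epic quality, optionally night-only.
# Switches add UV1 layer maskouts, vertex-color masking and an object-relative wet/dry gradient, also exposed as an output.
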MaterialFunction /Game/Environments/Asteria/Sets/NeonCity/MaterialFunctions/MF_WaterPuddles
graph {nodes [
  n1 = NamedRerouteUsage()
  Ceil(None=@n1)
  n2 = FunctionInput() {inputType=FunctionInput_MaterialAttributes}
  n3 = GetMaterialAttributes(None=@n2) {attributeGetTypes [
    #1
    #2
    #3
    #4
    #5
   ]}
  n4 = Min(
   A=LinearInterpolate(A=@n3.Roughness, B=ScalarParameter() {group=WaterPuddles, parameterName=Puddle_Roughness}, Alpha=NamedRerouteUsage())
   B=ScalarParameter() {defaultValue=0.32, group=WaterPuddles, parameterName=Puddle_Roughness_min}
  )
  n5 = Saturate(None=Multiply(A=@n3.Metallic, B=ScalarParameter() {defaultValue=1, parameterName=Puddle_Roughness_OrigianlMetal}))
  n6 = Multiply(A=ScalarParameter() {defaultValue=1, group=WaterPuddles, parameterName=Puddle_Specular}, B=@n3.Specular)
  n7 = NamedRerouteUsage()
  n8 = Constant3Vector() {constant=<rgba 0 0 1 0>}
  n9 = Transform(None=@n8) {transformSourceType=TRANSFORMSOURCE_World, transformType=TRANSFORM_Tangent}
  n10 = Constant()
  n11 = MaterialFunctionCall(Value_Morning=@n10, Value_Day=@n10, Value_Evening=@n10, Value_Night=Constant() {r=1}) {materialFunction=&Scalar_Time_of_Day}
  n12 = ComponentMask(None=WorldPosition().XYZ) {g=true, r=true}
  n13 = ScalarParameter() {defaultValue=134.7562, group=WaterPuddles, parameterName=Ripple_Size}
  n14 = Divide(
   A=ComponentMask(None=WorldPosition().XYZ) {g=true, r=true}
   B=ScalarParameter() {defaultValue=111.5522, group=WaterPuddles, parameterName=Ripple_Size2}
  )
  n15 = Add(
   A=MaterialFunctionCall!(UV=Divide(A=@n12, B=@n13)) {materialFunction=&/MF_WaterRainRipple}.Normal
   B=Multiply(
    A=MaterialFunctionCall(UV=@n14) {materialFunction=&MF_WaterRainRipple}.Normal
    B=Constant3Vector() {constant=<rgba 1 1 0 0>}
   )
  )
  n16 = Transform(None=StaticSwitchParameter(True=LinearInterpolate(A=@n8, B=@n15, Alpha=@n11.Result), False=@n15) {defaultValue=true, group=WaterPuddles, parameterName=RainDrops TODM}) {transformSourceType=TRANSFORMSOURCE_World, transformType=TRANSFORM_Tangent}
  n17 = ScalarParameter() {defaultValue=0.6, group=WaterPuddles, parameterName=PuddleWet_Darken}
  n18 = NamedRerouteUsage()
  n19 = SetMaterialAttributes!(
   MaterialAttributes=@n3.MaterialAttributes
   Roughness=StaticSwitchParameter(True=LinearInterpolate(A=@n4, B=@n3.Roughness, Alpha=@n5), False=@n4) {parameterName=UseMetallic}
   Specular=Min(A=LinearInterpolate(A=@n3.Specular, B=@n6, Alpha=@n7)) {constB=0.5}
   Normal=LinearInterpolate(A=@n3.Normal, B=QualitySwitch(Default=@n9, Epic=@n16), Alpha=NamedRerouteUsage())
   BaseColor=Multiply(A=@n3.BaseColor, B=LinearInterpolate(B=@n17, Alpha=@n18) {constA=1})
  ) {attributeSetTypes [
    #1
    #2
    #3
    #4
   ]}
  n20 = StaticSwitchParameter(
   True=BlendMaterialAttributes(A=@n2, B=@n19, Alpha=NamedRerouteUsage())
   False=BlendMaterialAttributes(A=@n2, B=@n19, Alpha=NamedRerouteUsage())
  ) {group=WetMasking, parameterName=UseWetDryTransition}
  n21 = VertexColor()
  n22 = AppendVector(A=@n21, B=@n21)
  n23 = StaticSwitchParameter(
   True=BlendMaterialAttributes(A=@n2, B=@n20, Alpha=ChannelMaskParameter(None=@n22) {group=WetMasking, parameterName=WaterWetness_VertexColorChannelMask}.R) {vertexAttributeBlendType=UseA}
   False=@n20
  ) {group=WetMasking, parameterName=UseWaterWetness_VertexColorMask}
  FunctionOutput(None=MaterialProxyReplace(
    Realtime=StaticSwitchParameter(True=@n23, False=@n2) {defaultValue=true, group=WaterPuddles, parameterName=UseWaterPuddles}
    MaterialProxy=@n19
   ))
  n24 = DotProduct(
   A=Transform(None=VectorParameter() {group=WetMasking, parameterName=WetDry_Gradient Direction, defaultValue=<rgba 0 1 0 1>}.RGB) {transformSourceType=TRANSFORMSOURCE_Local}
   B=Subtract(A=WorldPosition().XYZ, B=ObjectPositionWS())
  )
  n25 = Multiply(A=ScalarParameter() {defaultValue=1, group=WetMasking, parameterName=WetDry_GradientLength}, B=ObjectRadius())
  n26 = Divide(A=@n25)
  n27 = Power(Base=Divide(A=Add(A=@n24, B=@n26), B=@n25), Exp=ScalarParameter() {defaultValue=1, group=WetMasking, parameterName=WetDry_Gradient_Falloff})
  FunctionOutput(None=NamedRerouteDeclaration(None=Saturate(None=@n27)) {name=WetDryMask, nodeColor=<rgba 1 0 0.1172 1>}) {outputName=DryWetMaskTransition, sortPriority=10}
  n28 = Round(None=@n1)
  n29 = ScalarParameter() {defaultValue=0.752, parameterName=Offset PuddleByHeight}
  n30 = ComponentMask(None=WorldPosition().XYZ) {g=true, r=true}
  n31 = ScalarParameter() {defaultValue=2000, group=WaterPuddles, parameterName=Puddle_Size}
  n32 = TextureSample(UVs=Add(A=Multiply(A=@n28, B=@n29), B=Divide(A=@n30, B=@n31))) {
   samplerSource=SSM_Wrap_WorldGroupSettings
   samplerType=SAMPLERTYPE_LinearColor
   texture=/Game/Environments/Asteria/Sets/NeonCity/Textures/WaterPuddles_M
  }
  n33 = Saturate(None=LinearInterpolate(
    A=ScalarParameter() {defaultValue=0.4608, group=WaterPuddles, parameterName=PuddleWetness_Contrast_Low}
    B=ScalarParameter() {defaultValue=2.7439, group=WaterPuddles, parameterName=PuddleWetness_Contrast_High}
    Alpha=@n32.R
   ))
  n34 = Multiply(A=NamedRerouteUsage(), B=ScalarParameter() {defaultValue=0.1, group=WaterPuddles, parameterName=PuddleWetness_HeightMap_influence})
  NamedRerouteDeclaration(None=Saturate(None=Subtract(A=@n33, B=@n34))) {name=WetMask, nodeColor=<rgba 0 1 0.4844 1>}
  n35 = ScalarParameter() {defaultValue=-6.4879, group=WaterPuddles, parameterName=Puddle_Contrast_Low}
  n36 = ScalarParameter() {defaultValue=17.5196, group=WaterPuddles, parameterName=Puddle_Contrast_High}
  n37 = Subtract(
   A=Saturate(None=LinearInterpolate(A=@n35, B=@n36, Alpha=@n32.R))
   B=Multiply(A=NamedRerouteUsage(), B=ScalarParameter() {defaultValue=1, group=WaterPuddles, parameterName=Puddle_HeightMap_influence})
  )
  n38 = LinearInterpolate(
   A=ScalarParameter() {defaultValue=-78.5354, group=WaterPuddles, parameterName=PuddleSlopeMask_Low}
   B=ScalarParameter() {defaultValue=1.2891, group=WaterPuddles, parameterName=PuddleSlopeMask_High}
   Alpha=ComponentMask(None=VertexNormalWS()) {b=true}
  )
  NamedRerouteDeclaration(None=Multiply(A=Saturate(None=@n37), B=Saturate(None=@n38))) {name=PuddleMask, nodeColor=<rgba 1 0.8438 0 1>}
  NamedRerouteDeclaration(None=FunctionInput() {inputName=Height, inputType=FunctionInput_Scalar, sortPriority=10}) {name=Height, nodeColor=<rgba 1 0 0.1875 1>}
  n39 = Constant() {r=3}
  n40 = ComponentMask(None=TextureCoordinate() {coordinateIndex=1}) {r=true}
  n41 = Floor(None=@n40)
  n42 = Abs(None=Subtract(A=@n41, B=Constant() {r=2}))
  n43 = Abs(None=Subtract(A=@n41, B=Constant() {r=1}))
  n44 = Constant() {r=1}
  n45 = StaticSwitch(
   True=If(A=@n40, B=Constant() {r=1}, "A > B"=@n44, "A < B"=Constant()) {desc=0}
   False=@n44
   Value=StaticBoolParameter() {group=WetMasking, parameterName=Wet_Maskout_Layer_1}
  )
  n46 = StaticSwitch(
   True=If(A=@n40, B=Constant() {r=1}, "A > B"=Saturate(None=@n43), "A < B"=@n45) {desc=0}
   False=@n45
   Value=StaticBoolParameter() {group=WetMasking, parameterName=Wet_Maskout_Layer_2}
  )
  n47 = StaticSwitch(
   True=If(A=@n40, B=Constant() {r=2}, "A > B"=Saturate(None=@n42), "A < B"=@n46)
   False=@n46
   Value=StaticBoolParameter() {group=WetMasking, parameterName=Wet_Maskout_Layer_3}
  )
  n48 = Constant() {r=3}
  n49 = Saturate(None=Abs(None=Subtract(A=@n41, B=@n48)))
  NamedRerouteDeclaration(None=StaticSwitch(
    True=If(A=@n40, B=@n39, "A > B"=@n49, "A < B"=@n47)
    False=@n47
    Value=StaticBoolParameter() {group=WetMasking, parameterName=Wet_Maskout_Layer_4}
   )) {name=WetLayerMask, nodeColor=<rgba 0.8047 0 1 1>}
  NamedRerouteUsage()
  QualitySwitch()
  Reroute()
  Reroute()
  Reroute()
  Reroute()
  Reroute()
  Reroute()
  Reroute()
  ScalarParameter() {defaultValue=1.487, parameterName=RippleSpeed}
 ]}
