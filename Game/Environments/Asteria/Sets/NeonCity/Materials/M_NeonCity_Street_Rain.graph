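# Rainy NeonCity street material for Nanite and spline meshes, with a wet concrete physical material.
# An asphalt function gains world-space patches and cracks with heightmap-derived normals, then passes through road markings, dirt and water puddle functions.
# Puddles come from two rotated world-tiled masks, and the puddle function's wet-dry transition can optionally mask patches and cracks.
# Quality and shading-path switches drop patches and cracks on Low, Medium and Mobile.
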Material /Game/Environments/Asteria/Sets/NeonCity/Materials/M_NeonCity_Street_Rain
graph {
 properties {
  bUseMaterialAttributes=true
  bUsedWithNanite=true
  bUsedWithSplineMeshes=true
  physMaterial=&/PhysMat_NeonCity_WetConcrete
  pixelDepthOffsetMode=PDOM_Legacy
  usedShadingModels=Default Lit
 }
 nodes [
  n1 = ScalarParameter() {defaultValue=0.5, group=AsphaltPatches, parameterName=AsphaltPatch_BevelStrength}
  n2 = ScalarParameter() {defaultValue=6, group=AsphaltPatches, parameterName=AsphaltPatch Normal Intensity}
  n3 = MaterialFunctionCall!(TexelSize=NamedRerouteUsage(), Texture=NamedRerouteUsage(), Bevel Strength=@n1, Normal Intensity=@n2, UVs=NamedRerouteUsage()) {materialFunction=/Game/Environments/Asteria/Roads/SciFi/Meshes/Pieces_Update_v2/MF_SobelGrayscale2Normal}
  n4 = MaterialFunctionCall(TexelSize=NamedRerouteUsage(), Texture=NamedRerouteUsage(), Bevel Strength=@n1, Normal Intensity=@n2, UVs=NamedRerouteUsage()) {materialFunction=&MF_SobelGrayscale2Normal}
  n5 = NamedRerouteUsage()
  AppendVector(A=ComponentMask(None=LinearInterpolate(A=@n3.Result - B, B=@n4.Result - R, Alpha=@n5)) {g=true, r=true}, B=Constant())
  MaterialFunctionCall(
   TexelSize=NamedRerouteUsage()
   Texture=NamedRerouteUsage()
   Bevel Strength=Multiply(A=NamedRerouteUsage(), B=ScalarParameter() {defaultValue=0.25, group=Asphalt, parameterName=Cracks Intensity, sortPriority=40})
   Normal Intensity=NamedRerouteUsage()
   UVs=NamedRerouteUsage()
  ) {materialFunction=&MF_SobelGrayscale2Normal}
  n6 = NamedRerouteUsage()
  n7 = ScalarParameter() {defaultValue=1300, group=Puddles, parameterName=WorldTileDivisor_Puddles_A}
  n8 = VertexInterpolator(VS=Rotator(Coordinate=Divide(A=@n6, B=@n7), Time=ScalarParameter() {defaultValue=1.0787, group=Puddles, parameterName=Puddles_Rotation_A}))
  n9 = NamedRerouteUsage()
  n10 = ScalarParameter() {defaultValue=675.0765, group=Puddles, parameterName=WorldTileDivisor_Puddles_B}
  n11 = VertexInterpolator(VS=Rotator(Coordinate=Divide(A=@n9, B=@n10), Time=ScalarParameter() {defaultValue=3.2432, group=Puddles, parameterName=Puddles_Rotation_B}))
  n12 = Power(
   Base=Min(
    A=TextureSample(UVs=@n8.PS) {samplerSource=SSM_Wrap_WorldGroupSettings, texture=/Game/Environments/Asteria/Roads/SciFi/Textures/CityRoads_MaskMap3}.G
    B=TextureSample(UVs=@n11.PS) {samplerSource=SSM_Wrap_WorldGroupSettings, texture=&CityRoads_MaskMap3}.G
   )
   Exp=ScalarParameter() {defaultValue=1, group=Puddles, parameterName=PuddleExponent}
  )
  n13 = ScalarParameter() {defaultValue=12, group=Puddles, parameterName=PuddleScale}
  NamedRerouteDeclaration(None=Saturate(None=Multiply(A=@n12, B=@n13))) {name=PuddleMask, nodeColor=<rgba 1 0 0.7734 1>}
  n14 = Add(A=NamedRerouteUsage(), B=NamedRerouteUsage())
  n15 = Saturate(None=LinearInterpolate(
    A=ScalarParameter() {parameterName=AsphaltPatch_Mask_low}
    B=ScalarParameter() {defaultValue=4, parameterName=AsphaltPatch_Mask_High}
    Alpha=Saturate(None=@n14)
   ))
  n16 = StaticSwitchParameter(True=Multiply(A=@n15, B=NamedRerouteUsage()), False=@n15) {group=WaterPuddles, parameterName=UseWetDryTransition}
  n17 = MaterialFunctionCall() {materialFunction=/Game/Environments/Asteria/Sets/NeonCity/MaterialFunctions/MF_NeonCity_Asphalt}
  n18 = StaticSwitchParameter(
   True=BlendMaterialAttributes(A=@n17.Result, B=NamedRerouteUsage(), Alpha=NamedRerouteDeclaration(None=@n16) {name=AsphaltPatchMask, nodeColor=<rgba 0 0.4141 1 1>}) {vertexAttributeBlendType=UseA}
   False=@n17.Result
  ) {defaultValue=true, group=AsphaltPatches, parameterName=AsphaltPatches On/Off}
  n19 = ShadingPathSwitch(Default=QualitySwitch(Default=@n18, Low=@n17.Result, High=@n18, Medium=@n17.Result, Epic=@n18), Mobile=@n17.Result) {desc=Migrated from FeatureLevelSwitch}
  n20 = GetMaterialAttributes(None=@n19) {attributeGetTypes [
    #1
    #2
    #3
    #4
   ]}
  n21 = NamedRerouteUsage()
  n22 = StaticSwitchParameter(True=Multiply(A=@n21, B=NamedRerouteUsage()), False=@n21) {group=WaterPuddles, parameterName=UseWetDryTransition}
  n23 = NamedRerouteUsage()
  n24 = ScalarParameter() {defaultValue=0.175, group=Asphalt, parameterName=Cracks Normal Intensity, sortPriority=40}
  n25 = ScalarParameter() {defaultValue=0.005, group=Asphalt, parameterName=Cracks Normal uvoffset, sortPriority=40}
  n26 = NamedRerouteUsage()
  n27 = Constant4Vector() {constant=<rgba 0 0 0 1>}
  n28 = Transform(None=Multiply(
    A=MaterialFunctionCall!(Height Map=@n23, Normal Map Intensity=@n24, Height Map UV Offset=@n25, Coordinates=@n26, Height Map Channel Selector=@n27.RGBA) {materialFunction=/Engine/Functions/Engine_MaterialFunctions03/Procedurals/NormalFromHeightmap}.Result
    B=Constant3Vector() {constant=<rgba -1 -1 0 0>}
   )) {transformSourceType=TRANSFORMSOURCE_World, transformType=TRANSFORM_Tangent}
  n29 = StaticSwitchParameter(True=Multiply(A=@n28, B=NamedRerouteUsage()), False=@n28) {group=WaterPuddles, parameterName=UseWetDryTransition}
  n30 = SetMaterialAttributes!(
   MaterialAttributes=@n20.MaterialAttributes
   Base Color=LinearInterpolate(A=@n20.Base Color, B=Multiply(A=@n20.Base Color) {constB=0.5}, Alpha=@n22)
   Specular=Subtract(A=@n20.Specular, B=Multiply(A=@n22) {constB=0.1})
   Roughness=LinearInterpolate(A=@n20.Roughness, B=Constant() {r=0.5}, Alpha=@n22)
   Normal=Normalize(VectorInput=Add(A=@n20.Normal, B=@n29))
  ) {attributeSetTypes [
    #1
    #2
    #3
    #4
   ]}
  n31 = StaticSwitchParameter(True=@n30, False=@n19) {defaultValue=true, group=AsphaltPatches, parameterName=UseCracks}
  n32 = MaterialFunctionCall!(
   In=ShadingPathSwitch(Default=QualitySwitch(Default=@n31, Low=@n19, High=@n31, Medium=@n19, Epic=@n31), Mobile=@n19) {desc=Migrated from FeatureLevelSwitch}
   Normal=NamedRerouteUsage()
  ) {materialFunction=/Game/Environments/Asteria/Sets/NeonCity/MaterialFunctions/MF_NeonCity_RoadMarkings}
  n33 = NamedRerouteUsage()
  n34 = NamedRerouteUsage()
  n35 = NamedRerouteUsage()
  n36 = Constant()
  n37 = ShadingPathSwitch(Default=QualitySwitch(Default=@n35, Low=@n36, High=@n35, Medium=@n36, Epic=@n35), Mobile=@n36) {desc=Migrated from FeatureLevelSwitch}
  n38 = TextureSampleParameter2D() {
   group=Asphalt
   parameterName=PuddleHeight
   samplerSource=SSM_Wrap_WorldGroupSettings
   samplerType=SAMPLERTYPE_Masks
   sortPriority=3
   texture=/Game/Athena/Apollo/Environments/Landscape/Textures/Roads/T_AsphaltDetail1_S
  }
  n39 = ChannelMaskParameter(None=@n38.RGBA) {parameterName=HeightChannel}
  NamedRerouteDeclaration(None=MaterialFunctionCall!(
    In=MaterialFunctionCall!(In=@n32.Result, Pebble_Mask=@n33, Puddle_Mask=@n34) {materialFunction=/Game/Environments/Asteria/Sets/NeonCity/MaterialFunctions/MF_NeonCity_RoadDirt}.Result
    Height=StaticSwitchParameter(True=@n37, False=@n39.R) {defaultValue=true, group=AsphaltPatches, parameterName=AsphaltPatches On/Off}
   ) {materialFunction=/Game/Environments/Asteria/Sets/NeonCity/MaterialFunctions/MF_WaterPuddles}.DryWetMaskTransition) {name=DryWetTransition, nodeColor=<rgba 0 0.3437 1 1>}
  NamedRerouteDeclaration(None=@n1) {name=AsphaltPatch_BevelStrength, nodeColor=<rgba 0.9219 0 1 1>}
  NamedRerouteDeclaration(None=@n2) {name=AsphaltPatch_NormalIntensity, nodeColor=<rgba 0 0.6953 1 1>}
  n40 = ComponentMask(None=WorldPosition().XYZ) {g=true, r=true}
  n41 = ScalarParameter() {defaultValue=3000, group=AsphaltPatches, parameterName=WorldTileDivisor_AsphaltPatches_A}
  n42 = ScalarParameter() {defaultValue=-0.2, group=AsphaltPatches, parameterName=AsphaltPatches_Rotation_A}
  n43 = VertexInterpolator(VS=Rotator(Coordinate=Divide(A=@n40, B=@n41), Time=@n42))
  n44 = TextureObject() {texture=/Game/Environments/Asteria/Roads/SciFi/Textures/CityRoads_MaskMap2}
  n45 = TextureSample(UVs=@n43.PS, Tex=@n44) {
   samplerSource=SSM_Wrap_WorldGroupSettings
   samplerType=SAMPLERTYPE_Masks
   texture=/Game/Environments/Asteria/Roads/SciFi/Textures/CityRoads_MaskMap1
  }
  NamedRerouteDeclaration(None=Saturate(None=Power(Base=@n45.R) {constExponent=0.5})) {name=AsphaltPatchA_Mask, nodeColor=<rgba 1 0.0938 0 1>}
  n46 = ScalarParameter() {defaultValue=3500, group=AsphaltPatches, parameterName=WorldTileDivisor_AsphaltPatches_B}
  n47 = VertexInterpolator(VS=Rotator(Coordinate=Divide(A=@n40, B=@n46), Time=@n42))
  n48 = TextureSample(UVs=@n47.PS, Tex=@n44) {
   samplerSource=SSM_Wrap_WorldGroupSettings
   samplerType=SAMPLERTYPE_Masks
   texture=&CityRoads_MaskMap1
  }
  NamedRerouteDeclaration(None=Saturate(None=Power(Base=@n48.B) {constExponent=1})) {name=AsphaltPatchB_Mask, nodeColor=<rgba 0 1 0.7656 1>}
  n49 = TextureSampleParameter2D(UVs=NamedRerouteUsage()) {
   group=Asphalt
   parameterName=Diffuse
   samplerSource=SSM_Wrap_WorldGroupSettings
   sortPriority=2
   texture=/Game/Athena/Apollo/Environments/Landscape/Textures/Roads/T_AsphaltDetail1_D
  }
  n50 = ScalarParameter() {defaultValue=0.65, group=AsphaltPatches, parameterName=AsphaltPatchesB_diffuseDarken}
  n51 = ScalarParameter() {defaultValue=0.9, group=AsphaltPatches, parameterName=AsphaltPatchesA_diffuseDarken}
  n52 = TextureSampleParameter2D(UVs=NamedRerouteUsage()) {
   group=Asphalt
   parameterName=Masks
   samplerSource=SSM_Wrap_WorldGroupSettings
   samplerType=SAMPLERTYPE_Masks
   sortPriority=3
   texture=&T_AsphaltDetail1_S
  }
  n53 = TextureSampleParameter2D(UVs=NamedRerouteUsage()) {
   group=Asphalt
   parameterName=Normals
   samplerSource=SSM_Wrap_WorldGroupSettings
   samplerType=SAMPLERTYPE_Normal
   sortPriority=4
   texture=/Game/Athena/Apollo/Environments/Landscape/Textures/Roads/T_AsphaltDetail1_N
  }
  n54 = NamedRerouteUsage()
  n55 = ScalarParameter() {defaultValue=3, group=AsphaltPatches, parameterName=AsphaltPatch Normal Base}
  n56 = ScalarParameter() {defaultValue=0.0001, group=AsphaltPatches, parameterName=AsphaltPatch_BevelOffset}
  n57 = NamedRerouteUsage()
  n58 = Constant4Vector() {constant=<rgba 1 0 1 0>}
  n59 = NamedRerouteUsage()
  n60 = Constant4Vector() {constant=<rgba 1 0 0 0>}
  n61 = Multiply(
   A=LinearInterpolate(
    A=MaterialFunctionCall(Height Map=@n54, Normal Map Intensity=@n55, Height Map UV Offset=@n56, Coordinates=@n57, Height Map Channel Selector=@n58.RGBA) {materialFunction=&NormalFromHeightmap}.Result
    B=MaterialFunctionCall(Height Map=@n54, Normal Map Intensity=@n55, Height Map UV Offset=@n56, Coordinates=@n59, Height Map Channel Selector=@n60.RGBA) {materialFunction=&NormalFromHeightmap}.Result
    Alpha=NamedRerouteUsage()
   )
   B=ScalarParameter() {defaultValue=2, group=AsphaltPatches, parameterName=AsphaltPatch Normal Boost}
  )
  n62 = ScalarParameter() {defaultValue=0.148, group=AsphaltPatches, parameterName=AsphaltPatch Normal Max}
  n63 = Multiply(A=@n62) {constB=-1}
  n64 = AppendVector(A=ComponentMask(None=Clamp(None=@n61, Min=@n63, Max=@n62)) {g=true, r=true}, B=Constant())
  n65 = SetMaterialAttributes!(
   MaterialAttributes
   Base Color=LinearInterpolate(A=Multiply(A=@n49.RGB, B=@n50), B=Multiply(A=@n49.RGB, B=@n51), Alpha=NamedRerouteUsage())
   Metallic
   Specular=Multiply(A=ScalarParameter() {defaultValue=0.502, group=AsphaltPatches, parameterName=AsphaltPatch_Specular}, B=@n52.R)
   Roughness=Multiply(A=ScalarParameter() {defaultValue=0.7, group=AsphaltPatches, parameterName=AsphaltPatcheB_roughness}, B=@n52.B)
   Normal=Add(
    A=NamedRerouteDeclaration(None=@n53.RGB) {name=AsphaltOffsetNormal, nodeColor=<rgba 0 0.0625 1 1>}
    B=Transform(None=@n64) {transformSourceType=TRANSFORMSOURCE_World, transformType=TRANSFORM_Tangent}
   )
  ) {attributeSetTypes [
    #1
    #5
    #2
    #3
    #4
   ]}
  NamedRerouteDeclaration(None=@n65) {name=AsphaltPatchMat, nodeColor=<rgba 0 0.3437 1 1>}
  NamedRerouteDeclaration(None=@n17.Asphalt Patches UVs) {name=AsphaltPatchOffsetUVs, nodeColor=<rgba 0.875 1 0 1>}
  NamedRerouteDeclaration(None=@n47.PS) {name=AsphaltPatchesB_WS_UVs, nodeColor=<rgba 0.2422 0 1 1>}
  NamedRerouteDeclaration(None=@n17.Asphalt UVs) {name=Asphalt UVs, nodeColor=<rgba 0 0.9531 1 1>}
  NamedRerouteDeclaration(None=TextureProperty(Tex=NamedRerouteDeclaration(None=@n44) {name=MaskTexture, nodeColor=<rgba 0.875 0 1 1>}) {property=TMTM_TexelSize}) {name=TexelSize, nodeColor=<rgba 1 0.0545 0.1242 1>}
  NamedRerouteDeclaration(None=@n43.PS) {name=AsphaltPatchesA_WS_UVs, nodeColor=<rgba 0.0778 1 0.5258 1>}
  NamedRerouteDeclaration(None=Saturate(None=@n45.A)) {name=AsphaltCracks_Mask, nodeColor=<rgba 1 0.5859 0 1>}
 ]
}
