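# Material function that optionally adds a rainbow-like anodized fresnel sheen to incoming Base Color, intended for layered metals.
# A normal-driven Fresnel indexes a LUT texture, scaled by four layer strengths and tinted by base color, optionally blended toward a custom edge color.
# An SFX texture channel or the optionally inverted Metallic attribute masks the result before it is added.
# A shading path switch limits the effect to the default path, excluding forward and mobile.
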MaterialFunction /Game/Global/Base4Layers_2022/MaterialFunctions/MF_Anodizing_New
graph {nodes [
  Constant() {r=0.05}
  n1 = NamedRerouteUsage()
  n2 = StaticSwitchParameter(True=TextureCoordinate() {coordinateIndex=2}, False=TextureCoordinate() {coordinateIndex=1}) {group=C_Anodizing, parameterName=Anodizing SFX UV2, sortPriority=-3}
  n3 = FunctionInput() {inputName=SFX Textures, inputType=FunctionInput_Texture2D, sortPriority=1}
  n4 = StaticSwitchParameter(
   True=NamedRerouteDeclaration(None=TextureSample(UVs=@n2, Tex=@n3).RGB) {name=SFX Shared, nodeColor=<rgba 1 0.6094 0 1>}
   False=FunctionInput() {inputName=SFX Blended}
  ) {
   desc="Default is layers UV 0 ( UE/Maya) Will add 1 texture sample when enabled."
   group=C_Anodizing
   parameterName=Anodizing SFX UV1
   sortPriority=-3
  }
  n5 = VectorParameter() {
   desc="Use one of R G B channels to select mask channels from your SFX texture. Per level or global. "
   group=C_Anodizing
   parameterName=Anodizing_SFX_Mask_Ch
   sortPriority=2
  }
  n6 = StaticSwitchParameter(
   True=StaticSwitchParameter(True=OneMinus(None=@n1), False=@n1) {desc=Uses your Height maps as the mask, group=C_Anodizing, parameterName=Anodizing Invert Metal Mask, sortPriority=-2}
   False=DotProduct(A=@n4, B=@n5.RGB)
  ) {desc=Uses your Height maps as the mask, group=C_Anodizing, parameterName=Anodizing Use Layer Metal as Mask, sortPriority=-1}
  n7 = Fresnel(
   ExponentIn=ScalarParameter() {defaultValue=1.5, group=C_Anodizing, parameterName=Anodizing Exponent, sortPriority=3}
   BaseReflectFractionIn=ScalarParameter() {group=C_Anodizing, parameterName=Anodizing Camera Facing, sortPriority=3}
   Normal=Transform(None=NamedRerouteUsage())
  ) {baseReflectFraction=0.5, exponent=3}
  n8 = Multiply(
   A=MaterialFunctionCall!(
    1=ScalarParameter() {defaultValue=1, group=C_Anodizing, parameterName=Anodizing Layer 1, sortPriority=4}
    2=ScalarParameter() {defaultValue=1, group=C_Anodizing, parameterName=Anodizing Layer 2, sortPriority=5}
    3=ScalarParameter() {defaultValue=1, group=C_Anodizing, parameterName=Anodizing Layer 3, sortPriority=6}
    4=ScalarParameter() {defaultValue=1, group=C_Anodizing, parameterName=Anodizing Layer 4, sortPriority=7}
   ) {materialFunction=&/MF_UV2LerpScalar_Layers_New}.Result
   B=@n7
  )
  n9 = TextureSampleParameter2D(UVs=@n7) {
   group=C_Anodizing
   parameterName=IrradianceCurveLUT
   samplerSource=SSM_Clamp_WorldGroupSettings
   texture=/Game/Global/Textures/Default/Masks_and_Gradients/3pxRGBImage
  }
  n10 = NamedRerouteUsage()
  n11 = StaticSwitchParameter(
   True=LinearInterpolate(A=@n10, B=VectorParameter() {group=C_Anodizing, parameterName=Anodizing Edge, sortPriority=9}.RGB, Alpha=Saturate(None=@n7))
   False=@n10
  ) {
   desc=Enable for rainbow like fresnel effect. Good for metals. Disabled on mobile and swtich
   group=C_Anodizing
   parameterName=Anodizing Custom Edge Color
   sortPriority=8
  }
  n12 = Multiply(A=@n6, B=Multiply(A=Multiply(A=@n8, B=@n9.RGB), B=@n11))
  n13 = StaticSwitchParameter(
   True=SetMaterialAttributes!(MaterialAttributes=NamedRerouteUsage(), BaseColor=Add(A=@n12, B=@n11)) {attributeSetTypes [#1]}
   False=NamedRerouteUsage()
  ) {
   desc=Enable for rainbow like fresnel effect. Good for metals. Disabled on mobile and swtich
   group=C_Anodizing
   parameterName=Anodizing
   sortPriority=0
  }
  n14 = NamedRerouteUsage()
  FunctionOutput(None=ShadingPathSwitch(Default=@n13, Forward=@n14, Mobile=@n14))
  n15 = GetMaterialAttributes(None=NamedRerouteDeclaration(None=FunctionInput() {inputType=FunctionInput_MaterialAttributes, sortPriority=-3}) {name=AnodizingMatInput, nodeColor=<rgba 0.0312 1 0 1>}) {attributeGetTypes [#1, #2, #3]}
  NamedRerouteDeclaration(None=@n15.Normal) {name=Normal Map, nodeColor=<rgba 0 0.8828 1 1>}
  NamedRerouteDeclaration(None=@n15.BaseColor) {name=Base Color, nodeColor=<rgba 0.4297 0 1 1>}
  NamedRerouteDeclaration(None=@n15.Metallic) {name=Metal Mask, nodeColor=<rgba 0 1 0.8828 1>}
 ]}
