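# Material function that adds an optional fuzzy edge effect to incoming material attributes, toggled by the FuzzyEdge static switch.
# A world-normal Fresnel term brightens base color with a tint and lerps roughness toward a fuzz value.
# The result blends in through a channel-selected SFX mask, from either a texture sample with switchable UVs or a passed-in input.
# MF_UV2LerpScalar_Layers_New combines four per-layer fuzz values to scale that mask.
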MaterialFunction /Game/Global/Base4Layers_2022/MaterialFunctions/MF_FuzzyEdge2
graph {nodes [
  n1 = FunctionInput() {inputType=FunctionInput_MaterialAttributes, sortPriority=-2}
  n2 = GetMaterialAttributes(None=@n1) {attributeGetTypes [#1, #2, #3]}
  n3 = VectorParameter() {group=B_FuzzyEdge, parameterName=Fuzz Tint, defaultValue=<rgba 1 1 1 0>}
  n4 = ScalarParameter() {group=B_FuzzyEdge, parameterName=FuzzCenter, sliderMax=1, sortPriority=3}
  n5 = Saturate(None=Fresnel(
    ExponentIn=ScalarParameter() {defaultValue=3, group=B_FuzzyEdge, parameterName=FuzzExponent, sliderMax=5, sliderMin=1, sortPriority=1}
    BaseReflectFractionIn=Saturate(None=@n4)
    Normal=Transform(None=@n2.Normal)
   ))
  n6 = Add(A=@n2.BaseColor, B=Multiply(A=@n2.BaseColor, B=Multiply(A=@n3.RGB, B=@n5)))
  n7 = LinearInterpolate(A=@n2.Roughness, B=ScalarParameter() {defaultValue=1, group=B_FuzzyEdge, parameterName=FuzzRoughness, sortPriority=2}, Alpha=@n5)
  n8 = VectorParameter() {
   desc="Use one of R G B channels to select mask channels from your SFX texture. Per level or global. "
   group=B_FuzzyEdge
   parameterName=Fuzz_SFX_Mask_Ch
   sortPriority=-1
   defaultValue=<rgba 1 0 0 0>
  }
  n9 = ScalarParameter() {defaultValue=1, group=B_FuzzyEdge, parameterName=Fuzz Value Layer 1, sortPriority=0}
  n10 = ScalarParameter() {defaultValue=1, group=B_FuzzyEdge, parameterName=Fuzz Value Layer 2, sortPriority=0}
  n11 = ScalarParameter() {defaultValue=1, group=B_FuzzyEdge, parameterName=Fuzz Value Layer 3, sortPriority=0}
  n12 = ScalarParameter() {defaultValue=1, group=B_FuzzyEdge, parameterName=Fuzz Value Layer 4, sortPriority=0}
  n13 = TextureSample(
   UVs=StaticSwitchParameter(True=TextureCoordinate() {coordinateIndex=1}, False=TextureCoordinate() {coordinateIndex=2}) {desc="Default is UV1 ( UE/Maya)", group=B_FuzzyEdge, parameterName=Fuzz SFX UV2, sortPriority=-3}
   Tex=FunctionInput() {inputName=SFX Textures, inputType=FunctionInput_Texture2D}
  )
  n14 = DotProduct(
   A=StaticSwitchParameter(True=NamedRerouteDeclaration(None=@n13.RGB) {name=SFX Shared, nodeColor=<rgba 1 0.6094 0 1>}, False=FunctionInput() {inputName=SFX Blend}) {desc="Use to enable SFX texture with custom UV's  Adds texture sample", group=B_FuzzyEdge, parameterName=Fuzz SFX UV1}
   B=Multiply(
    A=@n8.RGB
    B=MaterialFunctionCall!(1=@n9, 2=@n10, 3=@n11, 4=@n12) {materialFunction=&/MF_UV2LerpScalar_Layers_New}.Result
   )
  )
  n15 = StaticSwitchParameter(
   True=BlendMaterialAttributes(
    A=@n1
    B=SetMaterialAttributes!(MaterialAttributes=@n2.MaterialAttributes, BaseColor=@n6, Roughness=@n7) {attributeSetTypes [#1, #2]}
    Alpha=Saturate(None=@n14)
   ) {vertexAttributeBlendType=UseA}
   False=@n1
  ) {
   desc="Fuzzy effect on object edges, uses SFX texture to mask area(0_SFX_Global). Use Fuzz_SFX_Mask to select correct channel. You can control it per material layer if needed."
   group=B_FuzzyEdge
   parameterName=FuzzyEdge
   sortPriority=-4
  }
  FunctionOutput(None=@n15)
 ]}
