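# Four-layer base parent material for NeonCity environment meshes in Fortnite, built on the Base4Layers 2022 functions, with Nanite, instanced and spline mesh support.
# Each layer combines textures with SFX-mask-driven emissive, clearcoat, SSS and opacity before a four-to-one blend.
# The result passes through emissive, fuzzy edge, sticker, anodizing, detail, damage, wind, decal, RVT, puddle and wetness functions.
# Static switches add layer debug views, and mobile flattens specular and roughness.
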Material /Game/Environments/Asteria/Sets/NeonCity/Materials/P_FN_Base_NeonCity_4Layer
graph {
 properties {
  bAutomaticallySetUsageInEditor=false
  bTangentSpaceNormal=false
  bUseMaterialAttributes=true
  bUsedWithInstancedStaticMeshes=true
  bUsedWithNanite=true
  bUsedWithSplineMeshes=true
  pixelDepthOffsetMode=PDOM_Legacy
  shadingModel=MSM_FromMaterialExpression
  usedShadingModels=Default Lit | Subsurface | Clear Coat | Two Sided Foliage
 }
 nodes [
  n1 = VectorParameter() {
   group="0 (Expand)  Documentation (RMB - Copy display name) and Debug view"
   parameterName=CheckMyPOI
   sortPriority=-4
   defaultValue=<rgba 1 0 1 0>
  }
  n2 = MaterialFunctionCall!(SFX V3, 1=NamedRerouteUsage(), 2=NamedRerouteUsage(), 3=NamedRerouteUsage(), 4=NamedRerouteUsage()) {materialFunction=/Game/Global/Base4Layers_2022/MaterialFunctions/MF_Layer_4to1}
  n3 = MaterialFunctionCall!(
   Material In=MaterialFunctionCall!(
    In=MaterialFunctionCall!(SFX, Material In=@n2.Result, SFX Textures RGB) {materialFunction=/Game/Global/Base4Layers_2022/MaterialFunctions/MF_EmissiveControls_2}.Result
    SFX Blend=NamedRerouteUsage()
    SFX Textures=NamedRerouteUsage()
   ) {materialFunction=/Game/Global/Base4Layers_2022/MaterialFunctions/MF_FuzzyEdge2}.Result
   SFX Texture=NamedRerouteUsage()
   SFX RGB=NamedRerouteUsage()
  ) {materialFunction=/Game/Global/Base4Layers_2022/MaterialFunctions/MF_StickersPrint}
  n4 = NamedRerouteUsage()
  n5 = NamedRerouteUsage()
  n6 = MaterialFunctionCall!(
   In=MaterialFunctionCall!(
    SFX V3
    Mat In=MaterialFunctionCall!(In=@n3.Result, SFX Blended=@n4, SFX Textures=@n5) {materialFunction=/Game/Global/Base4Layers_2022/MaterialFunctions/MF_Anodizing_New}.Result
   ) {materialFunction=/Game/Global/Base4Layers_2022/MaterialFunctions/MF_DetailsWS}.Result
   Emissive Mask=Constant()
  ) {materialFunction=/Game/Athena/WinterArt/Materials/MF_EmissiveColor}
  MaterialExpressionReroute_70 = Reroute(None=@n6.Result)
  n7 = ScalarParameter() {
   desc=Documentation link
   group="0 (Expand)  Documentation (RMB - Copy display name) and Debug view"
   parameterName=https://confluence.it.epicgames.com/display/FORT/Fortnite+Base+Material+2022
  }
  n8 = SetMaterialAttributes!(MaterialAttributes=@MaterialExpressionReroute_70.MaterialAttributes, World Position Offset=Multiply(A=@n7)) {attributeSetTypes [#1]}
  MaterialExpressionReroute_71 = Reroute(None=@n8)
  MaterialExpressionReroute_76 = Reroute(None=@MaterialExpressionReroute_71.Output)
  n9 = GetMaterialAttributes(None=@MaterialExpressionReroute_76.InputPin) {attributeGetTypes [#2]}
  n10 = Add(A=MaterialFunctionCall!(1=Constant() {r=1}, 2=Constant() {r=2}, 3=Constant() {r=3}, 4=Constant() {r=4}) {materialFunction=/Game/Global/Base4Layers_2022/MaterialFunctions/MF_UV2LerpScalar_Layers_New}.Result) {constB=-1}
  n11 = MaterialFunctionCall("Animation  Phase (0-1)"=Divide(A=Ceil(None=@n10)) {constB=4}, Number of Rows=Constant() {r=2}) {materialFunction=&FlipBook}
  n12 = Multiply(
   A=TextureSample(UVs=@n11.UVs) {samplerType=SAMPLERTYPE_Masks, texture=/Game/Global/Base4Layers_2022/Textures/T_Flipbook_M}.R
   B=Constant3Vector() {constant=<rgba 0 1 0 0>}
  )
  n13 = StaticSwitchParameter(
   True=SetMaterialAttributes!(MaterialAttributes=@n9.MaterialAttributes, Emissive Color=Add(A=@n9.Emissive Color, B=@n12)) {attributeSetTypes [#2]}
   False=@MaterialExpressionReroute_76.InputPin
  ) {
   desc="Uses flipbook projection on available Material Layers "
   group="0 (Expand)  Documentation (RMB - Copy display name) and Debug view"
   parameterName=Layers Debug View Flipbook
   sortPriority=50
  }
  MaterialExpressionReroute_77 = Reroute(None=@n13)
  MaterialExpressionReroute_50 = Reroute(None=@MaterialExpressionReroute_77.Output)
  n14 = StaticSwitchParameter(
   True=SetMaterialAttributes(MaterialAttributes=@MaterialExpressionReroute_50.InputPin, Emissive Color=@n1.RGB) {attributeSetTypes [#2]}
   False=@MaterialExpressionReroute_50.InputPin
  ) {group="0 (Expand)  Documentation (RMB - Copy display name) and Debug view", parameterName=CheckMyPOI, sortPriority=-5}
  MaterialExpressionReroute_45 = Reroute(None=@n14)
  MaterialExpressionReroute_69 = Reroute(None=@MaterialExpressionReroute_45.Output)
  n15 = MaterialFunctionCall(Material_In=@MaterialExpressionReroute_69.MaterialAttributes) {materialFunction=&MF_FN_DestructibleDamage}
  n16 = MaterialFunctionCall!(
   In=MaterialFunctionCall!(Material Attributes=MaterialFunctionCall!(In=@n15.Material_Out) {materialFunction=/Game/Athena/WinterArt/Materials/MF_WPO_WindNew}.Result) {materialFunction=/Game/Creative/Sets/MilitaryBase/Props/MF_FanSpin}.Result
   UseSphericalMappingOverride
  ) {materialFunction=/Game/Packages/Fortress_SharedMaterials/Base_Material_Functions/MF_CurieFallbackEffect}
  MaterialExpressionReroute_68 = Reroute(None=@n16.Result)
  Composite!(None=@MaterialExpressionReroute_68.Result) {desc=Transform to World in case we will use RVT after this point, subgraphName=Tangent to World}
  Composite!(MaterialAttributes=@n6.Result) {subgraphName=Documentation Link}
  Composite(MaterialAttributes=@MaterialExpressionReroute_45.Output) {desc=Game Shared WPO and Fire Effect, subgraphName=WPO And Shared Functions}
  Composite!(InputPin=@MaterialExpressionReroute_71.Output) {subgraphName=Flipbook Debug view}
  Composite(InputPin=@MaterialExpressionReroute_77.Output) {subgraphName=Check POI}
  Constant() {desc="Select page from 1 to 4 ", r=1}
  Constant() {desc="Select page from 1 to 4 ", r=2}
  Constant() {desc="Select page from 1 to 4 ", r=3}
  Constant() {desc="Select page from 1 to 4 ", r=4}
  MaterialExpressionReroute_3 = Reroute(None=@MaterialExpressionReroute_68.Result)
  n17 = GetMaterialAttributes(None=@MaterialExpressionReroute_3.Input) {attributeGetTypes [#3]}
  n18 = SetMaterialAttributes!(MaterialAttributes=@n17.MaterialAttributes, Normal=Transform(None=@n17.Normal)) {attributeSetTypes [#3]}
  MaterialExpressionReroute_2 = Reroute(None=@n18)
  n19 = MaterialFunctionCall!(
   In=MaterialFunctionCall!(MA=@MaterialExpressionReroute_2.Output) {materialFunction=/Game/Global/Base4Layers_2022/MaterialFunctions/MF_DirectDecalProjection}.Result
   "RVT Height Map (Roughness)"
  ) {materialFunction=/Game/Global/Base4Layers_2022/MaterialFunctions/MF_RVTBlending}
  n20 = GetMaterialAttributes(None=@n19.Result) {attributeGetTypes [#4, #3]}
  n21 = SetMaterialAttributes!(
   MaterialAttributes=@n20.MaterialAttributes
   Specular=Constant() {r=0.5}
   Roughness=Constant() {r=0.8}
   Base Color=@n20.MaterialAttributes
   Normal=@n20.MaterialAttributes
  ) {attributeSetTypes [
    #5
    #6
    #4
    #3
   ]}
  n22 = ShadingPathSwitch(Default=@n19.Result, Mobile=@n21) {desc=Migrated from FeatureLevelSwitch}
  MaterialFunctionCall(MA In=MaterialFunctionCall!(In=MaterialFunctionCall!(In=@n22) {materialFunction=/Game/Environments/Asteria/Sets/NeonCity/MaterialFunctions/MF_WaterPuddlesBuilding}.Result) {materialFunction=/Game/Environments/Asteria/Sets/NeonCity/MaterialFunctions/MF_WaterWetness}.Result) {materialFunction=&MF_CreativeMoveTool}
  n23 = TextureObjectParameter() {
   group=4_Layer_Nanite
   parameterName="Normals  Nanite 4"
   samplerType=SAMPLERTYPE_Normal
   sortPriority=0
   texture=/Engine/EngineMaterials/FlatNormal
  }
  n24 = TextureObjectParameter() {
   group=4_Layer_Nanite
   parameterName="SpecularMasks  Nanite 4"
   samplerType=SAMPLERTYPE_Masks
   sortPriority=0
   texture=&T_Fortnite_Default_S
  }
  n25 = TextureObjectParameter() {group=4_Layer_Nanite, parameterName=Diffuse Nanite 4, sortPriority=0, texture=/Game/Global/Textures/Default/Blanks/T_MidgreysRGB}
  n26 = StaticBoolParameter() {desc="Enables SFX for Opacity Mask, mast have Masked blending set in MI", group=4_Layer SFX, parameterName=SFX OpacityMask 4, sortPriority=1}
  n27 = TextureObjectParameter() {group=4_Layer, parameterName=Diffuse_Texture_4, sortPriority=0, texture=&T_MidgreysRGB}
  n28 = VectorParameter() {
   desc=Use R or G or B to select your SFX Mask Channel
   group=0_SFX Global
   parameterName=SFX_Tint_Ch_4
   sortPriority=4
   defaultValue=<rgba 1 0 0 0>
  }
  n29 = TextureObjectParameter() {
   group=4_Layer
   parameterName=SpecularMasks_4
   samplerType=SAMPLERTYPE_Masks
   sortPriority=0
   texture=&T_Fortnite_Default_S
  }
  n30 = TextureObjectParameter() {group=4_Layer, parameterName=Normals_Texture_4, samplerType=SAMPLERTYPE_Normal, sortPriority=0, texture=&FlatNormal}
  n31 = VectorParameter() {
   group=4_Layer SFX
   parameterName="SFX_SSS Tint(Expand)_4"
   primitiveDataIndex=2
   sortPriority=0
   channelNames {a=SSS Depth 0-1, b="", g="", r=""}
  }
  n32 = VectorParameter() {
   desc="Use R and G channels of this input to controll ClearCoat amount and Rounghness. It uses SFX texture to mask area affected. "
   group=4_Layer SFX
   parameterName="SFX_ClearCoat(Expand)_4"
   sortPriority=-3
   channelNames {a=Not Used, b=Not Used, g=Clearcoat Roughness, r=Clearcoat Value}
   defaultValue=<rgba 1 0 0 0>
  }
  n33 = VectorParameter() {group=4_Layer SFX, parameterName=SFX_Clearcoat_Ch_4, sortPriority=-4, defaultValue=<rgba 0 0 0 0.5>}
  n34 = ScalarParameter() {desc=If you need to cancel shared Mask texture on specific layer, group=4_Layer SFX, parameterName=SFX_OpacityMaskCancel_4, sortPriority=3}
  n35 = VectorParameter() {
   desc="Emissive color, by default mult by Base color. You can control that blending via Alpha channel of this input. Uses SFX texture as the Mask"
   group=4_Layer SFX
   parameterName="SFX_Emissive_Tint(Expand)_4"
   sortPriority=-6
   channelNames {a=Base color to Emissive, b="", g="", r=""}
   defaultValue=<rgba 0 0 0 1>
  }
  n36 = MaterialFunctionCall!(
   WSM Scale
   Use WS Mapping
   Use Z
   Use XYZ
   WSM_ProjectionTransitionContrast
   UV2 True
   WSM_HeightLerpUsingAlpha
   WSM_HeightLerpUsingAlpha_Contrast
   UV's
   Normals_NN=@n23
   SpecularMasks_NN=@n24
   Diffuse_NN=@n25
   Use SFX Emissive=StaticBoolParameter() {desc=Enables SFX, group=4_Layer SFX, parameterName=SFX Emissive 4, sortPriority=-8}
   SFX Opacity Mask=@n26
   SFX SSS Two Sided Foliange=StaticBoolParameter() {group=4_Layer SFX, parameterName=SFX_SSS Foliage 4, sortPriority=-2}
   SFX SSS=StaticBoolParameter() {group=4_Layer SFX, parameterName=SFX_SSS 4, sortPriority=-2}
   SFX Clearcoat=StaticBoolParameter() {group=4_Layer SFX, parameterName="SFX_Clearcoat  4", sortPriority=-5}
   Diffuse=@n27
   World Position Z
   SFX Texture Complete
   SFX Tint Channel=@n28.RGB
   SpecularMasks=@n29
   Normals=@n30
   Normals Power=ScalarParameter() {defaultValue=1, group=4_Layer, parameterName=NormalsValue_4, sortPriority=1}
   "SFX  SSS Channel"=VectorParameter() {group=4_Layer SFX, parameterName=SFX_SSS_Ch_4, sortPriority=-1}.RGB
   SSS Color Value=AppendVector(A=@n31, B=@n31)
   "Clearcoat (R) CC Roughness (G)"=@n32
   Clearcoat Mask Channal=@n33.RGB
   SFX Opacilty Mask Channel=VectorParameter() {group=4_Layer SFX, parameterName=SFX_OpacityMask_Ch_4, sortPriority=2}.RGB
   Opacilty Mask Cancel=@n34
   "SFX Emissive  Channel"=VectorParameter() {group=4_Layer SFX, parameterName=SFX_Emissive_Ch_4, sortPriority=-7}.RGB
   Emissive Tint=AppendVector(A=@n35, B=@n35)
   Data Ramp RGB
   Output Diffuse Only
   Output Normals Only
  ) {materialFunction=/Game/Global/Base4Layers_2022/MaterialFunctions/MF_CombineLayers}
  NamedRerouteDeclaration(None=@n36.Result) {name=L4, nodeColor=<rgba 0.6172 0 1 1>}
  n37 = TextureObjectParameter() {
   group=3_Layer_Nanite
   parameterName="Normals  Nanite 3"
   samplerType=SAMPLERTYPE_Normal
   sortPriority=0
   texture=&FlatNormal
  }
  n38 = TextureObjectParameter() {
   group=3_Layer_Nanite
   parameterName="SpecularMasks  Nanite 3"
   samplerType=SAMPLERTYPE_Masks
   sortPriority=0
   texture=&T_Fortnite_Default_S
  }
  n39 = TextureObjectParameter() {group=3_Layer_Nanite, parameterName=Diffuse Nanite 3, sortPriority=0, texture=&T_MidgreysRGB}
  n40 = StaticBoolParameter() {desc="Enables SFX for Opacity Mask, mast have Masked blending set in MI", group=3_Layer SFX, parameterName=SFX OpacityMask 3, sortPriority=1}
  n41 = TextureObjectParameter() {group=3_Layer, parameterName=Diffuse_Texture_3, sortPriority=0, texture=&T_MidgreysRGB}
  n42 = VectorParameter() {
   desc=Use R or G or B to select your SFX Mask Channel
   group=0_SFX Global
   parameterName=SFX_Tint_Ch_3
   sortPriority=4
   defaultValue=<rgba 1 0 0 0>
  }
  n43 = TextureObjectParameter() {
   group=3_Layer
   parameterName=SpecularMasks_3
   samplerType=SAMPLERTYPE_Masks
   sortPriority=0
   texture=&T_Fortnite_Default_S
  }
  n44 = TextureObjectParameter() {group=3_Layer, parameterName=Normals_Texture_3, samplerType=SAMPLERTYPE_Normal, sortPriority=0, texture=&FlatNormal}
  n45 = VectorParameter() {
   group=3_Layer SFX
   parameterName="SFX_SSS Tint(Expand)_3"
   primitiveDataIndex=2
   sortPriority=0
   channelNames {a=SSS Depth 0-1, b="", g="", r=""}
   defaultValue=<rgba 0 0 0 0.5>
  }
  n46 = VectorParameter() {
   desc="Use R and G channels of this input to controll ClearCoat amount and Rounghness. It uses SFX texture to mask area affected. "
   group=3_Layer SFX
   parameterName="SFX_ClearCoat(Expand)_3"
   sortPriority=-3
   channelNames {a=Not Used, b=Not Used, g=Clearcoat Roughness, r=Clearcoat Value}
   defaultValue=<rgba 1 0 0 0>
  }
  n47 = ScalarParameter() {desc=If you need to cancel shared Mask texture on specific layer, group=3_Layer SFX, parameterName=SFX_OpacityMaskCancel_3, sortPriority=3}
  n48 = VectorParameter() {
   desc="Emissive color, by default mult by Base color. You can control that blending via Alpha channel of this input. Uses SFX texture as the Mask"
   group=3_Layer SFX
   parameterName="SFX_Emissive_Tint(Expand)_3"
   sortPriority=-6
   channelNames {a=Base color to Emissive, b="", g="", r=""}
   defaultValue=<rgba 0 0 0 1>
  }
  n49 = MaterialFunctionCall(
   Normals_NN=@n37
   SpecularMasks_NN=@n38
   Diffuse_NN=@n39
   Use SFX Emissive=StaticBoolParameter() {desc=Enables SFX, group=3_Layer SFX, parameterName=SFX Emissive 3, sortPriority=-8}
   SFX Opacity Mask=@n40
   SFX SSS Two Sided Foliange=StaticBoolParameter() {group=3_Layer SFX, parameterName=SFX_SSS Foliage 3, sortPriority=-2}
   SFX SSS=StaticBoolParameter() {group=3_Layer SFX, parameterName=SFX_SSS 3, sortPriority=-2}
   SFX Clearcoat=StaticBoolParameter() {group=3_Layer SFX, parameterName="SFX_Clearcoat  3", sortPriority=-5}
   Diffuse=@n41
   SFX Tint Channel=@n42.RGB
   SpecularMasks=@n43
   Normals=@n44
   Normals Power=ScalarParameter() {defaultValue=1, group=3_Layer, parameterName=NormalsValue_3, sortPriority=1}
   "SFX  SSS Channel"=VectorParameter() {group=3_Layer SFX, parameterName=SFX_SSS_Ch_3, sortPriority=-1}.RGB
   SSS Color Value=@n45
   "Clearcoat (R) CC Roughness (G)"=@n46
   Clearcoat Mask Channal=VectorParameter() {group=3_Layer SFX, parameterName=SFX_Clearcoat_Ch_3, sortPriority=-4}.RGB
   SFX Opacilty Mask Channel=VectorParameter() {group=3_Layer SFX, parameterName=SFX_OpacityMask_Ch_3, sortPriority=2}.RGB
   Opacilty Mask Cancel=@n47
   "SFX Emissive  Channel"=VectorParameter() {group=3_Layer SFX, parameterName=SFX_Emissive_Ch_3, sortPriority=-7}.RGB
   Emissive Tint=AppendVector(A=@n48, B=@n48)
  ) {materialFunction=&MF_CombineLayers}
  NamedRerouteDeclaration(None=@n49.Result) {name=L3, nodeColor=<rgba 0.6406 1 0 1>}
  n50 = TextureObjectParameter() {
   group=1_Layer_Nanite
   parameterName="Normals  Nanite"
   samplerType=SAMPLERTYPE_Normal
   sortPriority=0
   texture=&FlatNormal
  }
  n51 = TextureObjectParameter() {
   group=1_Layer_Nanite
   parameterName="SpecularMasks  Nanite"
   samplerType=SAMPLERTYPE_Masks
   sortPriority=0
   texture=&T_Fortnite_Default_S
  }
  n52 = TextureObjectParameter() {group=1_Layer_Nanite, parameterName=Diffuse Nanite, sortPriority=0, texture=&T_MidgreysRGB}
  n53 = StaticBoolParameter() {desc="Enables SFX for Opacity Mask, mast have Masked blending set in MI", group=1_Layer SFX, parameterName=SFX OpacityMask, sortPriority=1}
  n54 = VectorParameter() {
   desc=Use R or G or B to select your SFX Mask Channel
   group=0_SFX Global
   parameterName=SFX_Tint_Ch
   sortPriority=4
   defaultValue=<rgba 1 0 0 0>
  }
  n55 = TextureObjectParameter() {
   group=1_Layer
   parameterName=SpecularMasks
   samplerType=SAMPLERTYPE_Masks
   sortPriority=0
   texture=&T_Fortnite_Default_S
  }
  n56 = TextureObjectParameter() {group=1_Layer, parameterName=Normals, samplerType=SAMPLERTYPE_Normal, sortPriority=0, texture=&FlatNormal}
  n57 = VectorParameter() {
   group=1_Layer SFX
   parameterName="SFX_SSS Tint(Expand)"
   sortPriority=0
   channelNames {a=SSS Depth 0-1, b="", g="", r=""}
   defaultValue=<rgba 0 0 0 0.5>
  }
  n58 = VectorParameter() {
   desc=Use R and G channels of this input to controll ClearCoat amount and Rounghness. It uses SFX texture to mask area affected.
   group=1_Layer SFX
   parameterName="SFX_ClearCoat(Expand)"
   sortPriority=-3
   channelNames {a=Not Used, b=Not Used, g=Clearcoat Roughness, r=Clearcoat Value}
   defaultValue=<rgba 1 0 0 0>
  }
  n59 = ScalarParameter() {desc=If you need to cancel shared Mask texture on specific layer, group=1_Layer SFX, parameterName=SFX_OpacityMaskCancel, sortPriority=3}
  n60 = VectorParameter() {
   desc="Emissive color, by default mult by Base color. You can control that blending via Alpha channel of this input. Uses SFX texture as the Mask"
   group=1_Layer SFX
   parameterName="SFX_Emissive_Tint(Expand)"
   sortPriority=-6
   channelNames {a=Base color to Emissive, b="", g="", r=""}
   defaultValue=<rgba 0 0 0 1>
  }
  n61 = ScalarParameter() {
   bUseCustomPrimitiveData=true
   desc="For now Scalar, this is possible alternative to Texture Swatch. Uses single ramp texture shared between all 4 layers and can have indepent row selected per layer. Enabled globally via function. Will disable use of color swatch if enabled. Diffuse tint still can be used if needed via separate mask "
   group=Y_CPD
   parameterName=CPD_PackHSV_1
   primitiveDataIndex=30
   sliderMin=1
  }
  n62 = MaterialFunctionCall(
   Normals_NN=@n50
   SpecularMasks_NN=@n51
   Diffuse_NN=@n52
   Use SFX Emissive=StaticBoolParameter() {desc=Enables SFX, group=1_Layer SFX, parameterName=SFX Emissive, sortPriority=-8}
   SFX Opacity Mask=@n53
   SFX SSS Two Sided Foliange=StaticBoolParameter() {group=1_Layer SFX, parameterName=SFX_SSS Foliage, sortPriority=-2}
   SFX SSS=StaticBoolParameter() {group=1_Layer SFX, parameterName=SFX_SSS, sortPriority=-2}
   SFX Clearcoat=StaticBoolParameter() {group=1_Layer SFX, parameterName=SFX_Clearcoat, sortPriority=-5}
   Diffuse=TextureObjectParameter() {group=1_Layer, parameterName=Diffuse, sortPriority=0, texture=&T_MidgreysRGB}
   SFX Tint Channel=@n54.RGB
   SpecularMasks=@n55
   Normals=@n56
   Normals Power=ScalarParameter() {defaultValue=1, group=1_Layer, parameterName=NormalsValue, sortPriority=1}
   "SFX  SSS Channel"=VectorParameter() {group=1_Layer SFX, parameterName=SFX_SSS_Ch, sortPriority=-1}.RGB
   SSS Color Value=AppendVector(A=@n57, B=@n57)
   "Clearcoat (R) CC Roughness (G)"=@n58
   Clearcoat Mask Channal=VectorParameter() {group=1_Layer SFX, parameterName=SFX_Clearcoat_Chl, sortPriority=-4}.RGB
   SFX Opacilty Mask Channel=VectorParameter() {group=1_Layer SFX, parameterName=SFX_OpacityMask_Ch_1, sortPriority=2}.RGB
   Opacilty Mask Cancel=@n59
   "SFX Emissive  Channel"=VectorParameter() {group=1_Layer SFX, parameterName=SFX_Emissive_Ch, sortPriority=-7}.RGB
   Emissive Tint=AppendVector(A=@n60, B=@n60)
   Data Ramp RGB=MaterialFunctionCall!(In=@n61) {materialFunction=/Game/Global/Base4Layers_2022/MaterialFunctions/MF_ColorTint_Unpack}.Result
  ) {materialFunction=&MF_CombineLayers}
  NamedRerouteDeclaration(None=@n62.SFX) {name=SFX L1, nodeColor=<rgba 0 0.8594 1 1>}
  NamedRerouteDeclaration(None=@n49.SFX) {name=SFX L3, nodeColor=<rgba 1 0 0.8437 1>}
  NamedRerouteDeclaration(None=@n36.SFX) {name=SFX L4, nodeColor=<rgba 0 1 0.5781 1>}
  NamedRerouteDeclaration(None=MaterialFunctionCall!(1=NamedRerouteUsage(), 2=NamedRerouteUsage(), 3=NamedRerouteUsage(), 4=NamedRerouteUsage()) {materialFunction=/Game/Global/Base4Layers_2022/MaterialFunctions/MF_UV2Lerp4Layers_Vec3_New}.Result) {name=SFX Blend, desc=Will use layer UV's, nodeColor=<rgba 0.8047 1 0 1>}
  NamedRerouteDeclaration(None=@n62.Result) {name=L1, nodeColor=<rgba 0 0.5078 1 1>}
  n63 = TextureObjectParameter() {
   group=2_Layer_Nanite
   parameterName="Normals  Nanite 2"
   samplerType=SAMPLERTYPE_Normal
   sortPriority=0
   texture=&FlatNormal
  }
  n64 = TextureObjectParameter() {
   group=2_Layer_Nanite
   parameterName="SpecularMasks  Nanite 2"
   samplerType=SAMPLERTYPE_Masks
   sortPriority=0
   texture=&T_Fortnite_Default_S
  }
  n65 = TextureObjectParameter() {group=2_Layer_Nanite, parameterName=Diffuse Nanite 2, sortPriority=0, texture=&T_MidgreysRGB}
  n66 = ScalarParameter() {
   bUseCustomPrimitiveData=true
   desc="For now Scalar, this is possible alternative to Texture Swatch. Uses single ramp texture shared between all 4 layers and can have indepent row selected per layer. Enabled globally via function. Will disable use of color swatch if enabled. Diffuse tint still can be used if needed via separate mask "
   group=Y_CPD
   parameterName=CPD_PackedHSV_2
   primitiveDataIndex=31
   sliderMin=1
  }
  n67 = VectorParameter() {
   desc=Use R or G or B to select your SFX Mask Channel
   group=0_SFX Global
   parameterName=SFX_Tint_Ch_2
   sortPriority=4
   defaultValue=<rgba 1 0 0 0>
  }
  n68 = TextureObjectParameter() {
   group=2_Layer
   parameterName=SpecularMasks_2
   samplerType=SAMPLERTYPE_Masks
   sortPriority=0
   texture=&T_Fortnite_Default_S
  }
  n69 = TextureObjectParameter() {group=2_Layer, parameterName=Normals_Texture_2, samplerType=SAMPLERTYPE_Normal, sortPriority=0, texture=&FlatNormal}
  n70 = VectorParameter() {
   group=2_Layer SFX
   parameterName="SFX_SSS Tint (Expand)_2"
   primitiveDataIndex=2
   sortPriority=0
   channelNames {a=SSS Depth 0-1, b="", g="", r=""}
   defaultValue=<rgba 0 0 0 0.5>
  }
  n71 = VectorParameter() {
   desc="Use R and G channels of this input to controll ClearCoat amount and Rounghness. It uses SFX texture to mask area affected. "
   group=2_Layer SFX
   parameterName="SFX_ClearCoat(Expand)_2"
   sortPriority=-3
   channelNames {a=Not Used, b=Not Used, g=Clearcoat Roughness, r=Clearcoat Value}
   defaultValue=<rgba 1 0 0 0>
  }
  n72 = ScalarParameter() {desc=If you need to cancel shared Mask texture on specific layer, group=2_Layer SFX, parameterName=SFX_OpacityMaskCancel_2, sortPriority=3}
  n73 = VectorParameter() {
   desc="Emissive color, by default mult by Base color. You can control that blending via Alpha channel of this input. Uses SFX texture as the Mask"
   group=2_Layer SFX
   parameterName="SFX_Emissive_Tint(Expand)_2"
   sortPriority=-6
   channelNames {a=Base color to Emissive, b="", g="", r=""}
   defaultValue=<rgba 0 0 0 1>
  }
  n74 = TextureObjectParameter() {group=2_Layer, parameterName=Diffuse_Texture_2, sortPriority=0, texture=&T_MidgreysRGB}
  n75 = MaterialFunctionCall(
   Normals_NN=@n63
   SpecularMasks_NN=@n64
   Diffuse_NN=@n65
   Use SFX Emissive=StaticBoolParameter() {desc=Enables SFX, group=2_Layer SFX, parameterName=SFX Emissive 2, sortPriority=-8}
   SFX Opacity Mask=StaticBoolParameter() {desc=Enables SFX, group=2_Layer SFX, parameterName=SFX OpacityMask 2, sortPriority=1}
   SFX SSS Two Sided Foliange=StaticBoolParameter() {group=2_Layer SFX, parameterName=SFX_SSS Foliage 2, sortPriority=-2}
   SFX SSS=StaticBoolParameter() {group=2_Layer SFX, parameterName=SFX_SSS 2, sortPriority=-2}
   SFX Clearcoat=StaticBoolParameter() {group=2_Layer SFX, parameterName="SFX_Clearcoat  2", sortPriority=-5}
   Diffuse=@n74
   SFX Tint Channel=@n67.RGB
   SpecularMasks=@n68
   Normals=@n69
   Normals Power=ScalarParameter() {defaultValue=1, group=2_Layer, parameterName=NormalsValue_2, sortPriority=1}
   "SFX  SSS Channel"=VectorParameter() {group=2_Layer SFX, parameterName=SFX_SSS_Ch_2, sortPriority=-1}.RGB
   SSS Color Value=AppendVector(A=@n70, B=@n70)
   "Clearcoat (R) CC Roughness (G)"=@n71
   Clearcoat Mask Channal=VectorParameter() {group=2_Layer SFX, parameterName=SFX_Clearcoat_Ch_2, sortPriority=-4}.RGB
   SFX Opacilty Mask Channel=VectorParameter() {group=2_Layer SFX, parameterName=SFX_OpacityMask_Ch_2, sortPriority=2}.RGB
   Opacilty Mask Cancel=@n72
   "SFX Emissive  Channel"=VectorParameter() {group=2_Layer SFX, parameterName=SFX_Emissive_Ch_2, sortPriority=-7}.RGB
   Emissive Tint=AppendVector(A=@n73, B=@n73)
   Data Ramp RGB=MaterialFunctionCall(In=@n66) {materialFunction=&MF_ColorTint_Unpack}.Result
  ) {materialFunction=&MF_CombineLayers}
  NamedRerouteDeclaration(None=@n75.Result) {name=L2, nodeColor=<rgba 1 0 0.5859 1>}
  n76 = TextureObjectParameter() {
   desc="SFX UniversalTextureMask Uses Dot to select channel used for various parts of material.You can reuse that texture across layers and functions. "
   group=0_SFX Global
   parameterName=SFX_Mask
   samplerType=SAMPLERTYPE_Masks
   sortPriority=0
   texture=/Game/Global/Base4Layers_2022/Textures/T_SFX_M
  }
  NamedRerouteDeclaration(None=@n76) {name=SFX Texture Global, nodeColor=<rgba 1 0 0 1>}
  n77 = TextureObjectParameter() {
   desc="SFX RGB - Graffiti, emissive RGB "
   group=0_SFX Global
   parameterName=SFX_RGB
   sortPriority=1
   texture=/Game/Global/Base4Layers_2022/Textures/T-GraffitiSample_D
  }
  NamedRerouteDeclaration(None=@n77) {name=SFX Texture RGB, nodeColor=<rgba 0 0.5312 1 1>}
  n78 = StaticBoolParameter() {
   desc="Enables SFX Flipbook on any UV channel. Set up to use pack of 4 texture pages to select per layer. You can pack all 4 layers into one SFX shared that way.  Layers will receive own pages from 1 to 4 automatically. "
   group=0_SFX Global
   parameterName=SFX Layers Flipbook UV
   sortPriority=-8
  }
  NamedRerouteDeclaration(None=@n78) {name=SFX Layers Flipbook UV, nodeColor=<rgba 0.8281 0 1 1>}
  NamedRerouteDeclaration(None=@n75.SFX) {name=SFX L2, nodeColor=<rgba 0 1 0.1562 1>}
  NamedRerouteUsage()
  NamedRerouteUsage()
  NamedRerouteUsage()
  NamedRerouteUsage()
  NamedRerouteUsage()
  NamedRerouteUsage()
  NamedRerouteUsage()
  NamedRerouteUsage()
  PinBase() {reroutePins [Input {expression=&self:MaterialExpressionReroute_3}]}
  PinBase!(Output=@n18) {reroutePins [Output {expression=&self:MaterialExpressionReroute_2}]}
  PinBase!() {reroutePins [InputPin {expression=&self:MaterialExpressionReroute_50}]}
  PinBase!(Output=@n14) {reroutePins [Output {expression=&self:MaterialExpressionReroute_45}]}
  PinBase!() {reroutePins [MaterialAttributes {expression=&self:MaterialExpressionReroute_70}]}
  PinBase!(Output=@n8) {reroutePins [Output {expression=&self:MaterialExpressionReroute_71}]}
  PinBase!() {reroutePins [MaterialAttributes {expression=&self:MaterialExpressionReroute_69}]}
  PinBase!(Result=@n16.Result) {reroutePins [Result {expression=&self:MaterialExpressionReroute_68}]}
  PinBase!() {reroutePins [InputPin {expression=&self:MaterialExpressionReroute_76}]}
  PinBase!(Output=@n13) {reroutePins [Output {expression=&self:MaterialExpressionReroute_77}]}
  Reroute(None=NamedRerouteUsage())
  Reroute(None=NamedRerouteUsage())
  Reroute(None=StaticBoolParameter() {group=3_Layer SFX, parameterName=SFX UV2_3, sortPriority=8})
  Reroute(None=StaticBoolParameter() {group=3_Layer SFX, parameterName=SFX UV3_3, sortPriority=9})
  Reroute(None=NamedRerouteUsage())
  Reroute(None=StaticBoolParameter() {group=4_Layer SFX, parameterName=SFX UV2_4, sortPriority=8})
  Reroute(None=StaticBoolParameter() {group=4_Layer SFX, parameterName=SFX UV3_4, sortPriority=9})
  SetMaterialAttributes!(MaterialAttributes, Base Color) {attributeSetTypes [#4]}
  StaticBoolParameter() {group=1_Layer SFX, parameterName=SFX UV3, sortPriority=9}
  StaticBoolParameter() {group=1_Layer SFX, parameterName=SFX UV2, sortPriority=8}
  StaticBoolParameter() {group=2_Layer SFX, parameterName=SFX UV3_2, sortPriority=9}
  StaticBoolParameter() {group=2_Layer SFX, parameterName=SFX UV2_2, sortPriority=8}
  StaticSwitchParameter(
   True=TextureSample(UVs=Multiply(A=@n11.UVs) {constB=10}) {texture=/Game/Packages/Fortress_SharedMaterials/Textures/T_UV_Checker_D}.RGB
   False=@n12
  ) {
   desc="Uses flipbook projection on available Material Layers "
   group="0 (Expand)  Documentation (RMB - Copy display name) and Debug view"
   parameterName=Layers Debug View Grid
   sortPriority=50
  }
 ]
}
